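# Translucent unlit material for display actors in the PerfCsvPlaybackTool editor utility, with per-primitive color and current-state highlighting.
# An isCurrent custom primitive data value blends in a time-driven sine pulse along vertex normals, swelling the current marker between 80% and full CurrentScale.
# A Fresnel blend of two opacity parameters brightens the rim, then a near-camera fade multiplies the result.
# Color comes from custom primitive data, so instances can share one material.
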Material /Game/NeverCook/EditorUtilities/PerfTools/PerfCsvPlaybackTool/Actors/PCPT_Display_Mat
graph {
 properties {blendMode=BLEND_Translucent, pixelDepthOffsetMode=PDOM_Legacy, shadingModel=MSM_Unlit, usedShadingModels=Unlit}
 nodes [
  n1 = ScalarParameter() {defaultValue=300, parameterName=CurrentScale}
  n2 = Multiply(A=@n1) {constB=0.8}
  n3 = MaterialFunctionCall(Value=Frac(None=Time())) {materialFunction=&LinearSine}
  LinearInterpolate(
   B=Multiply(A=VertexNormalWS(), B=LinearInterpolate(A=@n2, B=@n1, Alpha=@n3.Linear Sine)) {constB=100}
   Alpha=ScalarParameter() {bUseCustomPrimitiveData=true, parameterName=isCurrent}
  )
  n4 = ScalarParameter() {defaultValue=0.15, parameterName=Opacity}
  n5 = ScalarParameter() {defaultValue=150, parameterName=Opacity1}
  n6 = Fresnel(ExponentIn=ScalarParameter() {defaultValue=8, parameterName=Exp}) {baseReflectFraction=0}
  Multiply(
   A=Saturate(None=LinearInterpolate(A=@n4, B=@n5, Alpha=@n6))
   B=MaterialFunctionCall(Distance=ScalarParameter() {defaultValue=1500, parameterName=fadeDist}) {materialFunction=&Near_Camera_Fade}.Result
  )
  VectorParameter() {bUseCustomPrimitiveData=true, parameterName=Color, primitiveDataIndex=1, defaultValue=<rgba 0.9062 0.9062 0.9062 1>}
 ]
}
