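# PerfCsvPlaybackTool_EWidget is an editor utility widget that loads perf CSV captures, visualizes stats in the level and plays them back on a timeline.
# View position and forward columns, adjusted by an offset transform, become points colored by per-stat visualizers, with frustum, threshold and every-other controls.
# Points draw as pooled transient actors or through one instanced actor.
# Tick drives playback, camera readout and saving of settings and last CSV data to a SaveGame.
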EditorUtilityWidgetBlueprint /Game/NeverCook/EditorUtilities/PerfTools/PerfCsvPlaybackTool/PerfCsvPlaybackTool_EWidget
graphs [
 EventGraph {nodes [
   "On Clicked (btnPlay)"!() -> (OutputDelegate, then, widget:User Widget Object Reference, event:Pointer Event Structure) {is=title K2Node_ComponentBoundEvent}
   > setPlaying!(execute, self, isPlaying:bool=NOT Boolean(A=Get isPlaying() -> (isPlaying:bool))) -> (then)
   "On Clicked (btnLoad)"!() -> (OutputDelegate, then, widget:User Widget Object Reference, event:Pointer Event Structure) {is=title K2Node_ComponentBoundEvent}
   > initCSV()
   n1 = "On Stat Inclusion Changed (PCPT_ShowStatsList_Widget)"!() -> (OutputDelegate, then, widget:PCPT Stat Available List Item Object Reference, stat:String, included:bool) {is=title K2Node_ComponentBoundEvent}
   > handleStatInclusionChanged!(execute, self, widget:PCPT Stat Available List Item Object Reference=@n1.widget, stat:String=@n1.stat, included:bool=@n1.included) -> (then)
   "On Clicked (btnClearDraw_1)"!() -> (OutputDelegate, then, widget:User Widget Object Reference, event:Pointer Event Structure) {is=title K2Node_ComponentBoundEvent}
   > SetVisibility(
    self=Get PCPT_ShowStatsList_Widget() -> (PCPT_ShowStatsList_Widget:PCPT Show Stats List Widget Object Reference)
    InVisibility=SelfHitTestInvisible
   )
   "On Clicked (btnJumpTo)"!() -> (OutputDelegate, then, widget:User Widget Object Reference, event:Pointer Event Structure) {is=title K2Node_ComponentBoundEvent}
   > updatePawnPosition!(execute, self, Force:bool=true) -> (then)
   n2 = "On Value Committed (spinScale)"!() -> (OutputDelegate, then, InValue:float, CommitMethod:ETextCommit Enum=Default) {is=title K2Node_ComponentBoundEvent}
   > Set shouldSave(shouldSave:bool=true) -> (Output_Get:bool)
   > n3 = For Each Loop(Array:Array of PCPT Display Utility Actors Object References=Get currentPCPTActors() -> (currentPCPTActors:Array of PCPT Display Utility Actors Object References)) -> (Array Element:PCPT Display Utility Actors Object Reference)
   > Is Valid(InputObject=@n3)
   > setScale!(execute, self:PCPT Display Utility Actors Object Reference=@n3, scale:double=@n2.InValue) -> (then)
   "On Value Committed (spinEveryOther)"!() -> (OutputDelegate, then, InValue:float, CommitMethod:ETextCommit Enum=Default) {is=title K2Node_ComponentBoundEvent}
   > Set shouldSave(shouldSave=true)
   > Is Drawing Active!(execute:Exec) -> (Yep_Actors:Exec, Yep_Instances:Exec, Nope:Exec) {is=title K2Node_MacroInstance}
   > Set everyOther_Counter(everyOther_Counter:int=0) -> (Output_Get:int=0)
   > drawDataInWorld()
   "On Selection Changed (ddMoveType)"!() -> (OutputDelegate, then, SelectedItem:String, SelectionType:ESelectInfo Enum) {is=title K2Node_ComponentBoundEvent}
   n4 = GetSelectedIndex!(self:"Combo Box (String) Object Reference"=Get ddMoveType() -> (ddMoveType:"Combo Box (String) Object Reference")) -> (ReturnValue:int=0)
   n5 = Literal enum PCPT_CamMoveType_Enum!(Enum:PCPT_CamMoveType_Enum Enum=Byte to Enum PCPT_CamMoveType_Enum!(Byte:Byte="To Byte (Integer)"!(InInt:int=@n4) -> (ReturnValue:Byte=0)) -> (ReturnValue:PCPT_CamMoveType_Enum Enum) {is=title K2Node_CastByteToEnum}) -> (ReturnValue:PCPT_CamMoveType_Enum Enum) {is=title K2Node_EnumLiteral}
   > SetToolTipText!(
    execute
    self:Widget Object Reference=Get ddMoveType()
    InToolTipText:"Text (by ref)"="To Text (String)"(InString=Select!(
      NewEnumerator0:String=Do not move the editor camera during play
      NewEnumerator1:String=When the timeline updates: Move editor camera to the perspective of the data point
      NewEnumerator2:String=When the timeline updates: Move the camera to focus the data points location
      Index:PCPT_CamMoveType_Enum Enum=@n5
     ) -> (ReturnValue:String))
   ) -> (then)
   > Set shouldSave(shouldSave=true)
   n6 = "On Check State Changed (cbThresholdGreater)"!() -> (OutputDelegate, then, bIsChecked:bool) {is=title K2Node_ComponentBoundEvent}
   > Set shouldSave(shouldSave=true)
   > n7 = Is Drawing Active()
   > drawDataInWorld()
   n8 = getDrawingInstanceActor!(execute=@n7.Yep_Instances, self, createIfNone:bool=false) -> (then, Output:PCPT Display Utility Actor Instanced Object Reference, wasNew:bool)
   > Is Valid(InputObject=@n8)
   > setValueThresholdBool!(execute, self:PCPT Display Utility Actor Instanced Object Reference=@n8, hgherThanThreshold:bool=@n6.bIsChecked) -> (then)
   n9 = "On Value Committed (spinValueThreshold)"!() -> (OutputDelegate, then, InValue:float, CommitMethod:ETextCommit Enum=Default) {is=title K2Node_ComponentBoundEvent}
   > Set currentTheshold!(
    execute
    currentTheshold:double=@n9.InValue
    self:PCPT Timeline Widget Object Reference=Get timeline() -> (timeline:PCPT Timeline Widget Object Reference)
   ) -> (then, Output_Get:double)
   > Is Drawing Active()
   > drawDataInWorld()
   n10 = "On Value Committed (spinFrustumScale)"!() -> (OutputDelegate, then, InValue:float, CommitMethod:ETextCommit Enum=Default) {is=title K2Node_ComponentBoundEvent}
   > Set shouldSave(shouldSave=true)
   > Is Drawing Active()
   > n11 = getAllDrawingActors!(execute, self) -> (then, OutActors:Array of PCPT Display Utility Actors Object References)
   > n12 = For Each Loop(Array=@n11)
   > setFrustumRenderScale!(execute, self:PCPT Display Utility Actors Object Reference=@n12, frustumDist:double=@n10.InValue) -> (then)
   n13 = "On Check State Changed (cbFrustumRender)"!() -> (OutputDelegate, then, bIsChecked:bool) {is=title K2Node_ComponentBoundEvent}
   > Set shouldSave(shouldSave=true)
   > n14 = Is Drawing Active()
   > n15 = getAllDrawingActors()
   > n16 = For Each Loop(Array=@n15)
   > setFrustumVisible!(execute, self:PCPT Display Utility Actors Object Reference=@n16, visible:bool=@n13.bIsChecked) -> (then)
   n17 = getDrawingInstanceActor(execute=@n14.Yep_Instances, createIfNone=false)
   > Is Valid(InputObject=@n17)
   > setFrustumVisible(self:PCPT Display Utility Actor Instanced Object Reference=@n17, visible=@n13.bIsChecked)
   n18 = "On Value Committed (spinDataSmoothing)"!() -> (OutputDelegate, then, InValue:float, CommitMethod:ETextCommit Enum=Default) {is=title K2Node_ComponentBoundEvent}
   > n19 = Switch on ETextCommit!(execute, Selection:ETextCommit Enum=@n18.CommitMethod) -> (Default:Exec, OnEnter:Exec, OnUserMovedFocus:Exec, OnCleared:Exec) {is=title K2Node_SwitchEnum}
   Set shouldSave(execute=[@n19.OnEnter, @n19.OnUserMovedFocus], shouldSave=true)
   > Set averageOver!(execute, averageOver:int=Truncate(A=@n18.InValue), self:PCPT Timeline Widget Object Reference=Get timeline()) -> (then, Output_Get:int=0)
   > setSelectedStat!(execute, self, stat:String=Get currentStatToVisualize() -> (currentStatToVisualize:String)) -> (then)
   n20 = "On Value Committed (spinFrustumOpacity)"!() -> (OutputDelegate, then, InValue:float, CommitMethod:ETextCommit Enum=Default) {is=title K2Node_ComponentBoundEvent}
   > Set shouldSave(shouldSave=true)
   > Is Drawing Active()
   > n21 = getAllDrawingActors()
   > n22 = For Each Loop(Array=@n21)
   > setFrustumRenderOpacity!(execute, self:PCPT Display Utility Actors Object Reference=@n22, frustumOpacity:double=@n20.InValue) -> (then)
   n23 = "On Check State Changed (cbFrustumOnlyCurrent)"!() -> (OutputDelegate, then, bIsChecked:bool) {is=title K2Node_ComponentBoundEvent}
   > Set shouldSave(shouldSave=true)
   > n24 = Is Drawing Active()
   > n25 = getAllDrawingActors()
   > n26 = For Each Loop(Array=@n25)
   n27 = NOT Boolean(A=@n23.bIsChecked)
   > setFrustumForcedOn!(execute, self:PCPT Display Utility Actors Object Reference=@n26, forcedOn:bool=@n27) -> (then)
   n28 = getDrawingInstanceActor(execute=@n24.Yep_Instances, createIfNone=false)
   > Is Valid(InputObject=@n28)
   > setFrustumForcedOn(self:PCPT Display Utility Actor Instanced Object Reference=@n28, forcedOn=@n27)
   n29 = "On Value Changed (spinEveryOther)"!() -> (OutputDelegate, then, InValue:float) {is=title K2Node_ComponentBoundEvent}
   > n30 = Is Drawing Active()
   n31 = getDrawingInstanceActor(execute=@n30.Yep_Instances, createIfNone=false)
   > Is Valid(InputObject=@n31)
   > setDrawEveryOther!(execute, self:PCPT Display Utility Actor Instanced Object Reference=@n31, everyOther:double=@n29.InValue) -> (then)
   n32 = "On Value Changed (spinValueThreshold)"!() -> (OutputDelegate, then, InValue:float) {is=title K2Node_ComponentBoundEvent}
   > Set currentTheshold(currentTheshold=@n32.InValue, self=Get timeline())
   > n33 = Is Drawing Active()
   n34 = getDrawingInstanceActor(execute=@n33.Yep_Instances, createIfNone=false)
   > Is Valid(InputObject=@n34)
   > setValueThreshold!(execute, self:PCPT Display Utility Actor Instanced Object Reference=@n34, threshold:double=getValidatedThreshold!() -> (ReturnValue:double)) -> (then)
   "On Clicked (btnClearThreshold)"!() -> (OutputDelegate, then, widget:User Widget Object Reference, event:Pointer Event Structure) {is=title K2Node_ComponentBoundEvent}
   > SetValue!(execute, self:Spin Box Object Reference=Get spinValueThreshold() -> (spinValueThreshold:Spin Box Object Reference), NewValue:float=0) -> (then)
   "On Clicked (btnDraw)"!() -> (OutputDelegate, then, widget:User Widget Object Reference, event:Pointer Event Structure) {is=title K2Node_ComponentBoundEvent}
   > drawDataInWorld()
   "On Value Committed (spinMaxDataPoints)"!() -> (OutputDelegate, then, InValue:float, CommitMethod:ETextCommit Enum=Default) {is=title K2Node_ComponentBoundEvent}
   > Set shouldSave(shouldSave=true)
   n35 = "On Value Changed (spinFrustumOpacity)"!() -> (OutputDelegate, then, InValue:float) {is=title K2Node_ComponentBoundEvent}
   > Set shouldSave(shouldSave=true)
   > n36 = Is Drawing Active()
   n37 = getDrawingInstanceActor(execute=@n36.Yep_Instances, createIfNone=false)
   > Is Valid(InputObject=@n37)
   > setFrustumRenderOpacity(self:PCPT Display Utility Actor Instanced Object Reference=@n37, frustumOpacity=@n35.InValue)
   n38 = "On Value Changed (spinFrustumScale)"!() -> (OutputDelegate, then, InValue:float) {is=title K2Node_ComponentBoundEvent}
   > Set shouldSave(shouldSave=true)
   > n39 = Is Drawing Active()
   n40 = getDrawingInstanceActor(execute=@n39.Yep_Instances, createIfNone=false)
   > Is Valid(InputObject=@n40)
   > setFrustumRenderScale(self:PCPT Display Utility Actor Instanced Object Reference=@n40, frustumDist=@n38.InValue)
   "On Check State Changed (cbRepeat)"!() -> (OutputDelegate, then, bIsChecked:bool) {is=title K2Node_ComponentBoundEvent}
   > Set shouldSave(shouldSave=true)
   "On Clicked (btnClearDraw)"!() -> (OutputDelegate, then, widget:User Widget Object Reference, event:Pointer Event Structure) {is=title K2Node_ComponentBoundEvent}
   > clearDrawing()
   "On Transform Updated (PCPT_OffsetTransform_EWidget)"() {is=title K2Node_ComponentBoundEvent}
   > Set shouldSave(shouldSave=true)
   "On Value Committed (spinSpeed)"!() -> (OutputDelegate, then, InValue:float, CommitMethod:ETextCommit Enum=Default) {is=title K2Node_ComponentBoundEvent}
   > Set shouldSave(shouldSave=true)
   n41 = "On Value Changed (spinScale)"!() -> (OutputDelegate, then, InValue:float) {is=title K2Node_ComponentBoundEvent}
   > Set shouldSave(shouldSave=true)
   > n42 = getDrawingInstanceActor(createIfNone=false)
   > Is Valid(InputObject=@n42)
   > setScale(self:PCPT Display Utility Actor Instanced Object Reference=@n42, scale=@n41.InValue)
   n43 = "On Item Selection Changed (listStatCounters)"!() -> (
    OutputDelegate
    then
    selectedData:Array of Object References
    selectedWidgets:Array of User Widget Object References
    selectedIndices:Array of Integers
   ) {is=title K2Node_ComponentBoundEvent}
   > Branch(Condition=Is Valid Index!(TargetArray:Array of Object References=@n43.selectedData, IndexToTest:int) -> (ReturnValue:bool))
   > n44 = Cast To TA_GenericListItemData!(execute, Object="Get (a copy)"(Array:Array of Object References=@n43.selectedData) -> (Output:Object Reference)) -> (then, CastFailed, AsTA Generic List Item Data:TA Generic List Item Data Object Reference) {is=title K2Node_DynamicCast}
   > setSelectedStat(stat=Get displayString!(self:TA Generic List Item Data Object Reference=@n44) -> (displayString:String))
   Event Construct()
   > Parent: Construct()
   > initStatVisualizers()
   > signUpForTimelineEvents()
   Event Destruct()
   > Parent: Destruct()
   > clearDrawing()
   n45 = Event Tick()
   > Parent: Tick!(execute, MyGeometry:Geometry Structure=@n45.MyGeometry, InDeltaTime:float=@n45.InDeltaTime) -> (then) {is=title K2Node_CallParentFunction}
   > tick_handlePlay!(execute, self, deltaTime:double=@n45.InDeltaTime) -> (then)
   > saveOnTick()
   n46 = GetLevelViewportCameraInfo!(self:Unreal Editor Subsystem Object Reference=Get UnrealEditorSubsystem()) -> (CameraLocation:Vector, CameraRotation:Rotator, ReturnValue:bool)
   > Branch(Condition=@n46.ReturnValue)
   > setLocation!(
    execute
    self:PCPT Bugit Location Display EWidget Object Reference=Get bugIt_Camera() -> (bugIt_Camera:PCPT Bugit Location Display EWidget Object Reference)
    location:Vector=@n46
    rotation:Rotator=@n46.CameraRotation
   ) -> (then)
  ]}
 Save Hidden Stats {nodes [
   Save Hidden Stats() {is=title K2Node_FunctionEntry}
   > n1 = Cast To PSPT_SaveGame!(execute, Object=Get saveGame_Object() -> (saveGame_Object:Save Game Object Reference)) -> (then, CastFailed, AsPSPT Save Game:PSPT Save Game Object Reference) {is=title K2Node_DynamicCast}
   > n2 = requestHiddenCategories!(execute, self:PCPT Show Stats List Widget Object Reference=Get PCPT_ShowStatsList_Widget()) -> (then, return:Array of Strings)
   > Set HiddenCategories!(execute, HiddenCategories:Array of Strings=@n2, self:PSPT Save Game Object Reference=@n1) -> (then, Output_Get:Array of Strings)
  ]}
 SaveGame_PostLoad {nodes [
   n1 = SaveGame_PostLoad!() -> (then, saveGameExisted:bool, saveGame_Object:Save Game Object Reference) {is=title K2Node_FunctionEntry}
   > n2 = Branch(Condition=@n1)
   > n3 = Cast To PSPT_SaveGame(Object=@n1.saveGame_Object)
   > Set l_save(l_save:PSPT Save Game Object Reference=@n3) -> (Output_Get:PSPT Save Game Object Reference)
   > n4 = Sequence!(execute) -> (then_0, then_1, then_2:Exec, then_3:Exec)
   > initMoveTypeCombobox()
   n5 = Get l_save() -> (l_save:PSPT Save Game Object Reference)
   Set currentFilePath(execute=@n4.then_1, currentFilePath:String=Get filePath!(self:PSPT Save Game Object Reference=@n5) -> (filePath:String)) -> (Output_Get:String)
   > SetIsChecked!(
    execute
    self:Check Box Object Reference=Get cbRepeat() -> (cbRepeat:Check Box Object Reference)
    InIsChecked:bool=Get wrap!(self:PSPT Save Game Object Reference=@n5) -> (wrap:bool)
   ) -> (then)
   > SetSelectedIndex!(
    execute
    self:"Combo Box (String) Object Reference"=Get ddMoveType()
    Index:int=Get moveCamera!(self:PSPT Save Game Object Reference=@n5) -> (moveCamera:int=0)
   ) -> (then)
   > SetValue(
    self=Get spinSpeed() -> (spinSpeed:Spin Box Object Reference)
    NewValue=Get speed!(self:PSPT Save Game Object Reference=@n5) -> (speed:double)
   )
   > setHiddenCategories!(
    execute
    self:PCPT Show Stats List Widget Object Reference=Get PCPT_ShowStatsList_Widget()
    hiddenCategories:Array of Strings=Get HiddenCategories!(self:PSPT Save Game Object Reference=@n5) -> (HiddenCategories:Array of Strings)
   ) -> (then)
   > SetValue(
    self=Get spinEveryOther() -> (spinEveryOther:Spin Box Object Reference)
    NewValue="To Float (Integer)"(InInt=Get everyOther!(self:PSPT Save Game Object Reference=@n5) -> (everyOther:int=0))
   )
   > SetValue(
    self=Get spinScale() -> (spinScale:Spin Box Object Reference)
    NewValue=Get scale!(self:PSPT Save Game Object Reference=@n5) -> (scale:double)
   )
   > Set stats(stats:Map of Strings to PCPT Stat Struct Structures=Get LastCSVData!(self:PSPT Save Game Object Reference=@n5) -> (LastCSVData:Map of Strings to PCPT Stat Struct Structures)) -> (Output_Get:Map of Strings to PCPT Stat Struct Structures)
   > SetValue(
    self=Get spinFrustumScale() -> (spinFrustumScale:Spin Box Object Reference)
    NewValue=Get frustumScale!(self:PSPT Save Game Object Reference=@n5) -> (frustumScale:double)
   )
   > SetValue(
    self=Get spinFrustumOpacity() -> (spinFrustumOpacity:Spin Box Object Reference)
    NewValue=Get frustumOpacity!(self:PSPT Save Game Object Reference=@n5) -> (frustumOpacity:double)
   )
   > SetIsChecked(
    self=Get cbFrustumRender() -> (cbFrustumRender:Check Box Object Reference)
    InIsChecked=Get frustumVisible!(self:PSPT Save Game Object Reference=@n5) -> (frustumVisible:bool)
   )
   > SetIsChecked(
    self=Get cbFrustumOnlyCurrent() -> (cbFrustumOnlyCurrent:Check Box Object Reference)
    InIsChecked=Get frustumOnlyCurrent!(self:PSPT Save Game Object Reference=@n5) -> (frustumOnlyCurrent:bool)
   )
   > loadDataOffsets()
   > SetValue(
    self=Get spinMaxDataPoints() -> (spinMaxDataPoints:Spin Box Object Reference)
    NewValue=Get maxDataPoints!(self:PSPT Save Game Object Reference=@n5) -> (maxDataPoints:double)
   )
   Set currentMap(
    execute=@n4.then_2
    currentMap:Map of Strings to Strings=Get statMap!(self:PSPT Save Game Object Reference=Get l_save()) -> (statMap:Map of Strings to Strings)
   ) -> (Output_Get:Map of Strings to Strings)
   handleCsvDataUpdated(execute=@n4.then_3)
   initMoveTypeCombobox(execute=@n2.else)
   > SetSelectedIndex(self=Get ddMoveType(), Index=0)
   Return Node!(execute, " ":bool=false)
  ]}
 SaveGame_PrepareToSave {nodes [
   SaveGame_PrepareToSave!() -> (then, SaveGameObject:Save Game Object Reference) {is=title K2Node_FunctionEntry}
   > n1 = Cast To PSPT_SaveGame(Object=Get saveGame_Object())
   > Set filePath!(execute, filePath:String=Get currentFilePath() -> (currentFilePath:String), self:PSPT Save Game Object Reference=@n1) -> (then, Output_Get:String)
   > Set moveCamera!(execute, moveCamera:int=getShouldMoveCamera!() -> (index:int=0), self:PSPT Save Game Object Reference=@n1) -> (then, Output_Get:int=0)
   > Set wrap!(execute, wrap:bool=getAllowWrap!() -> (allowWrap:bool), self:PSPT Save Game Object Reference=@n1) -> (then, Output_Get:bool)
   > Set speed!(
    execute
    speed:double=GetValue!(self:Spin Box Object Reference=Get spinSpeed()) -> (ReturnValue:float)
    self:PSPT Save Game Object Reference=@n1
   ) -> (then, Output_Get:double)
   > Set statMap!(
    execute
    statMap:Map of Strings to Strings=Get currentMap() -> (currentMap:Map of Strings to Strings)
    self:PSPT Save Game Object Reference=@n1
   ) -> (then, Output_Get:Map of Strings to Strings)
   > Set LastCSVData!(
    execute
    LastCSVData:Map of Strings to PCPT Stat Struct Structures=Get stats() -> (stats:Map of Strings to PCPT Stat Struct Structures)
    self:PSPT Save Game Object Reference=@n1
   ) -> (then, Output_Get:Map of Strings to PCPT Stat Struct Structures)
   > Set scale!(execute, scale:double=GetValue(self=Get spinScale()), self:PSPT Save Game Object Reference=@n1) -> (then, Output_Get:double)
   n2 = Get spinEveryOther()
   > Set everyOther!(execute, everyOther:int=Floor(A=GetValue(self=@n2)) {is=title K2Node_CallFunction}, self:PSPT Save Game Object Reference=@n1) -> (then, Output_Get:int=0)
   > Set frustumScale!(execute, frustumScale:double=GetValue(self=Get spinFrustumScale()), self:PSPT Save Game Object Reference=@n1) -> (then, Output_Get:double)
   > Set frustumOpacity!(execute, frustumOpacity:double=GetValue(self=Get spinFrustumOpacity()), self:PSPT Save Game Object Reference=@n1) -> (then, Output_Get:double)
   > Set frustumVisible!(execute, frustumVisible:bool=IsChecked(self=Get cbFrustumRender()), self:PSPT Save Game Object Reference=@n1) -> (then, Output_Get:bool)
   > Set frustumOnlyCurrent!(execute, frustumOnlyCurrent:bool=IsChecked(self=Get cbFrustumOnlyCurrent()), self:PSPT Save Game Object Reference=@n1) -> (then, Output_Get:bool)
   > Set maxDataPoints!(execute, maxDataPoints:double=GetValue(self=Get spinMaxDataPoints()), self:PSPT Save Game Object Reference=@n1) -> (then, Output_Get:double)
   > Save Hidden Stats() {is=title K2Node_CallFunction}
   > saveDataOffsets()
   Return Node(" "=false)
  ]}
 cacheViewValues {nodes [
   cacheViewValues() {is=title K2Node_FunctionEntry}
   > n1 = Sequence!(execute) -> (then_0, then_1, then_2)
   > Clear(TargetArray:Array of Vectors=Get cachedViewLocations() -> (cachedViewLocations:Array of Vectors))
   > Clear(TargetArray=Get cachedViewRotations() -> (cachedViewRotations:Array of Vectors))
   n2 = getStatValues!(execute=@n1.then_1, self, Key:"String (by ref)"=MakeLiteralString(Value=View/PosX)) -> (then, OutputPin:"Array of Float (double-precision)s")
   > Set l_posX(l_posX:"Array of Float (double-precision)s"=@n2) -> (Output_Get:"Array of Float (double-precision)s")
   > n3 = getStatValues(Key=MakeLiteralString(Value=View/PosY))
   > Set l_posY(l_posY:"Array of Float (double-precision)s"=@n3) -> (Output_Get:"Array of Float (double-precision)s")
   > n4 = getStatValues(Key=MakeLiteralString(Value=View/PosZ))
   > Set l_posZ(l_posZ:"Array of Float (double-precision)s"=@n4) -> (Output_Get:"Array of Float (double-precision)s")
   > n5 = getStatValues(Key=MakeLiteralString(Value=View/ForwardX))
   > Set l_rotX(l_rotX:"Array of Float (double-precision)s"=@n5) -> (Output_Get:"Array of Float (double-precision)s")
   > n6 = getStatValues(Key=MakeLiteralString(Value=View/ForwardY))
   > Set l_rotY(l_rotY:"Array of Float (double-precision)s"=@n6) -> (Output_Get:"Array of Float (double-precision)s")
   > n7 = getStatValues(Key=MakeLiteralString(Value=View/ForwardZ))
   > Set l_rotZ(l_rotZ:"Array of Float (double-precision)s"=@n7) -> (Output_Get:"Array of Float (double-precision)s")
   n8 = For Each Loop(Exec=@n1.then_2, Array:"Array of Float (double-precision)s"=Get l_posX() -> (l_posX:"Array of Float (double-precision)s")) -> (Array Element:double)
   n9 = Get PCPT_OffsetTransform_EWidget() -> (PCPT_OffsetTransform_EWidget:PCPT Offset Transform EWidget Object Reference)
   n10 = getRotationOffset!(self:PCPT Offset Transform EWidget Object Reference=@n9) -> (ReturnValue:Rotator)
   n11 = getLocationOffset!(self:PCPT Offset Transform EWidget Object Reference=@n9) -> (ReturnValue:Vector)
   n12 = "Get (a copy)"(Array:"Array of Float (double-precision)s"=Get l_posY() -> (l_posY:"Array of Float (double-precision)s"), Dimension 1=@n8.Array Index) -> (Output:double)
   n13 = "Get (a copy)"(Array=Get l_posZ() -> (l_posZ:"Array of Float (double-precision)s"), Dimension 1=@n8.Array Index)
   > Add(
    TargetArray:Array of Vectors=Get cachedViewLocations()
    NewItem:"Vector (by ref)"=InverseTransformLocation!(T:"Transform (by ref)"=MakeTransform(Location=@n11, Rotation=@n10), Location:Vector=MakeVector(X=@n8, Y=@n12, Z=@n13)) -> (ReturnValue:Vector)
   ) {is=title K2Node_CallArrayFunction}
   n14 = "Get (a copy)"(Array=Get l_rotZ() -> (l_rotZ:"Array of Float (double-precision)s"), Dimension 1=@n8.Array Index)
   n15 = "Get (a copy)"(Array=Get l_rotX() -> (l_rotX:"Array of Float (double-precision)s"), Dimension 1=@n8.Array Index)
   n16 = "Get (a copy)"(Array=Get l_rotY() -> (l_rotY:"Array of Float (double-precision)s"), Dimension 1=@n8.Array Index)
   n17 = getRotationOffset(self=Get PCPT_OffsetTransform_EWidget())
   > Add(
    TargetArray=Get cachedViewRotations()
    NewItem=Rotate Vector!(A:Vector=MakeVector(X=@n15, Y=@n16, Z=@n14), B:Rotator=Invert Rotator!(A:Rotator=@n17) -> (ReturnValue:Rotator)) -> (ReturnValue:Vector)
   )
  ]}
 clearDrawing {nodes [
   clearDrawing() {is=title K2Node_FunctionEntry}
   > Clear(TargetArray:Array of PCPT Display Utility Actors Object References=Get cachedPCPTActors() -> (cachedPCPTActors:Array of PCPT Display Utility Actors Object References))
   > Clear(TargetArray=Get currentPCPTActors())
   > n1 = Sequence!(execute) -> (then_0, then_1)
   > n2 = getDrawingInstanceActor(createIfNone=false)
   > Is Valid(InputObject=@n2)
   > Destroy Actor(self=@n2)
   n3 = getAllDrawingActors(execute=@n1.then_1)
   > n4 = For Each Loop(Array:Array of PCPT Display Utility Actors Object References=@n3) -> (Array Element:PCPT Display Utility Actors Object Reference)
   > Destroy Actor(self=@n4)
  ]}
 drawDataInWorld {nodes [
   drawDataInWorld() {is=title K2Node_FunctionEntry}
   > cacheViewValues() {is=title K2Node_CallFunction}
   > n1 = Branch(Condition=Get useActors() -> (useActors:bool))
   > drawDataInWorld_wActors()
   drawDataInWorld_wInstances(execute=@n1.else)
  ]}
 drawDataInWorld_wActors {nodes [
   drawDataInWorld_wActors() {is=title K2Node_FunctionEntry}
   > Clear(TargetArray=Get currentPCPTActors())
   > n1 = getAllDrawingActors()
   > Set cachedPCPTActors(cachedPCPTActors:Array of PCPT Display Utility Actors Object References=@n1) -> (Output_Get:Array of PCPT Display Utility Actors Object References)
   > n2 = For Loop(LastIndex=int - int(A=getDataLength!() -> (ReturnValue:int=0), B=1) {is=title K2Node_CallFunction})
   n3 = For Each Loop(Exec=@n2.Completed, Array=Get cachedPCPTActors())
   > setIsVisible!(execute, self:PCPT Display Utility Actors Object Reference=@n3, isVisible:bool=false) -> (then)
   Set l_index(execute=@n2.LoopBody, l_index:int=@n2) -> (Output_Get:int=0)
   > Set l_isCurrent(l_isCurrent:bool="Equal (Integer)"(A=@n2, B=Get currentIndex() -> (currentIndex:int=0)) {is=title K2Node_CallFunction}) -> (Output_Get:bool)
   > n4 = getStatPositionAndForward!(execute, self, index:int=Get l_index() -> (l_index:int=0)) -> (then, position:Vector, forward:Vector)
   > Set l_start(l_start:Vector=@n4) -> (Output_Get:Vector)
   > Set l_rot(l_rot:Rotator=Rotation From X Vector!(InVec:Vector=@n4.forward) -> (ReturnValue:Rotator)) -> (Output_Get:Rotator)
   > n5 = Try Get PCPTActor!(" ":Exec) -> (true:Exec, false:Exec, actor:PCPT Display Utility Actors Object Reference) {is=title K2Node_MacroInstance}
   n6 = SpawnActorFromObject!(
    execute=@n5.false
    ObjectToUse:Object Reference=/Game/NeverCook/EditorUtilities/PerfTools/PerfCsvPlaybackTool/Actors/PCPT_DisplayUtilityActors
    Location:Vector=Get l_start() -> (l_start:Vector)
    Rotation:Rotator=Get l_rot() -> (l_rot:Rotator)
    bTransient:bool=true
   ) -> (then, ReturnValue:Actor Object Reference)
   > n7 = Cast To PCPT_DisplayUtilityActors!(execute, Object=@n6) -> (then, CastFailed, AsPCPT Display Utility Actors:PCPT Display Utility Actors Object Reference) {is=title K2Node_DynamicCast}
   > n8 = Set l_Actor(l_Actor:PCPT Display Utility Actors Object Reference=@n7) -> (Output_Get:PCPT Display Utility Actors Object Reference)
   n9 = Sequence!(execute=[@n17.then, @n8.then]) -> (then_0, then_1, then_2:Exec)
   n10 = Get l_Actor() -> (l_Actor:PCPT Display Utility Actors Object Reference)
   n11 = Get l_index() -> (l_index=0)
   > Set index!(execute, index:int=@n11, self:PCPT Display Utility Actors Object Reference=@n10) -> (then, Output_Get:int=0)
   > setColor!(
    execute
    self:PCPT Display Utility Actors Object Reference=@n10
    color:Linear Color Structure=getDrawColor!(checkIndex:int=@n11, Stat:"String (by ref)"=Get currentStatToVisualize()) -> (color:Linear Color Structure)
   ) -> (then)
   > setCurrent!(
    execute
    self:PCPT Display Utility Actors Object Reference=Get l_Actor()
    current:bool="Equal (Integer)"(A=Get l_index() -> (l_index=0), B=Get currentIndex() -> (currentIndex=0))
   ) -> (then)
   > setScale(self:PCPT Display Utility Actors Object Reference=Get l_Actor(), scale=GetValue(self=Get spinScale()))
   > setFrustumRenderOpacity(self:PCPT Display Utility Actors Object Reference=Get l_Actor(), frustumOpacity=GetValue(self=Get spinFrustumOpacity()))
   > setFrustumRenderScale(self:PCPT Display Utility Actors Object Reference=Get l_Actor(), frustumDist=GetValue(self=Get spinScale()))
   > setFrustumVisible(self:PCPT Display Utility Actors Object Reference=Get l_Actor(), visible=IsChecked(self=Get cbFrustumRender()))
   n12 = Get cbFrustumOnlyCurrent()
   > setFrustumForcedOn(self:PCPT Display Utility Actors Object Reference=Get l_Actor(), forcedOn=NOT Boolean(A=IsChecked(self=@n12)))
   Add(
    execute=@n9.then_1
    TargetArray:Array of PCPT Display Utility Actors Object References=Get currentPCPTActors()
    NewItem:"PCPT Display Utility Actors Object Reference (by ref)"=Get l_Actor()
   )
   n13 = Is Stat in Threshold!(execute:Exec=@n9.then_2, stat:"String (by ref)"=Get currentStatToVisualize(), index:int=Get l_index() -> (l_index=0)) -> (Allow:Exec, Nope:Exec) {is=title K2Node_MacroInstance}
   > Increment Int(Value=Get everyOther_Counter() -> (everyOther_Counter:int=0))
   n14 = GetValue(self=Get spinEveryOther())
   > n15 = Branch(Condition="integer >= integer"!(A:int=Get everyOther_Counter() -> (everyOther_Counter=0), B:int=Floor(A=@n14)) -> (ReturnValue:bool) {is=title K2Node_CallFunction})
   > Set everyOther_Counter(everyOther_Counter=0) -> (Output_Get=0)
   > setIsVisible(self=Get l_Actor(), isVisible=true)
   setIsVisible(execute=@n15.else, self=Get l_Actor(), isVisible=false)
   setIsVisible(execute=@n13.Nope, self=Get l_Actor(), isVisible=false)
   n16 = Set l_Actor(execute=@n5.true, l_Actor=@n5)
   > n17 = Set Actor Location And Rotation!(
    execute
    self:Actor Object Reference=@n16
    NewLocation:Vector=Get l_start()
    NewRotation:Rotator=Get l_rot()
    bSweep:bool=false
    bTeleport:bool=false
   ) -> (then, SweepHitResult:Hit Result Structure, ReturnValue:bool) {is=title K2Node_CallFunction}
  ]}
 drawDataInWorld_wInstances {nodes [
   drawDataInWorld_wInstances() {is=title K2Node_FunctionEntry}
   > n1 = Sequence!(execute) -> (then_0, then_1, then_2, then_3:Exec)
   > n2 = getDrawingInstanceActor(createIfNone=true)
   > Set l_actorManager(l_actorManager:PCPT Display Utility Actor Instanced Object Reference=@n2) -> (Output_Get:PCPT Display Utility Actor Instanced Object Reference)
   > clearInstances(self:PCPT Display Utility Actor Instanced Object Reference=Get l_actorManager() -> (l_actorManager:PCPT Display Utility Actor Instanced Object Reference))
   Set scale(execute=@n1.then_1, scale=GetValue(self=Get spinScale()), self:PCPT Display Utility Actor Instanced Object Reference=Get l_actorManager())
   > setFrustumVisible(self:PCPT Display Utility Actor Instanced Object Reference=Get l_actorManager(), visible=IsChecked(self=Get cbFrustumRender()))
   n3 = Get cbFrustumOnlyCurrent()
   > setFrustumForcedOn(self:PCPT Display Utility Actor Instanced Object Reference=Get l_actorManager(), forcedOn=NOT Boolean(A=IsChecked(self=@n3)))
   > setFrustumRenderOpacity(self:PCPT Display Utility Actor Instanced Object Reference=Get l_actorManager(), frustumOpacity=GetValue(self=Get spinFrustumOpacity()))
   > setFrustumRenderScale(self:PCPT Display Utility Actor Instanced Object Reference=Get l_actorManager(), frustumDist=GetValue(self=Get spinScale()))
   > setDrawEveryOther(self=Get l_actorManager(), everyOther=GetValue(self=Get spinEveryOther()))
   > setValueThreshold(self=Get l_actorManager(), threshold=getValidatedThreshold())
   > setValueThresholdBool(self=Get l_actorManager(), hgherThanThreshold=IsChecked(self=Get cbThresholdGreater() -> (cbThresholdGreater:Check Box Object Reference)))
   > setScale(self:PCPT Display Utility Actor Instanced Object Reference=Get l_actorManager(), scale=GetValue(self=Get spinScale()))
   n4 = getStatValues(execute=@n1.then_2, Key=Get currentStatToVisualize())
   > n5 = For Each Loop(Array:"Array of Float (double-precision)s"=@n4) -> (Array Element:double)
   Set Current Index!(
    execute=@n5.Completed
    self:PCPT Display Utility Actor Instanced Object Reference=Get l_actorManager()
    index:int=Get currentIndex() -> (currentIndex=0)
   ) -> (then) {is=title K2Node_CallFunction}
   Set l_index(execute=@n5.LoopBody, l_index=@n5.Array Index) -> (Output_Get=0)
   > Set l_isCurrent(l_isCurrent="Equal (Integer)"(A=@n5.Array Index, B=Get currentIndex() -> (currentIndex=0)))
   > n6 = getStatPositionAndForward(index=Get l_index() -> (l_index=0))
   > Set l_start(l_start=@n6)
   > Set l_rot(l_rot=Rotation From X Vector(InVec=@n6.forward))
   > addPoint!(
    execute
    self:PCPT Display Utility Actor Instanced Object Reference=Get l_actorManager()
    color:Linear Color Structure=getDrawColor(checkIndex=Get l_index() -> (l_index=0), Stat=Get currentStatToVisualize())
    WorldTransform_Location:Vector=Get l_start()
    WorldTransform_Rotation:Rotator=Get l_rot()
    WorldTransform_Scale:Vector="1,1,1"
    pointValue:double=getStatValueByIndex!(stat:"String (by ref)"=Get currentStatToVisualize(), index:int=Get l_index() -> (l_index=0)) -> (Output:double)
   ) -> (then)
   validateFrustumVisual(execute=@n1.then_3, self:PCPT Display Utility Actor Instanced Object Reference=Get l_actorManager())
  ]}
 filterStatByThreshold {nodes [
   n1 = filterStatByThreshold!() -> (then, stat:"String (by ref)", index:int) {is=title K2Node_FunctionEntry}
   n2 = Get spinValueThreshold()
   > n3 = Branch(Condition="float > float"(A=GetValue(self=@n2), B=0) {is=title K2Node_CallFunction})
   n4 = Get cbThresholdGreater()
   n5 = GetValue(self=Get spinValueThreshold())
   n6 = Get stats()
   n7 = "Get (a copy)"(
    Array:"Array of Float (single-precision)s"=Break PCPT Stat Struct!(PCPT_Stat_Struct:"PCPT Stat Struct Structure (by ref)"=Find(TargetMap:Map of Strings to PCPT Stat Struct Structures=@n6, Key=@n1) -> (Value:PCPT Stat Struct Structure)) -> (statName_5_#1:String, stats_8_#2:"Array of Float (single-precision)s") {is=title K2Node_BreakStruct}.stats_8_#2
    Dimension 1=@n1.index
   ) -> (Output:float)
   > Return Node!(
    execute
    ReturnValue:bool=Select!(
     Option 0:bool="float <= float"!(A:double=@n7, B:double=@n5) -> (ReturnValue:bool) {is=title K2Node_CallFunction}
     Option 1:bool="float >= float"(A=@n7, B=@n5) {is=title K2Node_CallFunction}
     Index:bool=IsChecked(self=@n4)
    ) -> (ReturnValue:bool)
   )
   Return Node(execute=@n3.else, ReturnValue=true)
  ]}
 filterStatForExclusion {nodes [
   n1 = Inputs!() -> (" ":Exec, stat:String)
   > n2 = Branch(Condition=Should Stat be Included!(self:PCPT Show Stats List Widget Object Reference=Get PCPT_ShowStatsList_Widget(), ItemToFind:"String (by ref)"=@n1) -> (shouldInclude:bool))
   > Outputs!(Include:Exec, Exclude:Exec=@n2.else)
  ]}
 finalizeCsvDataAndSave {nodes [finalizeCsvDataAndSave() {is=title K2Node_FunctionEntry}]}
 getAllDrawingActors {nodes [
   getAllDrawingActors() {is=title K2Node_FunctionEntry}
   n1 = Get UnrealEditorSubsystem()
   > n2 = GetGameWorld!(execute, self:Unreal Editor Subsystem Object Reference=@n1) -> (then, ReturnValue:World Object Reference)
   > n3 = Set l_world(l_world:Object Reference=@n2) -> (Output_Get:Object Reference)
   > n4 = Is Valid(InputObject=@n3)
   n5 = GetEditorWorld(execute=@n4.Is Not Valid, self=@n1)
   > Is Valid(InputObject=@n5)
   > n6 = Set l_world(l_world=@n5)
   n7 = GetAllActorsOfClass!(
    execute=[@n4.Is Valid, @n6.then]
    WorldContextObject:Object Reference=Get l_world()
    ActorClass:Actor Class Reference=/Game/NeverCook/EditorUtilities/PerfTools/PerfCsvPlaybackTool/Actors/PCPT_DisplayUtilityActors.PCPT_DisplayUtilityActors_C
   ) -> (then, OutActors:Array of PCPT Display Utility Actors Object References)
   > Return Node!(execute, OutActors:Array of PCPT Display Utility Actors Object References=@n7)
  ]}
 getAllowWrap {nodes [
   getAllowWrap() {is=title K2Node_FunctionEntry}
   > Return Node!(execute, allowWrap:bool=IsChecked(self=Get cbRepeat()))
  ]}
 getDataLength {nodes [
   getDataLength() {is=title K2Node_FunctionEntry}
   > n1 = getStatValues(Key=Get currentStatToVisualize())
   > Return Node!(execute, ReturnValue:int=Length(TargetArray:"Array of Float (double-precision)s"=@n1))
  ]}
 getDrawColor {nodes [
   n1 = getDrawColor!() -> (then, checkIndex:int, Stat:"String (by ref)") {is=title K2Node_FunctionEntry}
   n2 = Find(
    TargetMap:Map of Strings to PCPT Base Stat Visualizer Object References=Get statVisualizers() -> (statVisualizers:Map of Strings to PCPT Base Stat Visualizer Object References)
    Key=@n1.Stat
   ) -> (Value:PCPT Base Stat Visualizer Object Reference)
   > n3 = Branch(Condition=@n2.ReturnValue)
   > n4 = getStatValues(Key=@n1.Stat)
   n5 = Get checkIndex() -> (checkIndex:int=0)
   > Branch(Condition=Is Valid Index(TargetArray:"Array of Float (double-precision)s"=@n4, IndexToTest=@n5))
   > n6 = getColorFromStatValue!(
    execute
    self:PCPT Base Stat Visualizer Object Reference=@n2
    value:double="Get (a copy)"(Array:"Array of Float (double-precision)s"=@n4, Dimension 1=@n5) -> (Output:double)
   ) -> (then, color:Linear Color Structure)
   > Return Node!(execute, color:Linear Color Structure=@n6)
   Return Node(execute=@n3.else, color=(R=0.8177,G=0.5461,B=0.0541,A=1))
  ]}
 getDrawingInstanceActor {nodes [
   n1 = getDrawingInstanceActor!() -> (then, createIfNone:bool) {is=title K2Node_FunctionEntry}
   > n2 = GetAllActorsOfClass(ActorClass=/Game/NeverCook/EditorUtilities/PerfTools/PerfCsvPlaybackTool/Actors/PCPT_DisplayUtilityActor_Instanced.PCPT_DisplayUtilityActor_Instanced_C) -> (OutActors:Array of PCPT Display Utility Actor Instanced Object References)
   > n3 = Branch(Condition=Is Valid Index(TargetArray:Array of PCPT Display Utility Actor Instanced Object References=@n2))
   > Return Node!(
    execute
    Output:PCPT Display Utility Actor Instanced Object Reference="Get (a copy)"(Array:Array of PCPT Display Utility Actor Instanced Object References=@n2) -> (Output:PCPT Display Utility Actor Instanced Object Reference)
    wasNew:bool=false
   )
   n4 = Branch(execute=@n3.else, Condition=@n1)
   > n5 = SpawnActorFromObject(
    ObjectToUse=/Game/NeverCook/EditorUtilities/PerfTools/PerfCsvPlaybackTool/Actors/PCPT_DisplayUtilityActor_Instanced
    Location="0, 0, 0"
    bTransient=true
   )
   > n6 = Cast To PCPT_DisplayUtilityActor_Instanced!(execute, Object=@n5) -> (then, CastFailed, AsPCPT Display Utility Actor Instanced:PCPT Display Utility Actor Instanced Object Reference) {is=title K2Node_DynamicCast}
   > Return Node(Output=@n6, wasNew=true)
   Return Node(execute=@n4.else, wasNew=false)
  ]}
 getGraphDrawColorsFromStat {nodes [
   n1 = getGraphDrawColorsFromStat!() -> (then, stat:"String (by ref)", values:"Array of Float (double-precision)s") {is=title K2Node_FunctionEntry}
   > n2 = Set l_visualizer(l_visualizer:PCPT Base Stat Visualizer Object Reference=Find(TargetMap=Get statVisualizers(), Key=@n1)) -> (Output_Get:PCPT Base Stat Visualizer Object Reference)
   > Is Valid(InputObject=@n2)
   > n3 = For Each Loop(Array=@n1.values)
   Return Node!(execute=@n3.Completed, color:Array of Linear Color Structures=Get l_colors() -> (l_colors:Array of Linear Color Structures))
   > n4 = getColorFromStatValue(self=Get l_visualizer() -> (l_visualizer:PCPT Base Stat Visualizer Object Reference), value=@n3)
   > Add(TargetArray:Array of Linear Color Structures=Get l_colors(), NewItem:"Linear Color Structure (by ref)"=@n4)
  ]}
 getIndexFromTimelineValue {nodes [
   getIndexFromTimelineValue() {is=title K2Node_FunctionEntry}
   n1 = "To Float (Integer)"(InInt=int - int(A=getDataLength() -> (ReturnValue=0) {is=title K2Node_CallFunction}, B=1))
   n2 = getValue!(self:PCPT Timeline Widget Object Reference=Get timeline()) -> (value:double)
   > Return Node!(execute, ReturnValue:int=Floor(A=float * float(A=@n2, B=@n1)))
  ]}
 getNextExistingDrawActor {nodes [
   getNextExistingDrawActor() {is=title K2Node_FunctionEntry}
   > n1 = Branch(Condition=Is Valid Index(TargetArray:Array of PCPT Display Utility Actors Object References=Get cachedPCPTActors()))
   > Set l_Actor(l_Actor="Get (a copy)"(Array:Array of PCPT Display Utility Actors Object References=Get cachedPCPTActors()) -> (Output:PCPT Display Utility Actors Object Reference))
   > Remove Index!(execute, TargetArray:Array of PCPT Display Utility Actors Object References=Get cachedPCPTActors(), IndexToRemove:int=0) -> (then)
   > Return Node!(execute, actor:PCPT Display Utility Actors Object Reference=Get l_Actor())
   Return Node(execute=@n1.else)
  ]}
 getPlaybackSpeed {nodes [
   getPlaybackSpeed() {is=title K2Node_FunctionEntry}
   n1 = Make Literal Float(Value=0.2)
   n2 = Get spinSpeed()
   n3 = GetMinValue!(self:Spin Box Object Reference=@n2) -> (ReturnValue:float)
   n4 = GetMaxValue!(self:Spin Box Object Reference=@n2) -> (ReturnValue:float)
   n5 = float - float(A=@n4, B=@n3) {is=title K2Node_CallFunction}
   n6 = GetValue(self=Get spinSpeed())
   > Return Node!(
    execute
    value:double=Lerp(
     A=@n1
     B=float / float(A=@n1, B=@n5) {is=title K2Node_CallFunction}
     Alpha=NormalizeToRange!(Value:double=@n6, RangeMin:double=@n3, RangeMax:double=@n4) -> (ReturnValue:double)
    )
   )
  ]}
 getShouldMoveCamera {nodes [
   getShouldMoveCamera() {is=title K2Node_FunctionEntry}
   > Return Node!(execute, index:int=GetSelectedIndex(self=Get ddMoveType()) -> (ReturnValue=0))
  ]}
 getStatPositionAndForward {nodes [
   getStatPositionAndForward!() -> (then, index:int) {is=title K2Node_FunctionEntry}
   > Branch(Condition=Is Valid Index(TargetArray:Array of Vectors=Get cachedViewLocations(), IndexToTest=Get index() -> (index:int=0)))
   n1 = Get index() -> (index=0)
   > Return Node!(
    execute
    position:Vector="Get (a copy)"(Array:Array of Vectors=Get cachedViewLocations(), Dimension 1=@n1) -> (Output:Vector)
    forward:Vector="Get (a copy)"(Array=Get cachedViewRotations(), Dimension 1=@n1)
   )
   Get index() -> (index=0)
  ]}
 getStatValueByIndex {nodes [
   n1 = getStatValueByIndex!() -> (then, stat:"String (by ref)", index:int) {is=title K2Node_FunctionEntry}
   n2 = Find(TargetMap:Map of Strings to PCPT Stat Struct Structures=Get stats(), Key=@n1) -> (Value:PCPT Stat Struct Structure)
   > Return Node!(
    execute
    Output:double="Get (a copy)"(
     Array:"Array of Float (single-precision)s"=Break PCPT Stat Struct(PCPT_Stat_Struct=@n2).stats_8_#2
     Dimension 1=@n1.index
    ) -> (Output:float)
   )
  ]}
 getStatValues {nodes [
   n1 = getStatValues!() -> (then, Key:"String (by ref)") {is=title K2Node_FunctionEntry}
   n2 = Find(TargetMap=Get stats(), Key=@n1)
   > n3 = Branch(Condition=@n2.ReturnValue)
   > Return Node!(
    execute
    OutputPin:"Array of Float (double-precision)s"=Break PCPT Stat Struct(PCPT_Stat_Struct=@n2).stats_8_#2
   )
   n4 = Find(TargetMap:Map of Strings to Strings=Get currentMap(), Key=@n1) -> (Value:String)
   Branch(execute=@n3.else, Condition=@n4.ReturnValue)
   > Set l_stats(l_stats:String=@n4) -> (Output_Get:String)
   > n5 = For Each Loop(Array:Array of Strings=ParseIntoArray(SourceString=Get l_stats() -> (l_stats:String), Delimiter=",")) -> (Array Element:String)
   Add!(
    execute=@n5.Completed
    TargetMap:Map of Strings to PCPT Stat Struct Structures=Get stats()
    Key:"String (by ref)"=@n1
    Value:"PCPT Stat Struct Structure (by ref)"=Make PCPT Stat Struct!(
     statName_5_#1:String=@n1
     stats_8_#2:"Array of Float (single-precision)s"=Get l_output() -> (l_output:"Array of Float (double-precision)s")
    ) -> (PCPT_Stat_Struct:PCPT Stat Struct Structure) {is=title K2Node_MakeStruct}
   ) -> (then) {is=title K2Node_CallFunction}
   > Return Node(OutputPin=Get l_output())
   Add(
    execute=@n5.LoopBody
    TargetArray:"Array of Float (double-precision)s"=Get l_output()
    NewItem:"Float (double-precision) (by ref)"=String To Float(InString=@n5)
   ) {is=title K2Node_CallArrayFunction}
  ]}
 getTimelineValueFromCurrentIndex {nodes [
   getTimelineValueFromCurrentIndex() {is=title K2Node_FunctionEntry}
   n1 = getDataLength() -> (ReturnValue=0)
   > n2 = Branch(Condition="integer < integer"!(A:int=@n1, B:int=1) -> (ReturnValue:bool) {is=title K2Node_CallFunction})
   > Return Node!(execute, ReturnValue:double=0)
   n3 = Get currentIndex() -> (currentIndex=0)
   Return Node(execute=@n2.else, ReturnValue=float / float(A="To Float (Integer)"(InInt=@n3), B="To Float (Integer)"(InInt=@n1)))
  ]}
 getValidatedThreshold {nodes [
   getValidatedThreshold() {is=title K2Node_FunctionEntry}
   n1 = GetValue(self=Get spinValueThreshold())
   n2 = Get data_Lowest!(self:PCPT Timeline Widget Object Reference=Get timeline()) -> (data_Lowest:double)
   > Return Node(ReturnValue=SelectFloat(A=0, B=@n1, bPickA="float <= float"(A=@n1, B=@n2)))
  ]}
 handleCsvDataUpdated {nodes [
   handleCsvDataUpdated() {is=title K2Node_FunctionEntry}
   > n1 = Sequence!(execute) -> (then_0, then_1, then_2, then_3, then_4:Exec)
   > n2 = Keys!(execute, TargetMap:Map of Strings to Strings=Get currentMap()) -> (then, Keys:Array of Strings)
   > n3 = For Each Loop(Array=@n2)
   n4 = getListItems!(
    execute=@n3.Completed
    self:TA Generic 01 List Object Reference=Get listStatCounters() -> (listStatCounters:TA Generic 01 List Object Reference)
   ) -> (then, items:Array of User Widget Object References)
   > n5 = For Each Loop with Break!(Exec:Exec, Array:Array of User Widget Object References=@n4, Break:Exec=@n6.then) -> (LoopBody:Exec, Array Element:User Widget Object Reference, Array Index:int, Completed:Exec)
   > Branch(Condition=IsVisible(self=@n5))
   > n6 = setIndexSelected!(
    execute
    self:TA Generic 01 List Object Reference=Get listStatCounters()
    index:int=@n5.Array Index
    selected:bool=true
    throwEvent:bool=true
   ) -> (then)
   n7 = Construct Object from Class(
    execute=@n3.LoopBody
    Class=/Game/Blueprints/Blutilities/ForTechArt/Data/TA_GenericListItemData.TA_GenericListItemData_C
    self=Self-Reference()
   ) -> (ReturnValue:TA Generic List Item Data Object Reference)
   n8 = Get currentMap()
   > Set strings!(
    execute
    strings:Array of Strings=Make Array("[0]":String=Find(TargetMap=@n8, Key=@n3)) -> (Array:Array of Strings)
    self:TA Generic List Item Data Object Reference=@n7
   ) -> (then, Output_Get:Array of Strings)
   > Set displayString!(execute, displayString:String=@n3, self:TA Generic List Item Data Object Reference=@n7) -> (then, Output_Get:String)
   > n9 = addItem!(execute, self:TA Generic 01 List Object Reference=Get listStatCounters(), dataObject:Object Reference=@n7) -> (then, newItem:User Widget Object Reference, newData:Object Reference)
   > Add(
    TargetMap:Map of Strings to User Widget Object References=Get statToWidgetMap() -> (statToWidgetMap:Map of Strings to User Widget Object References)
    Key=@n3
    Value:"User Widget Object Reference (by ref)"=@n9
   ) {is=title K2Node_CallFunction}
   > n10 = Filter Stat for Exclusion!(" ":Exec, stat:String=@n3) -> (Include:Exec, Exclude:Exec) {is=title K2Node_MacroInstance}
   SetVisibility(execute=@n10.Exclude, self=@n9, InVisibility=Collapsed)
   n11 = Cast To PCPT_StatReport_Widget!(execute=@n10.Include, Object=@n9) -> (then, CastFailed, AsPCPT Stat Report Widget:PCPT Stat Report Widget Object Reference) {is=title K2Node_DynamicCast}
   > n12 = getStatValues(Key=@n3) {is=title K2Node_CallFunction}
   > setDataAndVisibility!(
    execute
    self:PCPT Stat Report Widget Object Reference=@n11
    vis:bool=true
    data:"Array of Float (double-precision)s"=@n12
    index:int=Get currentIndex() -> (currentIndex=0)
   ) -> (then)
   cacheViewValues(execute=@n1.then_1)
   > initAvailableStatsList()
   > updateStatsDisplay()
   loadDataOffsets(execute=@n1.then_2)
   setCurrentIndex!(execute=@n1.then_3, self, currentIndex:int=0, updateTimeline:bool=true) -> (then)
   Set shouldSave(execute=@n1.then_4, shouldSave=true)
  ]}
 handleCurrentIndexChangedForActors {nodes [
   n1 = handleCurrentIndexChangedForActors!() -> (then, last:int) {is=title K2Node_FunctionEntry}
   > Set l_last(l_last:int=@n1) -> (Output_Get:int=0)
   > n2 = Sequence!(execute) -> (then_0, then_1, then_2)
   > Branch(Condition=Is Valid Index(TargetArray:Array of PCPT Display Utility Actors Object References=Get currentPCPTActors(), IndexToTest=Get l_last() -> (l_last:int=0)))
   > setCurrent(
    self="Get (a copy)"(Array:Array of PCPT Display Utility Actors Object References=Get currentPCPTActors(), Dimension 1=Get l_last() -> (l_last=0)) -> (Output:PCPT Display Utility Actors Object Reference)
    current=false
   )
   Branch(execute=@n2.then_1, Condition=Is Valid Index(TargetArray=Get currentPCPTActors(), IndexToTest=Get currentIndex() -> (currentIndex=0)))
   n3 = "Get (a copy)"(Array=Get currentPCPTActors(), Dimension 1=Get currentIndex() -> (currentIndex=0))
   > setCurrent(self=@n3, current=true)
   > SetSelectedLevelActors(self=Get EditorActorSubsystem(), ActorsToSelect=Make Array("[0]":Actor Object Reference=@n3) -> (Array:Array of Actor Object References))
   n4 = getDrawingInstanceActor(execute=@n2.then_2, createIfNone=false) {is=title K2Node_CallFunction}
   > Is Valid(InputObject=@n4)
   > Set Current Index(self=@n4, index=Get currentIndex() -> (currentIndex=0))
  ]}
 handleStatInclusionChanged {nodes [
   n1 = handleStatInclusionChanged!() -> (then, widget:PCPT Stat Available List Item Object Reference, stat:String, included:bool=~) {is=title K2Node_FunctionEntry}
   > Set shouldSave(shouldSave=true)
   > n2 = Cast To PCPT_StatReport_Widget(Object=Find(TargetMap:Map of Strings to User Widget Object References=Get statToWidgetMap(), Key=@n1.stat) -> (Value:User Widget Object Reference))
   > n3 = getStatValues(Key=@n1.stat)
   > setDataAndVisibility(self=@n2, vis=@n1.included, data=@n3, index=Get currentIndex() -> (currentIndex=0))
  ]}
 handleTimelineChanged {nodes [
   handleTimelineChanged!() -> (then, Value:double=~) {is=title K2Node_FunctionEntry}
   > setCurrentIndex(currentIndex=getIndexFromTimelineValue!() -> (ReturnValue:int=0) {is=title K2Node_CallFunction}, updateTimeline=false)
   > EditorInvalidateViewports(self:Level Editor Subsystem Object Reference=Get LevelEditorSubsystem())
   > setPlaying(isPlaying=false)
  ]}
 initAvailableStatsList {nodes [
   initAvailableStatsList() {is=title K2Node_FunctionEntry}
   > n1 = Keys(TargetMap=Get currentMap())
   > setAvailableStats!(execute, self:PCPT Show Stats List Widget Object Reference=Get PCPT_ShowStatsList_Widget(), statList:Array of Strings=@n1) -> (then)
  ]}
 initCSV {nodes [
   n1 = "Execute Python Command (Advanced)"!(
    execute
    PythonCommand:String=Append!(
     A="import TechArt.Perf.perf_parseCSV as perf\r\nimport imp\r\nimp.reload(perf)\r\nprint(\"Test\")\r\nperf.loadEvents(\""
     B=Get currentFilePath()
     C:String="\")"
    ) -> (ReturnValue)
    ExecutionMode:EPythonCommandExecutionMode Enum=ExecuteFile
    FileExecutionScope:EPythonFileExecutionScope Enum=Private
   ) -> (then, CommandResult:String, LogOutput:Array of Python Log Output Entry Structures, ReturnValue:bool)
   > n2 = Branch(Condition=@n1.ReturnValue)
   > Set l_logOutput(l_logOutput:Array of Python Log Output Entry Structures=@n1.LogOutput) -> (Output_Get:Array of Python Log Output Entry Structures)
   > n3 = For Each Loop(Array:Array of Python Log Output Entry Structures=Get l_logOutput() -> (l_logOutput:Array of Python Log Output Entry Structures)) -> (Array Element:Python Log Output Entry Structure)
   n4 = Break Python Log Output Entry!(PythonLogOutputEntry:"Python Log Output Entry Structure (by ref)"=@n3) -> (Type:EPythonLogOutputType Enum, Output:String) {is=title K2Node_BreakStruct}
   > Branch(Condition="Equal (Enum)"(A:EPythonLogOutputType Enum=@n4, B:EPythonLogOutputType Enum=Info))
   > parse_Events!(execute, self, SourceString:String=@n4.Output) -> (then)
   Return Node!(execute=@n2.else)
   initCSV() {is=title K2Node_FunctionEntry}
   > n5 = Sequence!(execute) -> (then_0, then_1, then_2, then_3:Exec, then_4:Exec)
   > clearItems(self:TA Generic 01 List Object Reference=Get listStatCounters())
   > Clear!(execute, TargetMap:Map of Strings to PCPT Stat Struct Structures=Get stats()) -> (then) {is=title K2Node_CallFunction}
   > Set currentIndex(currentIndex:int=0) -> (Output_Get:int=0)
   > Clear(TargetMap:Map of Strings to User Widget Object References=Get statToWidgetMap())
   > Clear(TargetArray=Get cachedPCPTActors()) {is=title K2Node_CallArrayFunction}
   > Clear(TargetArray=Get currentPCPTActors())
   > SetValue(self=Get spinValueThreshold(), NewValue=0)
   n6 = Execute Python Script!(
    execute=@n5.then_1
    PythonScript="import TechArt.Perf.perf_parseCSV as perf\r\nimport imp\r\nimp.reload(perf)\r\nreturnVals = perf.loadCSV(data_count)\r\n\r\nfinal_map = returnVals[0]\r\nfile_path = returnVals[1]\r\nkey_names = returnVals[2]"
    data_count:float=GetValue(self=Get spinMaxDataPoints())
   ) -> (then, ReturnValue, final_map:Map of Strings to Strings, file_path:String) {inputs [data_count], outputs [final_map,file_path]}
   > Set currentMap(currentMap=@n6.final_map)
   > Set currentFilePath(currentFilePath=@n6.file_path)
   handleCsvDataUpdated(execute=@n5.then_4) {is=title K2Node_CallFunction}
   Break Python Log Output Entry(PythonLogOutputEntry=@n3)
  ]}
 initMoveTypeCombobox {nodes [
   initMoveTypeCombobox() {is=title K2Node_FunctionEntry}
   > n1 = ForEach PCPT_CamMoveType_Enum!(execute:Exec, SkipHidden:bool) -> (LoopBody:Exec, EnumValue:PCPT_CamMoveType_Enum Enum, then:Exec) {is=title K2Node_ForEachElementInEnum}
   SetSelectedIndex(execute=@n1.then, self=Get ddMoveType(), Index=0)
   AddOption!(
    execute=@n1.LoopBody
    self:"Combo Box (String) Object Reference"=Get ddMoveType()
    Option:String=Enum to String!(Enumerator:PCPT_CamMoveType_Enum Enum=@n1) -> (ReturnValue:String)
   ) -> (then)
  ]}
 initStatVisualizers {nodes [
   initStatVisualizers() {is=title K2Node_FunctionEntry}
   > n1 = To Array!(
    execute
    A:Set of PCPT Base Stat Visualizer Class References=Get visualizersToLoad() -> (visualizersToLoad:Set of PCPT Base Stat Visualizer Class References)
   ) -> (then, Result:Array of PCPT Base Stat Visualizer Class References)
   > n2 = For Each Loop(Array:Array of PCPT Base Stat Visualizer Class References=@n1) -> (Array Element:PCPT Base Stat Visualizer Class Reference)
   > n3 = Construct Object from Class(Class=@n2, self=Self-Reference()) -> (ReturnValue:PCPT Base Stat Visualizer Object Reference)
   > Add(
    TargetMap:Map of Strings to PCPT Base Stat Visualizer Object References=Get statVisualizers()
    Key=Get statName!(self:PCPT Base Stat Visualizer Object Reference=@n3) -> (statName:String)
    Value:"PCPT Base Stat Visualizer Object Reference (by ref)"=@n3
   )
  ]}
 isDrawingActive {nodes [
   Inputs!() -> (execute:Exec)
   > n1 = Branch(Condition=Is Valid Index(TargetArray=Get currentPCPTActors()))
   n2 = Branch(execute=[@n1.then, @n4.Is Valid], Condition=Get useActors())
   > Outputs!(Yep_Actors:Exec, Yep_Instances:Exec=@n2.else, Nope:Exec=@n4.Is Not Valid)
   n3 = getDrawingInstanceActor(execute=@n1.else, createIfNone=false)
   > n4 = Is Valid(InputObject=@n3)
  ]}
 isStatInThreshold {nodes [
   n1 = Inputs!() -> (execute, stat:"String (by ref)", index:int)
   > n2 = Branch(Condition=filterStatByThreshold!(stat:"String (by ref)"=@n1, index:int=@n1.index) -> (ReturnValue:bool) {is=title K2Node_CallFunction})
   > Outputs!(Allow:Exec, Nope=@n2.else)
  ]}
 loadDataOffsets {nodes [
   loadDataOffsets() {is=title K2Node_FunctionEntry}
   > n1 = Cast To PSPT_SaveGame(Object=Get saveGame_Object())
   > n2 = For Each Loop with Break!(
    Exec
    Array:Array of PCPT Saved Offsets Struct Structures=Get savedOffsets!(self:PSPT Save Game Object Reference=@n1) -> (savedOffsets:Array of PCPT Saved Offsets Struct Structures)
    Break
   ) -> (
    LoopBody
    Array Element_csvName_5_#3:String
    Array Element_offset_4_#4:Vector=~
    Array Element_rotationOffset_8_#5:Rotator=~
    Array Index
    Completed
   )
   setLocationOffset!(execute=@n2.Completed, self:PCPT Offset Transform EWidget Object Reference=Get PCPT_OffsetTransform_EWidget(), location:Vector="0, 0, 0") -> (then)
   > setRotationOffset!(execute, self:PCPT Offset Transform EWidget Object Reference=Get PCPT_OffsetTransform_EWidget(), location:Rotator="0, 0, 0") -> (then)
   Branch(execute=@n2.LoopBody, Condition="Equal Exactly (String)"(A=@n2, B=Get currentFilePath()))
   > setLocationOffset(self=Get PCPT_OffsetTransform_EWidget(), location=@n2.Array Element_offset_4_#4)
   > setRotationOffset(self=Get PCPT_OffsetTransform_EWidget(), location=@n2.Array Element_rotationOffset_8_#5)
   > Return Node()
  ]}
 parse_Events {nodes [
   n1 = parse_Events!() -> (then, SourceString:String) {is=title K2Node_FunctionEntry}
   > Clear(TargetArray:Array of Strings=Get currentEvents() -> (currentEvents:Array of Strings))
   n2 = Split(SourceString=@n1, InStr=:::, SearchDir=FromStart)
   > Branch(Condition=@n2.ReturnValue)
   > Set l_event(l_event:String=@n2.RightS) -> (Output_Get:String)
   > Add(TargetArray:Array of Strings=Get currentEvents(), NewItem:"String (by ref)"=Get l_event() -> (l_event:String)) {is=title K2Node_CallArrayFunction}
  ]}
 playToNextDataIndex {nodes [
   playToNextDataIndex() {is=title K2Node_FunctionEntry}
   > n1 = Branch(Condition="integer > integer"(A=getDataLength() -> (ReturnValue=0), B=0) {is=title K2Node_CallFunction})
   > Branch(Condition=Get isPlaying())
   > Set l_new(l_new:int=int + int(A=Get currentIndex() -> (currentIndex=0), B=1) -> (ReturnValue=0) {is=title K2Node_CommutativeAssociativeBinaryOperator}) -> (Output_Get:int=0)
   > n2 = Branch(Condition="integer < integer"(A=Get l_new() -> (l_new:int=0), B=getDataLength() -> (ReturnValue=0)))
   > setCurrentIndex(currentIndex=Get l_new() -> (l_new=0), updateTimeline=true)
   n3 = Branch(execute=@n2.else, Condition=getAllowWrap() {is=title K2Node_CallFunction})
   > setCurrentIndex(currentIndex=0, updateTimeline=true)
   setPlaying(execute=@n3.else, isPlaying=false)
   setPlaying(execute=@n1.else, isPlaying=false)
  ]}
 saveDataOffsets {nodes [
   saveDataOffsets() {is=title K2Node_FunctionEntry}
   > Branch(Condition="Not Equal Exactly (String)"(A=Get currentFilePath()))
   > n1 = Cast To PSPT_SaveGame(Object=Get saveGame_Object())
   > Set l_saveData(l_saveData:PSPT Save Game Object Reference=@n1) -> (Output_Get:PSPT Save Game Object Reference)
   > n2 = For Each Loop with Break(Array=Get savedOffsets(self=@n1), Break=@n9.then) -> (Array Element_offset_4_#4=~, Array Element_rotationOffset_8_#5=~)
   Set l_data(
    execute=@n2.Completed
    l_data:Array of PCPT Saved Offsets Struct Structures=Get savedOffsets(self=Get l_saveData() -> (l_saveData:PSPT Save Game Object Reference))
   ) -> (Output_Get:Array of PCPT Saved Offsets Struct Structures)
   > n3 = Branch(Condition="integer > integer"(A=Get l_index() -> (l_index=0), B=-1))
   n4 = Get PCPT_OffsetTransform_EWidget()
   n5 = Make PCPT Saved Offsets Struct!(
    csvName_5_#3:String=Get currentFilePath()
    offset_4_#4:Vector=getLocationOffset(self=@n4)
    rotationOffset_8_#5:Rotator=getRotationOffset(self=@n4)
   ) -> (PCPT_SavedOffsets_Struct:PCPT Saved Offsets Struct Structure) {is=title K2Node_MakeStruct}
   n6 = Get l_data() -> (l_data:Array of PCPT Saved Offsets Struct Structures)
   > n7 = Set Array Elem!(
    execute
    TargetArray:Array of PCPT Saved Offsets Struct Structures=@n6
    Index:int=Get l_index() -> (l_index=0)
    Item:"PCPT Saved Offsets Struct Structure (by ref)"=@n5
    bSizeToFit:bool=false
   ) -> (then)
   Set savedOffsets!(
    execute=[@n7.then, @n8.then]
    savedOffsets:Array of PCPT Saved Offsets Struct Structures=Get l_data()
    self:PSPT Save Game Object Reference=Get l_saveData()
   ) -> (then, Output_Get:Array of PCPT Saved Offsets Struct Structures)
   n8 = Add(execute=@n3.else, TargetArray:Array of PCPT Saved Offsets Struct Structures=@n6, NewItem:"PCPT Saved Offsets Struct Structure (by ref)"=@n5)
   Branch(execute=@n2.LoopBody, Condition="Equal Exactly (String)"(A=@n2, B=Get currentFilePath()))
   > n9 = Set l_index(l_index=@n2.Array Index) -> (Output_Get=0)
  ]}
 saveOnTick {nodes [
   saveOnTick() {is=title K2Node_FunctionEntry}
   > Branch(Condition=Get shouldSave() -> (shouldSave:bool))
   > Set shouldSave(shouldSave=false)
   > SaveGame_Save(self:TA Base Editor Widget Object Reference)
  ]}
 setCurrentIndex {nodes [
   n1 = setCurrentIndex!() -> (then, currentIndex:int, updateTimeline:bool=true) {is=title K2Node_FunctionEntry}
   > Set l_last(l_last=Get currentIndex() -> (currentIndex=0)) -> (Output_Get=0)
   > Set currentIndex(currentIndex=@n1) -> (Output_Get=0)
   > n2 = Sequence!(execute) -> (then_0, then_1, then_2, then_3)
   > Branch(Condition=@n1.updateTimeline)
   > Unbind all Events from On Value Changed!(execute:Exec, self:PCPT Timeline Widget Object Reference=Get timeline()) -> (then:Exec) {is=title K2Node_ClearDelegate}
   > setValue!(
    execute
    self:PCPT Timeline Widget Object Reference=Get timeline()
    value:double=getTimelineValueFromCurrentIndex!() -> (ReturnValue:double) {is=title K2Node_CallFunction}
    throwEvent:bool=false
   ) -> (then)
   > signUpForTimelineEvents()
   updateStatsDisplay(execute=@n2.then_1)
   > updatePawnPosition(Force=false)
   handleCurrentIndexChangedForActors!(execute=@n2.then_2, self, last:int=Get l_last() -> (l_last=0)) -> (then) {is=title K2Node_CallFunction}
   n3 = getStatPositionAndForward(execute=@n2.then_3, index=Get currentIndex() -> (currentIndex=0)) {is=title K2Node_CallFunction}
   > setLocation(
    self=Get bugit_CSV() -> (bugit_CSV:PCPT Bugit Location Display EWidget Object Reference)
    location=@n3
    rotation=MakeRotFromX!(X:"Vector (by ref)"=@n3.forward) -> (ReturnValue:Rotator)
   )
  ]}
 setPlaying {nodes [
   n1 = setPlaying!() -> (then, isPlaying:bool=~) {is=title K2Node_FunctionEntry}
   > Set isPlaying(isPlaying:bool=@n1) -> (Output_Get:bool)
   > setLabel!(
    execute
    self:TA Base Editor Button Widget Object Reference=Get btnPlay() -> (btnPlay:TA Generic 01 Button Object Reference)
    label:String=SelectString(A=Stop, B=Play, bPickA=Get isPlaying())
   ) -> (then)
  ]}
 setSelectedStat {nodes [
   n1 = setSelectedStat!() -> (then, stat:String) {is=title K2Node_FunctionEntry}
   > Set currentStatToVisualize(currentStatToVisualize:String=@n1) -> (Output_Get:String)
   > n2 = Sequence()
   > n3 = getStatValues(Key=Get currentStatToVisualize())
   > Set l_graphData(l_graphData:"Array of Float (double-precision)s"=@n3) -> (Output_Get:"Array of Float (double-precision)s")
   n4 = getGraphDrawColorsFromStat!(
    execute=@n2.then_1
    self
    stat:"String (by ref)"=Get currentStatToVisualize()
    values:"Array of Float (double-precision)s"=Get l_graphData() -> (l_graphData:"Array of Float (double-precision)s")
   ) -> (then, color:Array of Linear Color Structures) {is=title K2Node_CallFunction}
   > setGraphData!(
    execute
    self:PCPT Timeline Widget Object Reference=Get timeline()
    data:"Array of Float (double-precision)s"=Get l_graphData()
    dataColors:Array of Linear Color Structures=@n4
   ) -> (then)
   n5 = Get timeline()
   > "SetText (Text)"(
    self=Get txtMin() -> (txtMin:Common Text Object Reference)
    InText="To Text (Float)"(Value=Get data_Lowest(self=@n5), MinimumIntegralDigits=1, MaximumIntegralDigits=324, MinimumFractionalDigits=0, MaximumFractionalDigits=2)
   )
   n6 = Get timeline()
   > "SetText (Text)"(
    self=Get txtMax() -> (txtMax:Common Text Object Reference)
    InText="To Text (Float)"(
     Value=Get data_Highest!(self:PCPT Timeline Widget Object Reference=@n6) -> (data_Highest:double)
     MinimumIntegralDigits=1
     MaximumIntegralDigits=324
     MinimumFractionalDigits=0
     MaximumFractionalDigits=2
    )
   )
   n7 = Get data_Highest(self=Get timeline())
   n8 = Get spinValueThreshold()
   > SetMaxValue!(execute, self:Spin Box Object Reference=@n8, NewValue:float=@n7) -> (then)
   > SetMaxSliderValue!(execute, self:Spin Box Object Reference=@n8, NewValue:float=@n7) -> (then)
   n9 = Get data_Lowest(self=Get timeline())
   n10 = Get spinValueThreshold()
   > SetMinValue!(execute, self:Spin Box Object Reference=@n10, NewValue:float=@n9) -> (then)
   > SetMinSliderValue!(execute, self:Spin Box Object Reference=@n10, NewValue:float=@n9) -> (then)
   > SetValue(self=Get spinValueThreshold(), NewValue=@n9)
   n11 = getListItems(execute=@n2.then_2, self=Get listStatCounters())
   > n12 = For Each Loop(Array:Array of User Widget Object References=@n11) -> (Array Element:User Widget Object Reference)
   > n13 = Cast To PCPT_StatReport_Widget(Object=@n12)
   > setSelected!(
    execute
    self:TA Base Editor Button Widget Object Reference=@n13
    selected:bool="Equal Exactly (String)"(A=Get currentStatToVisualize(), B=getKey!(self:PCPT Stat Report Widget Object Reference=@n13) -> (key:String))
    throwEvent:bool=false
   ) -> (then)
   n14 = Is Drawing Active(execute=@n2.then_3)
   drawDataInWorld(execute=[@n14.Yep_Actors, @n14.Yep_Instances]) {is=title K2Node_CallFunction}
  ]}
 signUpForTimelineEvents {nodes [
   signUpForTimelineEvents() {is=title K2Node_FunctionEntry}
   > Bind Event to On Value Changed(self:PCPT Timeline Widget Object Reference=Get timeline(), Delegate=Create Event()) {is=title K2Node_AddDelegate}
  ]}
 tick_handlePlay {nodes [
   n1 = tick_handlePlay!() -> (then, deltaTime:double=~) {is=title K2Node_FunctionEntry}
   > n2 = Set timeSkip(timeSkip:double=float + float(A=@n1, B=Get timeSkip() -> (timeSkip:double)) {is=title K2Node_CommutativeAssociativeBinaryOperator}) -> (Output_Get:double)
   > Branch(Condition="float >= float"(A=@n2, B=getPlaybackSpeed!() -> (value:double) {is=title K2Node_CallFunction}))
   > Set timeSkip(timeSkip=0)
   > playToNextDataIndex() {is=title K2Node_CallFunction}
  ]}
 tryGetPCPTActor {nodes [
   Inputs!() -> (" ":Exec)
   > n1 = getNextExistingDrawActor!(execute, self) -> (then, actor:PCPT Display Utility Actors Object Reference) {is=title K2Node_CallFunction}
   > n2 = Is Valid(InputObject=@n1)
   > Outputs!(true:Exec, false:Exec=@n2.Is Not Valid, actor:PCPT Display Utility Actors Object Reference=@n1)
  ]}
 updatePawnPosition {nodes [
   n1 = updatePawnPosition!() -> (then, Force:bool=~) {is=title K2Node_FunctionEntry}
   n2 = getShouldMoveCamera() -> (index=0) {is=title K2Node_CallFunction}
   > Branch(Condition=OR Boolean(A=@n1, B="integer > integer"(A=@n2, B=0)))
   > n3 = Branch(Condition="Equal (Integer)"(A=getShouldMoveCamera() -> (index=0), B=2))
   n4 = GetLevelViewportCameraInfo(self=Get UnrealEditorSubsystem())
   > Branch(Condition=@n4.ReturnValue)
   > n5 = getStatPositionAndForward(index=Get currentIndex() -> (currentIndex=0))
   n6 = GetForwardVector!(InRot:Rotator=@n4.CameraRotation) -> (ReturnValue:Vector)
   > SetLevelViewportCameraInfo!(
    execute
    self:Unreal Editor Subsystem Object Reference=Get UnrealEditorSubsystem()
    CameraLocation:Vector=vector + vector(A:Vector=@n5, B:Vector=vector * float(A=@n6, B=-5000)) -> (ReturnValue:Vector) {is=title K2Node_CommutativeAssociativeBinaryOperator}
    CameraRotation:Rotator=@n4.CameraRotation
   ) -> (then)
   n7 = getStatPositionAndForward(execute=@n3.else, index=Get currentIndex() -> (currentIndex=0))
   > SetLevelViewportCameraInfo(self=Get UnrealEditorSubsystem(), CameraLocation=@n7, CameraRotation=Rotation From X Vector(InVec=@n7.forward))
  ]}
 updateStatsDisplay {nodes [
   updateStatsDisplay() {is=title K2Node_FunctionEntry}
   n1 = "To String (Integer)"(InInt=getDataLength() -> (ReturnValue=0))
   n2 = Split(SourceString=Get currentFilePath(), InStr=/, SearchDir=FromEnd)
   > "SetText (Text)"(
    self=Get txtTotal() -> (txtTotal:Common Text Object Reference)
    InText="To Text (String)"(InString=Append(A=@n2.RightS, B=" - Total Data Points: ", C=@n1))
   )
   > SetToolTipText(self=Get txtTotal(), InToolTipText="To Text (String)"(InString=Get currentFilePath()))
   > n3 = getListItems(self=Get listStatCounters())
   > n4 = For Each Loop(Array=@n3)
   > n5 = Cast To PCPT_StatReport_Widget(Object=@n4)
   n6 = Get currentIndex() -> (currentIndex=0)
   > setIndex!(execute, self:PCPT Stat Report Widget Object Reference=@n5, index:int=@n6) -> (then)
   > setColor(
    self:PCPT Stat Report Widget Object Reference=@n5
    color=getDrawColor(checkIndex=@n6, Stat=Get label!(self:TA Base Editor Button Widget Object Reference=@n5) -> (label:String)) {is=title K2Node_CallFunction}
   )
  ]}
]
variables [
 cachedPCPTActors {type="array<PCPT_DisplayUtilityActors_C>", default []}
 cachedViewLocations {type="array<Vector>"}
 cachedViewRotations {type="array<Vector>"}
 currentEvents {type="array<string>", default []}
 currentFilePath {type=string, default=""}
 currentIndex {type=integer}
 currentMap {type="map<string>", default {}}
 currentPCPTActors {type="array<PCPT_DisplayUtilityActors_C>", default []}
 currentStatToVisualize {type=string, default=""}
 currentStatValues {type="array<number>", default []}
 dataOffset {type=number}
 everyOther_Counter {type=integer}
 isPlaying {type=boolean}
 shouldSave {type=boolean}
 stats {type="map<PCPT_Stat_Struct>", default {}}
 statToWidgetMap {type="map<UserWidget>", default {}}
 statVisualizers {type="map<PCPT_Base_StatVisualizer_C>", default {}}
 timeSkip {type=number}
 useActors {type=boolean}
 visualizersToLoad {
  type="array<Class>"
  default [
   /Game/NeverCook/EditorUtilities/PerfTools/PerfCsvPlaybackTool/Data/StatVisualizers/DynamicRes_StatVisualizer.DynamicRes_StatVisualizer_C
   /Game/NeverCook/EditorUtilities/PerfTools/PerfCsvPlaybackTool/Data/StatVisualizers/DrawCall_StatVisualizer.DrawCall_StatVisualizer_C
   /Game/NeverCook/EditorUtilities/PerfTools/PerfCsvPlaybackTool/Data/StatVisualizers/FrameTime_StatVisualizer.FrameTime_StatVisualizer_C
   /Game/NeverCook/EditorUtilities/PerfTools/PerfCsvPlaybackTool/Data/StatVisualizers/GameThreadTime_StatVisualizer.GameThreadTime_StatVisualizer_C
   /Game/NeverCook/EditorUtilities/PerfTools/PerfCsvPlaybackTool/Data/StatVisualizers/GPUTime_StatVisualizer.GPUTime_StatVisualizer_C
   /Game/NeverCook/EditorUtilities/PerfTools/PerfCsvPlaybackTool/Data/StatVisualizers/RenderThreadTime_StatVisualizer.RenderThreadTime_StatVisualizer_C
   /Game/NeverCook/EditorUtilities/PerfTools/PerfCsvPlaybackTool/Data/StatVisualizers/Lights_StatVisualizer.Lights_StatVisualizer_C
   /Game/NeverCook/EditorUtilities/PerfTools/PerfCsvPlaybackTool/Data/StatVisualizers/Lights_ShadowOff_StatVisualizer.Lights_ShadowOff_StatVisualizer_C
   /Game/NeverCook/EditorUtilities/PerfTools/PerfCsvPlaybackTool/Data/StatVisualizers/Lights_Shadowed_StatVisualizer.Lights_Shadowed_StatVisualizer_C
   /Game/NeverCook/EditorUtilities/PerfTools/PerfCsvPlaybackTool/Data/StatVisualizers/GPUShadowDepths_StatVisualizer.GPUShadowDepths_StatVisualizer_C
   /Game/NeverCook/EditorUtilities/PerfTools/PerfCsvPlaybackTool/Data/StatVisualizers/LightsUpdateMaps_StatVisualizer.LightsUpdateMaps_StatVisualizer_C
   /Game/NeverCook/EditorUtilities/PerfTools/PerfCsvPlaybackTool/Data/StatVisualizers/MemFree_StatVisualizer.MemFree_StatVisualizer_C
   /Game/NeverCook/EditorUtilities/PerfTools/PerfCsvPlaybackTool/Data/StatVisualizers/GPUVirtTex_StatVisualizer.GPUVirtTex_StatVisualizer_C
  ]
 }
]
widgets {
 info {
  parentClass=/Game/Blueprints/Blutilities/ForTechArt/Widgets/TA_Base_EditorWidget.TA_Base_EditorWidget_C
  rootWidgetClass=&Overlay
  widgetCount=91
 }
 widgets [
  Overlay_0 {widgetClassPath=&Overlay, widget=Overlay_0}
  VerticalBox_0 {
   widgetClassPath=&VerticalBox
   parent=Overlay_0
   widget=VerticalBox_0
   slot {class=OverlaySlot, properties {horizontalAlignment=HAlign_Fill, verticalAlignment=VAlign_Fill}}
  }
  HorizontalBox_1 {
   widgetClassPath=&HorizontalBox
   parent=VerticalBox_0
   widget=HorizontalBox_1
   slot {class=VerticalBoxSlot, properties {padding=<ltrb 15 5 15 0>}}
  }
  btnLoad {
   widgetClassPath=&TA_Generic01_Button
   parent=HorizontalBox_1
   widget=btnLoad
   properties {label=Load New CSV, toolTipText=Use the windows dialog that appears to select a perf csv report.}
   slot {class=HorizontalBoxSlot, properties {verticalAlignment=VAlign_Center, size {sizeRule=Fill, value=1}}}
  }
  Border_5 {
   widgetClassPath=&Border
   parent=HorizontalBox_1
   widget=Border_5
   properties {brushColor=<rgba 0.0431 0.0627 0.0863 1>}
   slot {class=HorizontalBoxSlot, properties {verticalAlignment=VAlign_Center, padding=<ltrb 5 0 5 0>}}
  }
  HorizontalBox_7 {widgetClassPath=&HorizontalBox, parent=Border_5, widget=HorizontalBox_7, slot {class=BorderSlot}}
  txtMin_3 {
   widgetClassPath=&CommonTextBlock
   parent=HorizontalBox_7
   widget=txtMin_3
   properties {
    text=Max Data to Load:
    visibility=Visible
    font {
     bForceMonospaced=false
     bMaterialIsStencil=false
     fontMaterial=null
     fontObject=&NotoSans
     letterSpacing=0
     monospacedWidth=1
     size=10
     skewAmount=0
     typefaceFontName=NotoSans-Regular
     outlineSettings {
      bApplyOutlineToDropShadows=false
      bMiteredCorners=false
      bSeparateFillAlpha=false
      outlineMaterial=null
      outlineSize=0
      outlineColor=<rgba 0 0 0 1>
     }
    }
    shadowOffset=<0 0>
   }
   slot {class=HorizontalBoxSlot, properties {verticalAlignment=VAlign_Center}}
  }
  SizeBox_1 {
   widgetClassPath=&SizeBox
   parent=HorizontalBox_7
   widget=SizeBox_1
   properties {bOverride_WidthOverride=true, widthOverride=100.6438}
   slot {class=HorizontalBoxSlot}
  }
  spinMaxDataPoints {
   widgetClassPath=/Script/UMG.SpinBox
   bIsVariable=true
   parent=SizeBox_1
   widget=spinMaxDataPoints
   properties {
    bAlwaysUsesDeltaSnap=true
    bOverride_MaxSliderValue=true
    bOverride_MaxValue=true
    bOverride_MinSliderValue=true
    bOverride_MinValue=true
    delta=1
    maxFractionalDigits=0
    maxSliderValue=5000
    maxValue=5000
    minSliderValue=1000
    minValue=1000
    toolTipText="When we load a new CSV we will parse it down to <= this number of points.  Lower numbers sacrifice some accuracy to load faster."
    value=3000
    foregroundColor {colorUseRule=UseColor_Specified, specifiedColor=<rgba 1 1 1 1>}
    widgetStyle {
     activeBackgroundBrush {
      bIsDynamicallyLoaded=false
      drawAs=RoundedBox
      imageType=NoImage
      mirroring=NoMirror
      resourceName=None
      resourceObject=null
      tiling=NoTile
      imageSize=<0 0>
      margin=<ltrb 0 0 0 0>
      outlineSettings {
       bUseBrushTransparency=false
       roundingType=FixedRadius
       width=1
       color {colorUseRule=UseColor_ColorTable, specifiedColor=<rgba 0 0 0 0>}
       cornerRadii=<4 4 4 4>
      }
      tintColor {colorUseRule=UseColor_Specified, specifiedColor=<rgba 0.0048 0.0048 0.0048 1>}
      uVRegion {bIsValid=false, max=<0 0>, min=<0 0>}
     }
     activeFillBrush {
      bIsDynamicallyLoaded=false
      drawAs=RoundedBox
      imageType=NoImage
      mirroring=NoMirror
      resourceName=None
      resourceObject=null
      tiling=NoTile
      imageSize=<0 0>
      margin=<ltrb 0 0 0 0>
      outlineSettings {
       bUseBrushTransparency=false
       roundingType=FixedRadius
       width=1
       color {colorUseRule=UseColor_Specified, specifiedColor=<rgba 0 0 0 0>}
       cornerRadii=<4 4 4 4>
      }
      tintColor {colorUseRule=UseColor_Specified, specifiedColor=<rgba 0.0953 0.0953 0.0953 1>}
      uVRegion {bIsValid=false, max=<0 0>, min=<0 0>}
     }
     arrowsImage {
      bIsDynamicallyLoaded=false
      drawAs=NoDrawType
      imageType=NoImage
      mirroring=NoMirror
      resourceName=None
      resourceObject=null
      tiling=NoTile
      imageSize=<0 0>
      margin=<ltrb 0 0 0 0>
      outlineSettings {
       bUseBrushTransparency=false
       roundingType=HalfHeightRadius
       width=0
       color {colorUseRule=UseColor_Specified, specifiedColor=<rgba 0 0 0 0>}
       cornerRadii=<0 0 0 0>
      }
      tintColor {colorUseRule=UseColor_Specified, specifiedColor=<rgba 1 1 1 1>}
      uVRegion {bIsValid=false, max=<0 0>, min=<0 0>}
     }
     backgroundBrush {
      bIsDynamicallyLoaded=false
      drawAs=RoundedBox
      imageType=NoImage
      mirroring=NoMirror
      resourceName=None
      resourceObject=null
      tiling=NoTile
      imageSize=<0 0>
      margin=<ltrb 0 0 0 0>
      outlineSettings {
       bUseBrushTransparency=false
       roundingType=FixedRadius
       width=1
       color {colorUseRule=UseColor_ColorTable, specifiedColor=<rgba 0 0 0 0>}
       cornerRadii=<4 4 4 4>
      }
      tintColor {colorUseRule=UseColor_Specified, specifiedColor=<rgba 0.0048 0.0048 0.0048 1>}
      uVRegion {bIsValid=false, max=<0 0>, min=<0 0>}
     }
     foregroundColor {colorUseRule=UseColor_Specified, specifiedColor=<rgba 1 1 1 1>}
     hoveredBackgroundBrush {
      bIsDynamicallyLoaded=false
      drawAs=RoundedBox
      imageType=NoImage
      mirroring=NoMirror
      resourceName=None
      resourceObject=null
      tiling=NoTile
      imageSize=<0 0>
      margin=<ltrb 0 0 0 0>
      outlineSettings {
       bUseBrushTransparency=false
       roundingType=FixedRadius
       width=1
       color {colorUseRule=UseColor_ColorTable, specifiedColor=<rgba 0 0 0 0>}
       cornerRadii=<4 4 4 4>
      }
      tintColor {colorUseRule=UseColor_Specified, specifiedColor=<rgba 0.0048 0.0048 0.0048 1>}
      uVRegion {bIsValid=false, max=<0 0>, min=<0 0>}
     }
     hoveredFillBrush {
      bIsDynamicallyLoaded=false
      drawAs=RoundedBox
      imageType=NoImage
      mirroring=NoMirror
      resourceName=None
      resourceObject=null
      tiling=NoTile
      imageSize=<0 0>
      margin=<ltrb 0 0 0 0>
      outlineSettings {
       bUseBrushTransparency=false
       roundingType=FixedRadius
       width=1
       color {colorUseRule=UseColor_Specified, specifiedColor=<rgba 0 0 0 0>}
       cornerRadii=<4 4 4 4>
      }
      tintColor {colorUseRule=UseColor_Specified, specifiedColor=<rgba 0.0953 0.0953 0.0953 1>}
      uVRegion {bIsValid=false, max=<0 0>, min=<0 0>}
     }
     inactiveFillBrush {
      bIsDynamicallyLoaded=false
      drawAs=RoundedBox
      imageType=NoImage
      mirroring=NoMirror
      resourceName=None
      resourceObject=null
      tiling=NoTile
      imageSize=<0 0>
      margin=<ltrb 0 0 0 0>
      outlineSettings {
       bUseBrushTransparency=false
       roundingType=FixedRadius
       width=1
       color {colorUseRule=UseColor_Specified, specifiedColor=<rgba 0 0 0 0>}
       cornerRadii=<4 4 4 4>
      }
      tintColor {colorUseRule=UseColor_Specified, specifiedColor=<rgba 0.0395 0.0395 0.0395 1>}
      uVRegion {bIsValid=false, max=<0 0>, min=<0 0>}
     }
     insetPadding=<ltrb 1 1 1 1>
     textPadding=<ltrb 10 3.5 10 4>
    }
   }
   slot {class=SizeBoxSlot, properties {verticalAlignment=VAlign_Center, padding=<ltrb 5 0 0 0>}}
  }
  txtTotal {
   widgetClassPath=&CommonTextBlock
   bIsVariable=true
   parent=VerticalBox_0
   widget=txtTotal
   properties {
    visibility=Visible
    font {
     bForceMonospaced=false
     bMaterialIsStencil=false
     fontMaterial=null
     fontObject=&NotoSans
     letterSpacing=0
     monospacedWidth=1
     size=10
     skewAmount=0
     typefaceFontName=NotoSans-Regular
     outlineSettings {
      bApplyOutlineToDropShadows=false
      bMiteredCorners=false
      bSeparateFillAlpha=false
      outlineMaterial=null
      outlineSize=0
      outlineColor=<rgba 0 0 0 1>
     }
    }
    shadowOffset=<0 0>
   }
   slot {class=VerticalBoxSlot, properties {horizontalAlignment=HAlign_Center, padding=<ltrb 0 3 0 3>}}
  }
  VerticalBox_3 {widgetClassPath=&VerticalBox, parent=VerticalBox_0, widget=VerticalBox_3, slot {class=VerticalBoxSlot}}
  Image {
   widgetClassPath=&Image
   bIsVariable=true
   parent=VerticalBox_3
   widget=Image
   properties {
    brush {
     bIsDynamicallyLoaded=false
     drawAs=Image
     imageType=NoImage
     mirroring=NoMirror
     resourceName=None
     resourceObject=null
     tiling=NoTile
     imageSize=<32 2>
     margin=<ltrb 0 0 0 0>
     outlineSettings {
      bUseBrushTransparency=false
      roundingType=HalfHeightRadius
      width=0
      color {colorUseRule=UseColor_Specified, specifiedColor=<rgba 0 0 0 0>}
      cornerRadii=<0 0 0 0>
     }
     tintColor {colorUseRule=UseColor_Specified, specifiedColor=<rgba 1 1 1 1>}
     uVRegion {bIsValid=false, max=<0 0>, min=<0 0>}
    }
    colorAndOpacity=<rgba 1 1 1 0.781>
   }
   slot {class=VerticalBoxSlot, properties {padding=<ltrb 0 3 0 0>}}
  }
  Border_14 {
   widgetClassPath=&Border
   parent=VerticalBox_3
   widget=Border_14
   properties {brushColor=<rgba 0.0156 0.0156 0.0156 1>, padding=<ltrb 5 5 5 5>}
   slot {class=VerticalBoxSlot}
  }
  HorizontalBox_59 {
   widgetClassPath=&HorizontalBox
   parent=Border_14
   widget=HorizontalBox_59
   slot {class=BorderSlot, properties {padding=<ltrb 5 5 5 5>}}
  }
  btnClearDraw_1 {
   widgetClassPath=&TA_Generic01_Button
   parent=HorizontalBox_59
   widget=btnClearDraw_1
   properties {label=Edit}
   slot {class=HorizontalBoxSlot, properties {verticalAlignment=VAlign_Center, padding=<ltrb 0 0 5 0>}}
  }
  listStatCounters {
   widgetClassPath=/Game/Blueprints/Blutilities/ForTechArt/Widgets/List/TA_Generic01_List
   bIsVariable=true
   parent=HorizontalBox_59
   widget=listStatCounters
   properties {
    allowMultiSelect=false
    allowNoSelection=true
    listItemClass=/Game/NeverCook/EditorUtilities/PerfTools/PerfCsvPlaybackTool/SubWidgets/PCPT_StatReport_Widget.PCPT_StatReport_Widget_C
    scrollOrientation=Orient_Horizontal
   }
   slot {class=HorizontalBoxSlot, properties {horizontalAlignment=HAlign_Center, size {sizeRule=Fill, value=1}}}
  }
  Image_163 {
   widgetClassPath=&Image
   bIsVariable=true
   parent=VerticalBox_3
   widget=Image_163
   properties {
    brush {
     bIsDynamicallyLoaded=false
     drawAs=Image
     imageType=NoImage
     mirroring=NoMirror
     resourceName=None
     resourceObject=null
     tiling=NoTile
     imageSize=<32 2>
     margin=<ltrb 0 0 0 0>
     outlineSettings {
      bUseBrushTransparency=false
      roundingType=HalfHeightRadius
      width=0
      color {colorUseRule=UseColor_Specified, specifiedColor=<rgba 0 0 0 0>}
      cornerRadii=<0 0 0 0>
     }
     tintColor {colorUseRule=UseColor_Specified, specifiedColor=<rgba 1 1 1 1>}
     uVRegion {bIsValid=false, max=<0 0>, min=<0 0>}
    }
    colorAndOpacity=<rgba 1 1 1 0.781>
   }
   slot {class=VerticalBoxSlot, properties {padding=<ltrb 0 0 0 3>}}
  }
  HorizontalBox_207 {
   widgetClassPath=&HorizontalBox
   parent=VerticalBox_0
   widget=HorizontalBox_207
   slot {class=VerticalBoxSlot, properties {horizontalAlignment=HAlign_Center, padding=<ltrb 0 0 0 5>}}
  }
  txtMin_2 {
   widgetClassPath=&CommonTextBlock
   parent=HorizontalBox_207
   widget=txtMin_2
   properties {
    text="Smooth Data: "
    visibility=Visible
    font {
     bForceMonospaced=false
     bMaterialIsStencil=false
     fontMaterial=null
     fontObject=&NotoSans
     letterSpacing=0
     monospacedWidth=1
     size=12
     skewAmount=0
     typefaceFontName=NotoSans-Regular
     outlineSettings {
      bApplyOutlineToDropShadows=false
      bMiteredCorners=false
      bSeparateFillAlpha=false
      outlineMaterial=null
      outlineSize=0
      outlineColor=<rgba 0 0 0 1>
     }
    }
    shadowOffset=<0 0>
   }
   slot {class=HorizontalBoxSlot, properties {verticalAlignment=VAlign_Center, padding=<ltrb 5 0 0 0>}}
  }
  spinDataSmoothing {
   widgetClassPath=&SpinBox
   bIsVariable=true
   parent=HorizontalBox_207
   widget=spinDataSmoothing
   properties {
    bOverride_MaxSliderValue=true
    bOverride_MaxValue=true
    bOverride_MinSliderValue=true
    bOverride_MinValue=true
    delta=1
    maxSliderValue=10
    maxValue=10
    foregroundColor {colorUseRule=UseColor_Specified, specifiedColor=<rgba 1 1 1 1>}
    widgetStyle {
     activeBackgroundBrush {
      bIsDynamicallyLoaded=false
      drawAs=RoundedBox
      imageType=NoImage
      mirroring=NoMirror
      resourceName=None
      resourceObject=null
      tiling=NoTile
      imageSize=<0 0>
      margin=<ltrb 0 0 0 0>
      outlineSettings {
       bUseBrushTransparency=false
       roundingType=FixedRadius
       width=1
       color {colorUseRule=UseColor_ColorTable, specifiedColor=<rgba 0 0 0 0>}
       cornerRadii=<4 4 4 4>
      }
      tintColor {colorUseRule=UseColor_Specified, specifiedColor=<rgba 0.0048 0.0048 0.0048 1>}
      uVRegion {bIsValid=false, max=<0 0>, min=<0 0>}
     }
     activeFillBrush {
      bIsDynamicallyLoaded=false
      drawAs=RoundedBox
      imageType=NoImage
      mirroring=NoMirror
      resourceName=None
      resourceObject=null
      tiling=NoTile
      imageSize=<0 0>
      margin=<ltrb 0 0 0 0>
      outlineSettings {
       bUseBrushTransparency=false
       roundingType=FixedRadius
       width=1
       color {colorUseRule=UseColor_Specified, specifiedColor=<rgba 0 0 0 0>}
       cornerRadii=<4 4 4 4>
      }
      tintColor {colorUseRule=UseColor_Specified, specifiedColor=<rgba 0.0953 0.0953 0.0953 1>}
      uVRegion {bIsValid=false, max=<0 0>, min=<0 0>}
     }
     arrowsImage {
      bIsDynamicallyLoaded=false
      drawAs=NoDrawType
      imageType=NoImage
      mirroring=NoMirror
      resourceName=None
      resourceObject=null
      tiling=NoTile
      imageSize=<0 0>
      margin=<ltrb 0 0 0 0>
      outlineSettings {
       bUseBrushTransparency=false
       roundingType=HalfHeightRadius
       width=0
       color {colorUseRule=UseColor_Specified, specifiedColor=<rgba 0 0 0 0>}
       cornerRadii=<0 0 0 0>
      }
      tintColor {colorUseRule=UseColor_Specified, specifiedColor=<rgba 1 1 1 1>}
      uVRegion {bIsValid=false, max=<0 0>, min=<0 0>}
     }
     backgroundBrush {
      bIsDynamicallyLoaded=false
      drawAs=RoundedBox
      imageType=NoImage
      mirroring=NoMirror
      resourceName=None
      resourceObject=null
      tiling=NoTile
      imageSize=<0 0>
      margin=<ltrb 0 0 0 0>
      outlineSettings {
       bUseBrushTransparency=false
       roundingType=FixedRadius
       width=1
       color {colorUseRule=UseColor_ColorTable, specifiedColor=<rgba 0 0 0 0>}
       cornerRadii=<4 4 4 4>
      }
      tintColor {colorUseRule=UseColor_Specified, specifiedColor=<rgba 0.0048 0.0048 0.0048 1>}
      uVRegion {bIsValid=false, max=<0 0>, min=<0 0>}
     }
     foregroundColor {colorUseRule=UseColor_Specified, specifiedColor=<rgba 1 1 1 1>}
     hoveredBackgroundBrush {
      bIsDynamicallyLoaded=false
      drawAs=RoundedBox
      imageType=NoImage
      mirroring=NoMirror
      resourceName=None
      resourceObject=null
      tiling=NoTile
      imageSize=<0 0>
      margin=<ltrb 0 0 0 0>
      outlineSettings {
       bUseBrushTransparency=false
       roundingType=FixedRadius
       width=1
       color {colorUseRule=UseColor_ColorTable, specifiedColor=<rgba 0 0 0 0>}
       cornerRadii=<4 4 4 4>
      }
      tintColor {colorUseRule=UseColor_Specified, specifiedColor=<rgba 0.0048 0.0048 0.0048 1>}
      uVRegion {bIsValid=false, max=<0 0>, min=<0 0>}
     }
     hoveredFillBrush {
      bIsDynamicallyLoaded=false
      drawAs=RoundedBox
      imageType=NoImage
      mirroring=NoMirror
      resourceName=None
      resourceObject=null
      tiling=NoTile
      imageSize=<0 0>
      margin=<ltrb 0 0 0 0>
      outlineSettings {
       bUseBrushTransparency=false
       roundingType=FixedRadius
       width=1
       color {colorUseRule=UseColor_Specified, specifiedColor=<rgba 0 0 0 0>}
       cornerRadii=<4 4 4 4>
      }
      tintColor {colorUseRule=UseColor_Specified, specifiedColor=<rgba 0.0953 0.0953 0.0953 1>}
      uVRegion {bIsValid=false, max=<0 0>, min=<0 0>}
     }
     inactiveFillBrush {
      bIsDynamicallyLoaded=false
      drawAs=RoundedBox
      imageType=NoImage
      mirroring=NoMirror
      resourceName=None
      resourceObject=null
      tiling=NoTile
      imageSize=<0 0>
      margin=<ltrb 0 0 0 0>
      outlineSettings {
       bUseBrushTransparency=false
       roundingType=FixedRadius
       width=1
       color {colorUseRule=UseColor_Specified, specifiedColor=<rgba 0 0 0 0>}
       cornerRadii=<4 4 4 4>
      }
      tintColor {colorUseRule=UseColor_Specified, specifiedColor=<rgba 0.0395 0.0395 0.0395 1>}
      uVRegion {bIsValid=false, max=<0 0>, min=<0 0>}
     }
     insetPadding=<ltrb 1 1 1 1>
     textPadding=<ltrb 10 3.5 10 4>
    }
   }
   slot {class=HorizontalBoxSlot}
  }
  HorizontalBox_57 {
   widgetClassPath=&HorizontalBox
   parent=VerticalBox_0
   widget=HorizontalBox_57
   slot {class=VerticalBoxSlot, properties {verticalAlignment=VAlign_Bottom, padding=<ltrb 10 0 10 0>}}
  }
  HorizontalBox_562 {
   widgetClassPath=&HorizontalBox
   parent=HorizontalBox_57
   widget=HorizontalBox_562
   slot {class=HorizontalBoxSlot, properties {padding=<ltrb 10 0 0 0>, size {sizeRule=Fill, value=1}}}
  }
  VerticalBox_4 {
   widgetClassPath=&VerticalBox
   parent=HorizontalBox_562
   widget=VerticalBox_4
   slot {class=HorizontalBoxSlot, properties {horizontalAlignment=HAlign_Right}}
  }
  txtMax {
   widgetClassPath=&CommonTextBlock
   bIsVariable=true
   parent=VerticalBox_4
   widget=txtMax
   properties {
    justification=Right
    text=Max
    font {
     bForceMonospaced=false
     bMaterialIsStencil=false
     fontMaterial=null
     fontObject=&NotoSans
     letterSpacing=0
     monospacedWidth=1
     size=8
     skewAmount=0
     typefaceFontName=NotoSans-Regular
     outlineSettings {
      bApplyOutlineToDropShadows=false
      bMiteredCorners=false
      bSeparateFillAlpha=false
      outlineMaterial=null
      outlineSize=0
      outlineColor=<rgba 0 0 0 1>
     }
    }
    shadowOffset=<0 0>
   }
   slot {class=VerticalBoxSlot, properties {verticalAlignment=VAlign_Top, size {sizeRule=Fill, value=1}}}
  }
  txtMin {
   widgetClassPath=&CommonTextBlock
   bIsVariable=true
   parent=VerticalBox_4
   widget=txtMin
   properties {
    justification=Right
    text=Min
    font {
     bForceMonospaced=false
     bMaterialIsStencil=false
     fontMaterial=null
     fontObject=&NotoSans
     letterSpacing=0
     monospacedWidth=1
     size=8
     skewAmount=0
     typefaceFontName=NotoSans-Regular
     outlineSettings {
      bApplyOutlineToDropShadows=false
      bMiteredCorners=false
      bSeparateFillAlpha=false
      outlineMaterial=null
      outlineSize=0
      outlineColor=<rgba 0 0 0 1>
     }
    }
    shadowOffset=<0 0>
   }
   slot {class=VerticalBoxSlot, properties {verticalAlignment=VAlign_Bottom, size {sizeRule=Fill, value=1}}}
  }
  SizeBox_2 {
   widgetClassPath=&SizeBox
   parent=HorizontalBox_562
   widget=SizeBox_2
   properties {bOverride_HeightOverride=true, heightOverride=150}
   slot {class=HorizontalBoxSlot, properties {size {sizeRule=Fill, value=1}}}
  }
  timeline {
   widgetClassPath=/Game/NeverCook/EditorUtilities/PerfTools/PerfCsvPlaybackTool/SubWidgets/PCPT_Timeline_Widget
   bIsVariable=true
   parent=SizeBox_2
   widget=timeline
   properties {value=0}
   slot {class=SizeBoxSlot}
  }
  Border_2 {
   widgetClassPath=&Border
   parent=VerticalBox_0
   widget=Border_2
   properties {brushColor=<rgba 0.0156 0.0156 0.0156 1>, padding=<ltrb 10 5 10 5>}
   slot {class=VerticalBoxSlot, properties {padding=<ltrb 50 0 15 0>}}
  }
  HorizontalBox_452 {
   widgetClassPath=&HorizontalBox
   parent=Border_2
   widget=HorizontalBox_452
   slot {class=BorderSlot, properties {padding=<ltrb 10 5 10 5>}}
  }
  btnPlay {
   widgetClassPath=&TA_Generic01_Button
   parent=HorizontalBox_452
   widget=btnPlay
   properties {label=Play}
   slot {class=HorizontalBoxSlot, properties {horizontalAlignment=HAlign_Left, verticalAlignment=VAlign_Center}}
  }
  HorizontalBox_455 {
   widgetClassPath=&HorizontalBox
   parent=HorizontalBox_452
   widget=HorizontalBox_455
   properties {toolTipText=Playback Speed}
   slot {class=HorizontalBoxSlot, properties {horizontalAlignment=HAlign_Left, verticalAlignment=VAlign_Center, padding=<ltrb 20 0 0 0>}}
  }
  txtMin_1 {
   widgetClassPath=&CommonTextBlock
   parent=HorizontalBox_455
   widget=txtMin_1
   properties {
    text="Speed: "
    visibility=Visible
    font {
     bForceMonospaced=false
     bMaterialIsStencil=false
     fontMaterial=null
     fontObject=&NotoSans
     letterSpacing=0
     monospacedWidth=1
     size=12
     skewAmount=0
     typefaceFontName=NotoSans-Regular
     outlineSettings {
      bApplyOutlineToDropShadows=false
      bMiteredCorners=false
      bSeparateFillAlpha=false
      outlineMaterial=null
      outlineSize=0
      outlineColor=<rgba 0 0 0 1>
     }
    }
    shadowOffset=<0 0>
   }
   slot {class=HorizontalBoxSlot}
  }
  SizeBox_0 {
   widgetClassPath=&SizeBox
   bIsVariable=true
   parent=HorizontalBox_455
   widget=SizeBox_0
   properties {bOverride_WidthOverride=true, minDesiredWidth=50, toolTipText=Playback Speed, widthOverride=50}
   slot {class=HorizontalBoxSlot, properties {horizontalAlignment=HAlign_Center, verticalAlignment=VAlign_Center, size {sizeRule=Fill, value=1}}}
  }
  spinSpeed {
   widgetClassPath=&SpinBox
   bIsVariable=true
   parent=SizeBox_0
   widget=spinSpeed
   properties {
    bOverride_MaxSliderValue=true
    bOverride_MaxValue=true
    bOverride_MinSliderValue=true
    bOverride_MinValue=true
    delta=1
    maxSliderValue=10
    maxValue=10
    minSliderValue=1
    minValue=1
    value=1
    foregroundColor {colorUseRule=UseColor_Specified, specifiedColor=<rgba 1 1 1 1>}
    widgetStyle {
     activeBackgroundBrush {
      bIsDynamicallyLoaded=false
      drawAs=RoundedBox
      imageType=NoImage
      mirroring=NoMirror
      resourceName=None
      resourceObject=null
      tiling=NoTile
      imageSize=<0 0>
      margin=<ltrb 0 0 0 0>
      outlineSettings {
       bUseBrushTransparency=false
       roundingType=FixedRadius
       width=1
       color {colorUseRule=UseColor_ColorTable, specifiedColor=<rgba 0 0 0 0>}
       cornerRadii=<4 4 4 4>
      }
      tintColor {colorUseRule=UseColor_Specified, specifiedColor=<rgba 0.0048 0.0048 0.0048 1>}
      uVRegion {bIsValid=false, max=<0 0>, min=<0 0>}
     }
     activeFillBrush {
      bIsDynamicallyLoaded=false
      drawAs=RoundedBox
      imageType=NoImage
      mirroring=NoMirror
      resourceName=None
      resourceObject=null
      tiling=NoTile
      imageSize=<0 0>
      margin=<ltrb 0 0 0 0>
      outlineSettings {
       bUseBrushTransparency=false
       roundingType=FixedRadius
       width=1
       color {colorUseRule=UseColor_Specified, specifiedColor=<rgba 0 0 0 0>}
       cornerRadii=<4 4 4 4>
      }
      tintColor {colorUseRule=UseColor_Specified, specifiedColor=<rgba 0.0953 0.0953 0.0953 1>}
      uVRegion {bIsValid=false, max=<0 0>, min=<0 0>}
     }
     arrowsImage {
      bIsDynamicallyLoaded=false
      drawAs=NoDrawType
      imageType=NoImage
      mirroring=NoMirror
      resourceName=None
      resourceObject=null
      tiling=NoTile
      imageSize=<0 0>
      margin=<ltrb 0 0 0 0>
      outlineSettings {
       bUseBrushTransparency=false
       roundingType=HalfHeightRadius
       width=0
       color {colorUseRule=UseColor_Specified, specifiedColor=<rgba 0 0 0 0>}
       cornerRadii=<0 0 0 0>
      }
      tintColor {colorUseRule=UseColor_Specified, specifiedColor=<rgba 1 1 1 1>}
      uVRegion {bIsValid=false, max=<0 0>, min=<0 0>}
     }
     backgroundBrush {
      bIsDynamicallyLoaded=false
      drawAs=RoundedBox
      imageType=NoImage
      mirroring=NoMirror
      resourceName=None
      resourceObject=null
      tiling=NoTile
      imageSize=<0 0>
      margin=<ltrb 0 0 0 0>
      outlineSettings {
       bUseBrushTransparency=false
       roundingType=FixedRadius
       width=1
       color {colorUseRule=UseColor_ColorTable, specifiedColor=<rgba 0 0 0 0>}
       cornerRadii=<4 4 4 4>
      }
      tintColor {colorUseRule=UseColor_Specified, specifiedColor=<rgba 0.0048 0.0048 0.0048 1>}
      uVRegion {bIsValid=false, max=<0 0>, min=<0 0>}
     }
     foregroundColor {colorUseRule=UseColor_Specified, specifiedColor=<rgba 1 1 1 1>}
     hoveredBackgroundBrush {
      bIsDynamicallyLoaded=false
      drawAs=RoundedBox
      imageType=NoImage
      mirroring=NoMirror
      resourceName=None
      resourceObject=null
      tiling=NoTile
      imageSize=<0 0>
      margin=<ltrb 0 0 0 0>
      outlineSettings {
       bUseBrushTransparency=false
       roundingType=FixedRadius
       width=1
       color {colorUseRule=UseColor_ColorTable, specifiedColor=<rgba 0 0 0 0>}
       cornerRadii=<4 4 4 4>
      }
      tintColor {colorUseRule=UseColor_Specified, specifiedColor=<rgba 0.0048 0.0048 0.0048 1>}
      uVRegion {bIsValid=false, max=<0 0>, min=<0 0>}
     }
     hoveredFillBrush {
      bIsDynamicallyLoaded=false
      drawAs=RoundedBox
      imageType=NoImage
      mirroring=NoMirror
      resourceName=None
      resourceObject=null
      tiling=NoTile
      imageSize=<0 0>
      margin=<ltrb 0 0 0 0>
      outlineSettings {
       bUseBrushTransparency=false
       roundingType=FixedRadius
       width=1
       color {colorUseRule=UseColor_Specified, specifiedColor=<rgba 0 0 0 0>}
       cornerRadii=<4 4 4 4>
      }
      tintColor {colorUseRule=UseColor_Specified, specifiedColor=<rgba 0.0953 0.0953 0.0953 1>}
      uVRegion {bIsValid=false, max=<0 0>, min=<0 0>}
     }
     inactiveFillBrush {
      bIsDynamicallyLoaded=false
      drawAs=RoundedBox
      imageType=NoImage
      mirroring=NoMirror
      resourceName=None
      resourceObject=null
      tiling=NoTile
      imageSize=<0 0>
      margin=<ltrb 0 0 0 0>
      outlineSettings {
       bUseBrushTransparency=false
       roundingType=FixedRadius
       width=1
       color {colorUseRule=UseColor_Specified, specifiedColor=<rgba 0 0 0 0>}
       cornerRadii=<4 4 4 4>
      }
      tintColor {colorUseRule=UseColor_Specified, specifiedColor=<rgba 0.0395 0.0395 0.0395 1>}
      uVRegion {bIsValid=false, max=<0 0>, min=<0 0>}
     }
     insetPadding=<ltrb 1 1 1 1>
     textPadding=<ltrb 10 3.5 10 4>
    }
   }
   slot {class=SizeBoxSlot}
  }
  cbRepeat {
   widgetClassPath=/Script/UMG.CheckBox
   bIsVariable=true
   parent=HorizontalBox_452
   widget=cbRepeat
   properties {
    checkedState=Checked
    toolTipText="If True, when the timeline reaches the end it will start over again."
    widgetStyle {
     checkBoxType=CheckBox
     backgroundHoveredImage {
      bIsDynamicallyLoaded=false
      drawAs=RoundedBox
      imageType=NoImage
      mirroring=NoMirror
      resourceName=None
      resourceObject=null
      tiling=NoTile
      imageSize=<18 18>
      margin=<ltrb 0 0 0 0>
      outlineSettings {
       bUseBrushTransparency=false
       roundingType=FixedRadius
       width=1
       color {colorUseRule=UseColor_ColorTable, specifiedColor=<rgba 0 0 0 0>}
       cornerRadii=<3 3 3 3>
      }
      tintColor {colorUseRule=UseColor_ColorTable, specifiedColor=<rgba 0 0 0 0>}
      uVRegion {bIsValid=false, max=<0 0>, min=<0 0>}
     }
     backgroundImage {
      bIsDynamicallyLoaded=false
      drawAs=RoundedBox
      imageType=NoImage
      mirroring=NoMirror
      resourceName=None
      resourceObject=null
      tiling=NoTile
      imageSize=<18 18>
      margin=<ltrb 0 0 0 0>
      outlineSettings {
       bUseBrushTransparency=false
       roundingType=FixedRadius
       width=1
       color {colorUseRule=UseColor_ColorTable, specifiedColor=<rgba 0 0 0 0>}
       cornerRadii=<3 3 3 3>
      }
      tintColor {colorUseRule=UseColor_ColorTable, specifiedColor=<rgba 0 0 0 0>}
      uVRegion {bIsValid=false, max=<0 0>, min=<0 0>}
     }
     backgroundPressedImage {
      bIsDynamicallyLoaded=false
      drawAs=RoundedBox
      imageType=NoImage
      mirroring=NoMirror
      resourceName=None
      resourceObject=null
      tiling=NoTile
      imageSize=<18 18>
      margin=<ltrb 0 0 0 0>
      outlineSettings {
       bUseBrushTransparency=false
       roundingType=FixedRadius
       width=1
       color {colorUseRule=UseColor_ColorTable, specifiedColor=<rgba 0 0 0 0>}
       cornerRadii=<3 3 3 3>
      }
      tintColor {colorUseRule=UseColor_ColorTable, specifiedColor=<rgba 0 0 0 0>}
      uVRegion {bIsValid=false, max=<0 0>, min=<0 0>}
     }
     borderBackgroundColor {colorUseRule=UseColor_Specified, specifiedColor=<rgba 1 1 1 1>}
     checkedForeground {colorUseRule=UseColor_ColorTable, specifiedColor=<rgba 0 0 0 0>}
     checkedHoveredForeground {colorUseRule=UseColor_ColorTable, specifiedColor=<rgba 0 0 0 0>}
     checkedHoveredImage {
      bIsDynamicallyLoaded=false
      drawAs=Image
      imageType=Vector
      mirroring=NoMirror
      resourceName=../../../Engine/Content/Slate/Starship/CoreWidgets/CheckBox/check.svg
      resourceObject=null
      tiling=NoTile
      imageSize=<16 16>
      margin=<ltrb 0 0 0 0>
      outlineSettings {
       bUseBrushTransparency=false
       roundingType=HalfHeightRadius
       width=0
       color {colorUseRule=UseColor_Specified, specifiedColor=<rgba 0 0 0 0>}
       cornerRadii=<0 0 0 0>
      }
      tintColor {colorUseRule=UseColor_Specified, specifiedColor=<rgba 0.159 0.6445 1 1>}
      uVRegion {bIsValid=false, max=<0 0>, min=<0 0>}
     }
     checkedImage {
      bIsDynamicallyLoaded=false
      drawAs=Image
      imageType=Vector
      mirroring=NoMirror
      resourceName=../../../Engine/Content/Slate/Starship/CoreWidgets/CheckBox/check.svg
      resourceObject=null
      tiling=NoTile
      imageSize=<16 16>
      margin=<ltrb 0 0 0 0>
      outlineSettings {
       bUseBrushTransparency=false
       roundingType=HalfHeightRadius
       width=0
       color {colorUseRule=UseColor_Specified, specifiedColor=<rgba 0 0 0 0>}
       cornerRadii=<0 0 0 0>
      }
      tintColor {colorUseRule=UseColor_Specified, specifiedColor=<rgba 0.0194 0.4969 1 1>}
      uVRegion {bIsValid=false, max=<0 0>, min=<0 0>}
     }
     checkedPressedForeground {colorUseRule=UseColor_ColorTable, specifiedColor=<rgba 0 0 0 0>}
     checkedPressedImage {
      bIsDynamicallyLoaded=false
      drawAs=Image
      imageType=Vector
      mirroring=NoMirror
      resourceName=../../../Engine/Content/Slate/Starship/CoreWidgets/CheckBox/check.svg
      resourceObject=null
      tiling=NoTile
      imageSize=<16 16>
      margin=<ltrb 0 0 0 0>
      outlineSettings {
       bUseBrushTransparency=false
       roundingType=HalfHeightRadius
       width=0
       color {colorUseRule=UseColor_Specified, specifiedColor=<rgba 0 0 0 0>}
       cornerRadii=<0 0 0 0>
      }
      tintColor {colorUseRule=UseColor_Specified, specifiedColor=<rgba 0.0194 0.4969 1 1>}
      uVRegion {bIsValid=false, max=<0 0>, min=<0 0>}
     }
     checkedSlateSound {resourceObject=null}
     foregroundColor {colorUseRule=UseColor_Specified, specifiedColor=<rgba 0.5271 0.5271 0.5271 1>}
     hoveredForeground {colorUseRule=UseColor_ColorTable, specifiedColor=<rgba 0 0 0 0>}
     hoveredSlateSound {resourceObject=null}
     padding=<ltrb 2 0 0 0>
     pressedForeground {colorUseRule=UseColor_ColorTable, specifiedColor=<rgba 0 0 0 0>}
     uncheckedHoveredImage {
      bIsDynamicallyLoaded=false
      drawAs=NoDrawType
      imageType=NoImage
      mirroring=NoMirror
      resourceName=None
      resourceObject=null
      tiling=NoTile
      imageSize=<0 0>
      margin=<ltrb 0 0 0 0>
      outlineSettings {
       bUseBrushTransparency=false
       roundingType=HalfHeightRadius
       width=0
       color {colorUseRule=UseColor_Specified, specifiedColor=<rgba 0 0 0 0>}
       cornerRadii=<0 0 0 0>
      }
      tintColor {colorUseRule=UseColor_Specified, specifiedColor=<rgba 1 1 1 1>}
      uVRegion {bIsValid=false, max=<0 0>, min=<0 0>}
     }
     uncheckedImage {
      bIsDynamicallyLoaded=false
      drawAs=NoDrawType
      imageType=NoImage
      mirroring=NoMirror
      resourceName=None
      resourceObject=null
      tiling=NoTile
      imageSize=<0 0>
      margin=<ltrb 0 0 0 0>
      outlineSettings {
       bUseBrushTransparency=false
       roundingType=HalfHeightRadius
       width=0
       color {colorUseRule=UseColor_Specified, specifiedColor=<rgba 0 0 0 0>}
       cornerRadii=<0 0 0 0>
      }
      tintColor {colorUseRule=UseColor_Specified, specifiedColor=<rgba 1 1 1 1>}
      uVRegion {bIsValid=false, max=<0 0>, min=<0 0>}
     }
     uncheckedPressedImage {
      bIsDynamicallyLoaded=false
      drawAs=NoDrawType
      imageType=NoImage
      mirroring=NoMirror
      resourceName=None
      resourceObject=null
      tiling=NoTile
      imageSize=<0 0>
      margin=<ltrb 0 0 0 0>
      outlineSettings {
       bUseBrushTransparency=false
       roundingType=HalfHeightRadius
       width=0
       color {colorUseRule=UseColor_Specified, specifiedColor=<rgba 0 0 0 0>}
       cornerRadii=<0 0 0 0>
      }
      tintColor {colorUseRule=UseColor_Specified, specifiedColor=<rgba 1 1 1 1>}
      uVRegion {bIsValid=false, max=<0 0>, min=<0 0>}
     }
     uncheckedSlateSound {resourceObject=null}
     undeterminedForeground {colorUseRule=UseColor_ColorTable, specifiedColor=<rgba 0 0 0 0>}
     undeterminedHoveredImage {
      bIsDynamicallyLoaded=false
      drawAs=Image
      imageType=Vector
      mirroring=NoMirror
      resourceName=../../../Engine/Content/Slate/Starship/CoreWidgets/CheckBox/indeterminate.svg
      resourceObject=null
      tiling=NoTile
      imageSize=<16 16>
      margin=<ltrb 0 0 0 0>
      outlineSettings {
       bUseBrushTransparency=false
       roundingType=HalfHeightRadius
       width=0
       color {colorUseRule=UseColor_Specified, specifiedColor=<rgba 0 0 0 0>}
       cornerRadii=<0 0 0 0>
      }
      tintColor {colorUseRule=UseColor_Specified, specifiedColor=<rgba 0.159 0.6445 1 1>}
      uVRegion {bIsValid=false, max=<0 0>, min=<0 0>}
     }
     undeterminedImage {
      bIsDynamicallyLoaded=false
      drawAs=Image
      imageType=Vector
      mirroring=NoMirror
      resourceName=../../../Engine/Content/Slate/Starship/CoreWidgets/CheckBox/indeterminate.svg
      resourceObject=null
      tiling=NoTile
      imageSize=<16 16>
      margin=<ltrb 0 0 0 0>
      outlineSettings {
       bUseBrushTransparency=false
       roundingType=HalfHeightRadius
       width=0
       color {colorUseRule=UseColor_Specified, specifiedColor=<rgba 0 0 0 0>}
       cornerRadii=<0 0 0 0>
      }
      tintColor {colorUseRule=UseColor_Specified, specifiedColor=<rgba 0.0194 0.4969 1 1>}
      uVRegion {bIsValid=false, max=<0 0>, min=<0 0>}
     }
     undeterminedPressedImage {
      bIsDynamicallyLoaded=false
      drawAs=Image
      imageType=Vector
      mirroring=NoMirror
      resourceName=../../../Engine/Content/Slate/Starship/CoreWidgets/CheckBox/indeterminate.svg
      resourceObject=null
      tiling=NoTile
      imageSize=<16 16>
      margin=<ltrb 0 0 0 0>
      outlineSettings {
       bUseBrushTransparency=false
       roundingType=HalfHeightRadius
       width=0
       color {colorUseRule=UseColor_Specified, specifiedColor=<rgba 0 0 0 0>}
       cornerRadii=<0 0 0 0>
      }
      tintColor {colorUseRule=UseColor_Specified, specifiedColor=<rgba 0.0194 0.4969 1 1>}
      uVRegion {bIsValid=false, max=<0 0>, min=<0 0>}
     }
    }
   }
   slot {
    class=HorizontalBoxSlot
    properties {horizontalAlignment=HAlign_Left, verticalAlignment=VAlign_Center, padding=<ltrb 20 0 0 0>, size {sizeRule=Fill, value=1}}
   }
  }
  txtRepeat {
   widgetClassPath=&CommonTextBlock
   parent=cbRepeat
   widget=txtRepeat
   properties {
    text=Wrap
    font {
     bForceMonospaced=false
     bMaterialIsStencil=false
     fontMaterial=null
     fontObject=&NotoSans
     letterSpacing=0
     monospacedWidth=1
     size=12
     skewAmount=0
     typefaceFontName=NotoSans-Regular
     outlineSettings {
      bApplyOutlineToDropShadows=false
      bMiteredCorners=false
      bSeparateFillAlpha=false
      outlineMaterial=null
      outlineSize=0
      outlineColor=<rgba 0 0 0 1>
     }
    }
    shadowOffset=<0 0>
   }
   slot {class=PanelSlot}
  }
  HorizontalBox_2 {
   widgetClassPath=&HorizontalBox
   parent=HorizontalBox_452
   widget=HorizontalBox_2
   slot {class=HorizontalBoxSlot, properties {padding=<ltrb 5 0 0 0>, size {sizeRule=Fill, value=1}}}
  }
  txtRepeat_2 {
   widgetClassPath=&CommonTextBlock
   parent=HorizontalBox_2
   widget=txtRepeat_2
   properties {
    text=Camera When Playing:
    font {
     bForceMonospaced=false
     bMaterialIsStencil=false
     fontMaterial=null
     fontObject=&NotoSans
     letterSpacing=0
     monospacedWidth=1
     size=12
     skewAmount=0
     typefaceFontName=NotoSans-Regular
     outlineSettings {
      bApplyOutlineToDropShadows=false
      bMiteredCorners=false
      bSeparateFillAlpha=false
      outlineMaterial=null
      outlineSize=0
      outlineColor=<rgba 0 0 0 1>
     }
    }
    shadowOffset=<0 0>
   }
   slot {class=HorizontalBoxSlot, properties {verticalAlignment=VAlign_Center}}
  }
  ddMoveType {
   widgetClassPath=/Script/UMG.ComboBoxString
   bIsVariable=true
   parent=HorizontalBox_2
   widget=ddMoveType
   properties {
    font {
     bForceMonospaced=false
     bMaterialIsStencil=false
     fontMaterial=null
     fontObject=&Roboto
     letterSpacing=0
     monospacedWidth=1
     size=14
     skewAmount=0
     typefaceFontName=Bold
     outlineSettings {
      bApplyOutlineToDropShadows=false
      bMiteredCorners=false
      bSeparateFillAlpha=false
      outlineMaterial=null
      outlineSize=0
      outlineColor=<rgba 0 0 0 1>
     }
    }
    foregroundColor {colorUseRule=UseColor_Specified, specifiedColor=<rgba 0.5271 0.5271 0.5271 1>}
    itemStyle {
     bUseParentRowBrush=false
     activeBrush {
      bIsDynamicallyLoaded=false
      drawAs=Image
      imageType=NoImage
      mirroring=NoMirror
      resourceName=None
      resourceObject=null
      tiling=NoTile
      imageSize=<0 0>
      margin=<ltrb 0 0 0 0>
      outlineSettings {
       bUseBrushTransparency=false
       roundingType=HalfHeightRadius
       width=0
       color {colorUseRule=UseColor_Specified, specifiedColor=<rgba 0 0 0 0>}
       cornerRadii=<0 0 0 0>
      }
      tintColor {colorUseRule=UseColor_Specified, specifiedColor=<rgba 0.0194 0.4969 1 1>}
      uVRegion {bIsValid=false, max=<0 0>, min=<0 0>}
     }
     activeHighlightedBrush {
      bIsDynamicallyLoaded=false
      drawAs=Image
      imageType=NoImage
      mirroring=NoMirror
      resourceName=None
      resourceObject=null
      tiling=NoTile
      imageSize=<0 0>
      margin=<ltrb 0 0 0 0>
      outlineSettings {
       bUseBrushTransparency=false
       roundingType=HalfHeightRadius
       width=0
       color {colorUseRule=UseColor_Specified, specifiedColor=<rgba 0 0 0 0>}
       cornerRadii=<0 0 0 0>
      }
      tintColor {colorUseRule=UseColor_Specified, specifiedColor=<rgba 0.0252 0.0319 0.0423 1>}
      uVRegion {bIsValid=false, max=<0 0>, min=<0 0>}
     }
     activeHoveredBrush {
      bIsDynamicallyLoaded=false
      drawAs=Image
      imageType=NoImage
      mirroring=NoMirror
      resourceName=None
      resourceObject=null
      tiling=NoTile
      imageSize=<0 0>
      margin=<ltrb 0 0 0 0>
      outlineSettings {
       bUseBrushTransparency=false
       roundingType=HalfHeightRadius
       width=0
       color {colorUseRule=UseColor_Specified, specifiedColor=<rgba 0 0 0 0>}
       cornerRadii=<0 0 0 0>
      }
      tintColor {colorUseRule=UseColor_Specified, specifiedColor=<rgba 0.0194 0.4969 1 1>}
      uVRegion {bIsValid=false, max=<0 0>, min=<0 0>}
     }
     dropIndicator_Above {
      bIsDynamicallyLoaded=false
      drawAs=Box
      imageType=FullColor
      mirroring=NoMirror
      resourceName=../../../Engine/Content/Slate/Common/DropZoneIndicator_Above.png
      resourceObject=null
      tiling=NoTile
      imageSize=<0 0>
      margin=<ltrb 0.625 0.625 0 0>
      outlineSettings {
       bUseBrushTransparency=false
       roundingType=HalfHeightRadius
       width=0
       color {colorUseRule=UseColor_Specified, specifiedColor=<rgba 0 0 0 0>}
       cornerRadii=<0 0 0 0>
      }
      tintColor {colorUseRule=UseColor_Specified, specifiedColor=<rgba 0.0091 0.3515 0.9647 1>}
      uVRegion {bIsValid=false, max=<0 0>, min=<0 0>}
     }
     dropIndicator_Below {
      bIsDynamicallyLoaded=false
      drawAs=Box
      imageType=FullColor
      mirroring=NoMirror
      resourceName=../../../Engine/Content/Slate/Common/DropZoneIndicator_Below.png
      resourceObject=null
      tiling=NoTile
      imageSize=<0 0>
      margin=<ltrb 0.625 0 0 0.625>
      outlineSettings {
       bUseBrushTransparency=false
       roundingType=HalfHeightRadius
       width=0
       color {colorUseRule=UseColor_Specified, specifiedColor=<rgba 0 0 0 0>}
       cornerRadii=<0 0 0 0>
      }
      tintColor {colorUseRule=UseColor_Specified, specifiedColor=<rgba 0.0091 0.3515 0.9647 1>}
      uVRegion {bIsValid=false, max=<0 0>, min=<0 0>}
     }
     dropIndicator_Onto {
      bIsDynamicallyLoaded=false
      drawAs=Box
      imageType=FullColor
      mirroring=NoMirror
      resourceName=../../../Engine/Content/Slate/Common/DropZoneIndicator_Onto.png
      resourceObject=null
      tiling=NoTile
      imageSize=<0 0>
      margin=<ltrb 0.25 0.25 0.25 0.25>
      outlineSettings {
       bUseBrushTransparency=false
       roundingType=HalfHeightRadius
       width=0
       color {colorUseRule=UseColor_Specified, specifiedColor=<rgba 0 0 0 0>}
       cornerRadii=<0 0 0 0>
      }
      tintColor {colorUseRule=UseColor_Specified, specifiedColor=<rgba 0.0091 0.3515 0.9647 1>}
      uVRegion {bIsValid=false, max=<0 0>, min=<0 0>}
     }
     evenRowBackgroundBrush {
      bIsDynamicallyLoaded=false
      drawAs=Image
      imageType=NoImage
      mirroring=NoMirror
      resourceName=None
      resourceObject=null
      tiling=NoTile
      imageSize=<0 0>
      margin=<ltrb 0 0 0 0>
      outlineSettings {
       bUseBrushTransparency=false
       roundingType=HalfHeightRadius
       width=0
       color {colorUseRule=UseColor_Specified, specifiedColor=<rgba 0 0 0 0>}
       cornerRadii=<0 0 0 0>
      }
      tintColor {colorUseRule=UseColor_Specified, specifiedColor=<rgba 0.0103 0.0103 0.0103 1>}
      uVRegion {bIsValid=false, max=<0 0>, min=<0 0>}
     }
     evenRowBackgroundHoveredBrush {
      bIsDynamicallyLoaded=false
      drawAs=Image
      imageType=NoImage
      mirroring=NoMirror
      resourceName=None
      resourceObject=null
      tiling=NoTile
      imageSize=<0 0>
      margin=<ltrb 0 0 0 0>
      outlineSettings {
       bUseBrushTransparency=false
       roundingType=HalfHeightRadius
       width=0
       color {colorUseRule=UseColor_Specified, specifiedColor=<rgba 0 0 0 0>}
       cornerRadii=<0 0 0 0>
      }
      tintColor {colorUseRule=UseColor_Specified, specifiedColor=<rgba 0.0176 0.0176 0.0176 1>}
      uVRegion {bIsValid=false, max=<0 0>, min=<0 0>}
     }
     inactiveBrush {
      bIsDynamicallyLoaded=false
      drawAs=Image
      imageType=NoImage
      mirroring=NoMirror
      resourceName=None
      resourceObject=null
      tiling=NoTile
      imageSize=<0 0>
      margin=<ltrb 0 0 0 0>
      outlineSettings {
       bUseBrushTransparency=false
       roundingType=HalfHeightRadius
       width=0
       color {colorUseRule=UseColor_Specified, specifiedColor=<rgba 0 0 0 0>}
       cornerRadii=<0 0 0 0>
      }
      tintColor {colorUseRule=UseColor_Specified, specifiedColor=<rgba 0.3185 0.4508 0.521 1>}
      uVRegion {bIsValid=false, max=<0 0>, min=<0 0>}
     }
     inactiveHighlightedBrush {
      bIsDynamicallyLoaded=false
      drawAs=Image
      imageType=NoImage
      mirroring=NoMirror
      resourceName=None
      resourceObject=null
      tiling=NoTile
      imageSize=<0 0>
      margin=<ltrb 0 0 0 0>
      outlineSettings {
       bUseBrushTransparency=false
       roundingType=HalfHeightRadius
       width=0
       color {colorUseRule=UseColor_Specified, specifiedColor=<rgba 0 0 0 0>}
       cornerRadii=<0 0 0 0>
      }
      tintColor {colorUseRule=UseColor_Specified, specifiedColor=<rgba 0.0252 0.0319 0.0423 1>}
      uVRegion {bIsValid=false, max=<0 0>, min=<0 0>}
     }
     inactiveHoveredBrush {
      bIsDynamicallyLoaded=false
      drawAs=Image
      imageType=NoImage
      mirroring=NoMirror
      resourceName=None
      resourceObject=null
      tiling=NoTile
      imageSize=<0 0>
      margin=<ltrb 0 0 0 0>
      outlineSettings {
       bUseBrushTransparency=false
       roundingType=HalfHeightRadius
       width=0
       color {colorUseRule=UseColor_Specified, specifiedColor=<rgba 0 0 0 0>}
       cornerRadii=<0 0 0 0>
      }
      tintColor {colorUseRule=UseColor_Specified, specifiedColor=<rgba 0.3185 0.4508 0.521 1>}
      uVRegion {bIsValid=false, max=<0 0>, min=<0 0>}
     }
     oddRowBackgroundBrush {
      bIsDynamicallyLoaded=false
      drawAs=Image
      imageType=NoImage
      mirroring=NoMirror
      resourceName=None
      resourceObject=null
      tiling=NoTile
      imageSize=<0 0>
      margin=<ltrb 0 0 0 0>
      outlineSettings {
       bUseBrushTransparency=false
       roundingType=HalfHeightRadius
       width=0
       color {colorUseRule=UseColor_Specified, specifiedColor=<rgba 0 0 0 0>}
       cornerRadii=<0 0 0 0>
      }
      tintColor {colorUseRule=UseColor_Specified, specifiedColor=<rgba 0.0103 0.0103 0.0103 1>}
      uVRegion {bIsValid=false, max=<0 0>, min=<0 0>}
     }
     oddRowBackgroundHoveredBrush {
      bIsDynamicallyLoaded=false
      drawAs=Image
      imageType=NoImage
      mirroring=NoMirror
      resourceName=None
      resourceObject=null
      tiling=NoTile
      imageSize=<0 0>
      margin=<ltrb 0 0 0 0>
      outlineSettings {
       bUseBrushTransparency=false
       roundingType=HalfHeightRadius
       width=0
       color {colorUseRule=UseColor_Specified, specifiedColor=<rgba 0 0 0 0>}
       cornerRadii=<0 0 0 0>
      }
      tintColor {colorUseRule=UseColor_Specified, specifiedColor=<rgba 0.0176 0.0176 0.0176 1>}
      uVRegion {bIsValid=false, max=<0 0>, min=<0 0>}
     }
     parentRowBackgroundBrush {
      bIsDynamicallyLoaded=false
      drawAs=Image
      imageType=NoImage
      mirroring=NoMirror
      resourceName=None
      resourceObject=null
      tiling=NoTile
      imageSize=<32 32>
      margin=<ltrb 0 0 0 0>
      outlineSettings {
       bUseBrushTransparency=false
       roundingType=HalfHeightRadius
       width=0
       color {colorUseRule=UseColor_Specified, specifiedColor=<rgba 0 0 0 0>}
       cornerRadii=<0 0 0 0>
      }
      tintColor {colorUseRule=UseColor_Specified, specifiedColor=<rgba 1 1 1 1>}
      uVRegion {bIsValid=false, max=<0 0>, min=<0 0>}
     }
     parentRowBackgroundHoveredBrush {
      bIsDynamicallyLoaded=false
      drawAs=Image
      imageType=NoImage
      mirroring=NoMirror
      resourceName=None
      resourceObject=null
      tiling=NoTile
      imageSize=<32 32>
      margin=<ltrb 0 0 0 0>
      outlineSettings {
       bUseBrushTransparency=false
       roundingType=HalfHeightRadius
       width=0
       color {colorUseRule=UseColor_Specified, specifiedColor=<rgba 0 0 0 0>}
       cornerRadii=<0 0 0 0>
      }
      tintColor {colorUseRule=UseColor_Specified, specifiedColor=<rgba 1 1 1 1>}
      uVRegion {bIsValid=false, max=<0 0>, min=<0 0>}
     }
     selectedTextColor {colorUseRule=UseColor_Specified, specifiedColor=<rgba 0.0048 0.0048 0.0048 1>}
     selectorFocusedBrush {
      bIsDynamicallyLoaded=false
      drawAs=Border
      imageType=FullColor
      mirroring=NoMirror
      resourceName=../../../Engine/Content/Slate/Common/Selector.png
      resourceObject=null
      tiling=Both
      imageSize=<0 0>
      margin=<ltrb 0.25 0.25 0.25 0.25>
      outlineSettings {
       bUseBrushTransparency=false
       roundingType=HalfHeightRadius
       width=0
       color {colorUseRule=UseColor_Specified, specifiedColor=<rgba 0 0 0 0>}
       cornerRadii=<0 0 0 0>
      }
      tintColor {colorUseRule=UseColor_Specified, specifiedColor=<rgba 0.701 0.225 0.003 1>}
      uVRegion {bIsValid=false, max=<0 0>, min=<0 0>}
     }
     textColor {colorUseRule=UseColor_Specified, specifiedColor=<rgba 0.5271 0.5271 0.5271 1>}
    }
    scrollBarStyle {
     thickness=8
     draggedThumbImage {
      bIsDynamicallyLoaded=false
      drawAs=RoundedBox
      imageType=NoImage
      mirroring=NoMirror
      resourceName=None
      resourceObject=null
      tiling=NoTile
      imageSize=<0 0>
      margin=<ltrb 0 0 0 0>
      outlineSettings {
       bUseBrushTransparency=false
       roundingType=FixedRadius
       width=0
       color {colorUseRule=UseColor_Specified, specifiedColor=<rgba 0 0 0 0>}
       cornerRadii=<4 4 4 4>
      }
      tintColor {colorUseRule=UseColor_Specified, specifiedColor=<rgba 0.2159 0.2159 0.2159 1>}
      uVRegion {bIsValid=false, max=<0 0>, min=<0 0>}
     }
     horizontalBackgroundImage {
      bIsDynamicallyLoaded=false
      drawAs=NoDrawType
      imageType=NoImage
      mirroring=NoMirror
      resourceName=None
      resourceObject=null
      tiling=NoTile
      imageSize=<0 0>
      margin=<ltrb 0 0 0 0>
      outlineSettings {
       bUseBrushTransparency=false
       roundingType=HalfHeightRadius
       width=0
       color {colorUseRule=UseColor_Specified, specifiedColor=<rgba 0 0 0 0>}
       cornerRadii=<0 0 0 0>
      }
      tintColor {colorUseRule=UseColor_Specified, specifiedColor=<rgba 1 1 1 1>}
      uVRegion {bIsValid=false, max=<0 0>, min=<0 0>}
     }
     horizontalBottomSlotImage {
      bIsDynamicallyLoaded=false
      drawAs=NoDrawType
      imageType=NoImage
      mirroring=NoMirror
      resourceName=None
      resourceObject=null
      tiling=NoTile
      imageSize=<0 0>
      margin=<ltrb 0 0 0 0>
      outlineSettings {
       bUseBrushTransparency=false
       roundingType=HalfHeightRadius
       width=0
       color {colorUseRule=UseColor_Specified, specifiedColor=<rgba 0 0 0 0>}
       cornerRadii=<0 0 0 0>
      }
      tintColor {colorUseRule=UseColor_Specified, specifiedColor=<rgba 1 1 1 1>}
      uVRegion {bIsValid=false, max=<0 0>, min=<0 0>}
     }
     horizontalTopSlotImage {
      bIsDynamicallyLoaded=false
      drawAs=NoDrawType
      imageType=NoImage
      mirroring=NoMirror
      resourceName=None
      resourceObject=null
      tiling=NoTile
      imageSize=<0 0>
      margin=<ltrb 0 0 0 0>
      outlineSettings {
       bUseBrushTransparency=false
       roundingType=HalfHeightRadius
       width=0
       color {colorUseRule=UseColor_Specified, specifiedColor=<rgba 0 0 0 0>}
       cornerRadii=<0 0 0 0>
      }
      tintColor {colorUseRule=UseColor_Specified, specifiedColor=<rgba 1 1 1 1>}
      uVRegion {bIsValid=false, max=<0 0>, min=<0 0>}
     }
     hoveredThumbImage {
      bIsDynamicallyLoaded=false
      drawAs=RoundedBox
      imageType=NoImage
      mirroring=NoMirror
      resourceName=None
      resourceObject=null
      tiling=NoTile
      imageSize=<0 0>
      margin=<ltrb 0 0 0 0>
      outlineSettings {
       bUseBrushTransparency=false
       roundingType=FixedRadius
       width=0
       color {colorUseRule=UseColor_Specified, specifiedColor=<rgba 0 0 0 0>}
       cornerRadii=<4 4 4 4>
      }
      tintColor {colorUseRule=UseColor_Specified, specifiedColor=<rgba 0.2159 0.2159 0.2159 1>}
      uVRegion {bIsValid=false, max=<0 0>, min=<0 0>}
     }
     normalThumbImage {
      bIsDynamicallyLoaded=false
      drawAs=RoundedBox
      imageType=NoImage
      mirroring=NoMirror
      resourceName=None
      resourceObject=null
      tiling=NoTile
      imageSize=<0 0>
      margin=<ltrb 0 0 0 0>
      outlineSettings {
       bUseBrushTransparency=false
       roundingType=FixedRadius
       width=0
       color {colorUseRule=UseColor_Specified, specifiedColor=<rgba 0 0 0 0>}
       cornerRadii=<4 4 4 4>
      }
      tintColor {colorUseRule=UseColor_Specified, specifiedColor=<rgba 0.0953 0.0953 0.0953 1>}
      uVRegion {bIsValid=false, max=<0 0>, min=<0 0>}
     }
     verticalBackgroundImage {
      bIsDynamicallyLoaded=false
      drawAs=NoDrawType
      imageType=NoImage
      mirroring=NoMirror
      resourceName=None
      resourceObject=null
      tiling=NoTile
      imageSize=<0 0>
      margin=<ltrb 0 0 0 0>
      outlineSettings {
       bUseBrushTransparency=false
       roundingType=HalfHeightRadius
       width=0
       color {colorUseRule=UseColor_Specified, specifiedColor=<rgba 0 0 0 0>}
       cornerRadii=<0 0 0 0>
      }
      tintColor {colorUseRule=UseColor_Specified, specifiedColor=<rgba 1 1 1 1>}
      uVRegion {bIsValid=false, max=<0 0>, min=<0 0>}
     }
     verticalBottomSlotImage {
      bIsDynamicallyLoaded=false
      drawAs=NoDrawType
      imageType=NoImage
      mirroring=NoMirror
      resourceName=None
      resourceObject=null
      tiling=NoTile
      imageSize=<0 0>
      margin=<ltrb 0 0 0 0>
      outlineSettings {
       bUseBrushTransparency=false
       roundingType=HalfHeightRadius
       width=0
       color {colorUseRule=UseColor_Specified, specifiedColor=<rgba 0 0 0 0>}
       cornerRadii=<0 0 0 0>
      }
      tintColor {colorUseRule=UseColor_Specified, specifiedColor=<rgba 1 1 1 1>}
      uVRegion {bIsValid=false, max=<0 0>, min=<0 0>}
     }
     verticalTopSlotImage {
      bIsDynamicallyLoaded=false
      drawAs=NoDrawType
      imageType=NoImage
      mirroring=NoMirror
      resourceName=None
      resourceObject=null
      tiling=NoTile
      imageSize=<0 0>
      margin=<ltrb 0 0 0 0>
      outlineSettings {
       bUseBrushTransparency=false
       roundingType=HalfHeightRadius
       width=0
       color {colorUseRule=UseColor_Specified, specifiedColor=<rgba 0 0 0 0>}
       cornerRadii=<0 0 0 0>
      }
      tintColor {colorUseRule=UseColor_Specified, specifiedColor=<rgba 1 1 1 1>}
      uVRegion {bIsValid=false, max=<0 0>, min=<0 0>}
     }
    }
    widgetStyle {
     comboButtonStyle {
      downArrowAlign=VAlign_Center
      buttonStyle {
       clickedSlateSound {resourceObject=null}
       disabled {
        bIsDynamicallyLoaded=false
        drawAs=RoundedBox
        imageType=NoImage
        mirroring=NoMirror
        resourceName=None
        resourceObject=null
        tiling=NoTile
        imageSize=<0 0>
        margin=<ltrb 0 0 0 0>
        outlineSettings {
         bUseBrushTransparency=false
         roundingType=FixedRadius
         width=1
         color {colorUseRule=UseColor_ColorTable, specifiedColor=<rgba 0 0 0 0>}
         cornerRadii=<4 4 4 4>
        }
        tintColor {colorUseRule=UseColor_Specified, specifiedColor=<rgba 0.0395 0.0395 0.0395 1>}
        uVRegion {bIsValid=false, max=<0 0>, min=<0 0>}
       }
       disabledForeground {colorUseRule=UseColor_ColorTable, specifiedColor=<rgba 0 0 0 0>}
       hovered {
        bIsDynamicallyLoaded=false
        drawAs=RoundedBox
        imageType=NoImage
        mirroring=NoMirror
        resourceName=None
        resourceObject=null
        tiling=NoTile
        imageSize=<0 0>
        margin=<ltrb 0 0 0 0>
        outlineSettings {
         bUseBrushTransparency=false
         roundingType=FixedRadius
         width=1
         color {colorUseRule=UseColor_ColorTable, specifiedColor=<rgba 0 0 0 0>}
         cornerRadii=<4 4 4 4>
        }
        tintColor {colorUseRule=UseColor_Specified, specifiedColor=<rgba 0.0048 0.0048 0.0048 1>}
        uVRegion {bIsValid=false, max=<0 0>, min=<0 0>}
       }
       hoveredForeground {colorUseRule=UseColor_ColorTable, specifiedColor=<rgba 0 0 0 0>}
       hoveredSlateSound {resourceObject=null}
       normal {
        bIsDynamicallyLoaded=false
        drawAs=RoundedBox
        imageType=NoImage
        mirroring=NoMirror
        resourceName=None
        resourceObject=null
        tiling=NoTile
        imageSize=<0 0>
        margin=<ltrb 0 0 0 0>
        outlineSettings {
         bUseBrushTransparency=false
         roundingType=FixedRadius
         width=1
         color {colorUseRule=UseColor_ColorTable, specifiedColor=<rgba 0 0 0 0>}
         cornerRadii=<4 4 4 4>
        }
        tintColor {colorUseRule=UseColor_Specified, specifiedColor=<rgba 0.0048 0.0048 0.0048 1>}
        uVRegion {bIsValid=false, max=<0 0>, min=<0 0>}
       }
       normalForeground {colorUseRule=UseColor_ColorTable, specifiedColor=<rgba 0 0 0 0>}
       normalPadding=<ltrb 8 1 4 1>
       pressed {
        bIsDynamicallyLoaded=false
        drawAs=RoundedBox
        imageType=NoImage
        mirroring=NoMirror
        resourceName=None
        resourceObject=null
        tiling=NoTile
        imageSize=<0 0>
        margin=<ltrb 0 0 0 0>
        outlineSettings {
         bUseBrushTransparency=false
         roundingType=FixedRadius
         width=1
         color {colorUseRule=UseColor_ColorTable, specifiedColor=<rgba 0 0 0 0>}
         cornerRadii=<4 4 4 4>
        }
        tintColor {colorUseRule=UseColor_Specified, specifiedColor=<rgba 0.0048 0.0048 0.0048 1>}
        uVRegion {bIsValid=false, max=<0 0>, min=<0 0>}
       }
       pressedForeground {colorUseRule=UseColor_ColorTable, specifiedColor=<rgba 0 0 0 0>}
       pressedPadding=<ltrb 8 2 4 2>
       pressedSlateSound {resourceObject=null}
      }
      contentPadding=<ltrb 0 0 0 0>
      downArrowImage {
       bIsDynamicallyLoaded=false
       drawAs=Image
       imageType=Vector
       mirroring=NoMirror
       resourceName=../../../Engine/Content/Slate/Starship/CoreWidgets/ComboBox/wide-chevron-down.svg
       resourceObject=null
       tiling=NoTile
       imageSize=<20 16>
       margin=<ltrb 0 0 0 0>
       outlineSettings {
        bUseBrushTransparency=false
        roundingType=HalfHeightRadius
        width=0
        color {colorUseRule=UseColor_Specified, specifiedColor=<rgba 0 0 0 0>}
        cornerRadii=<0 0 0 0>
       }
       tintColor {colorUseRule=UseColor_Specified, specifiedColor=<rgba 1 1 1 1>}
       uVRegion {bIsValid=false, max=<0 0>, min=<0 0>}
      }
      downArrowPadding=<ltrb 2 2 2 2>
      menuBorderBrush {
       bIsDynamicallyLoaded=false
       drawAs=RoundedBox
       imageType=NoImage
       mirroring=NoMirror
       resourceName=None
       resourceObject=null
       tiling=NoTile
       imageSize=<0 0>
       margin=<ltrb 0 0 0 0>
       outlineSettings {
        bUseBrushTransparency=false
        roundingType=FixedRadius
        width=1
        color {colorUseRule=UseColor_ColorTable, specifiedColor=<rgba 0 0 0 0>}
        cornerRadii=<0 0 0 0>
       }
       tintColor {colorUseRule=UseColor_Specified, specifiedColor=<rgba 0.0395 0.0395 0.0395 1>}
       uVRegion {bIsValid=false, max=<0 0>, min=<0 0>}
      }
      menuBorderPadding=<ltrb 0 0 0 0>
      shadowColorAndOpacity=<rgba 0 0 0 1>
      shadowOffset=<0 0>
     }
     contentPadding=<ltrb 0 0 0 0>
     menuRowPadding=<ltrb 8 3 8 3>
     pressedSlateSound {resourceObject=null}
     selectionChangeSlateSound {resourceObject=null}
    }
   }
   slot {class=HorizontalBoxSlot, properties {size {sizeRule=Fill, value=1}}}
  }
  btnJumpTo {
   widgetClassPath=&TA_Generic01_Button
   parent=HorizontalBox_2
   widget=btnJumpTo
   properties {label=Jump To, toolTipText=Jump to the current location/rotation of the current timeline location.}
   slot {class=HorizontalBoxSlot, properties {horizontalAlignment=HAlign_Right, verticalAlignment=VAlign_Center}}
  }
  Image_1 {
   widgetClassPath=&Image
   bIsVariable=true
   parent=VerticalBox_0
   widget=Image_1
   properties {
    brush {
     bIsDynamicallyLoaded=false
     drawAs=Image
     imageType=NoImage
     mirroring=NoMirror
     resourceName=None
     resourceObject=null
     tiling=NoTile
     imageSize=<32 2>
     margin=<ltrb 0 0 0 0>
     outlineSettings {
      bUseBrushTransparency=false
      roundingType=HalfHeightRadius
      width=0
      color {colorUseRule=UseColor_Specified, specifiedColor=<rgba 0 0 0 0>}
      cornerRadii=<0 0 0 0>
     }
     tintColor {colorUseRule=UseColor_Specified, specifiedColor=<rgba 1 1 1 1>}
     uVRegion {bIsValid=false, max=<0 0>, min=<0 0>}
    }
    colorAndOpacity=<rgba 1 1 1 0.781>
   }
   slot {class=VerticalBoxSlot, properties {padding=<ltrb 0 5 0 0>}}
  }
  Border_1 {
   widgetClassPath=&Border
   parent=VerticalBox_0
   widget=Border_1
   properties {brushColor=<rgba 0.0156 0.0156 0.0156 1>, padding=<ltrb 10 10 10 5>}
   slot {class=VerticalBoxSlot}
  }
  VerticalBox_6 {
   widgetClassPath=&VerticalBox
   parent=Border_1
   widget=VerticalBox_6
   slot {class=BorderSlot, properties {padding=<ltrb 10 10 10 5>}}
  }
  HorizontalBox_560 {widgetClassPath=&HorizontalBox, parent=VerticalBox_6, widget=HorizontalBox_560, slot {class=VerticalBoxSlot}}
  btnDraw {
   widgetClassPath=&TA_Generic01_Button
   parent=HorizontalBox_560
   widget=btnDraw
   properties {label=Draw}
   slot {class=HorizontalBoxSlot, properties {verticalAlignment=VAlign_Center, size {sizeRule=Fill, value=1}}}
  }
  btnClearDraw {
   widgetClassPath=&TA_Generic01_Button
   parent=HorizontalBox_560
   widget=btnClearDraw
   properties {label=Clear Drawings, buttonColor_Default=<rgba 1 0.2181 0.0581 1>}
   slot {class=HorizontalBoxSlot, properties {verticalAlignment=VAlign_Center, padding=<ltrb 5 0 0 0>}}
  }
  PCPT_OffsetTransform_EWidget {
   widgetClassPath=&/PCPT_OffsetTransform_EWidget
   bIsVariable=true
   parent=VerticalBox_6
   widget=PCPT_OffsetTransform_EWidget
   slot {class=VerticalBoxSlot, properties {padding=<ltrb 15 0 15 0>}}
  }
  Border_13 {
   widgetClassPath=&Border
   parent=VerticalBox_6
   widget=Border_13
   properties {brushColor=<rgba 1 1 1 0.1>, padding=<ltrb 4 4 4 4>}
   slot {class=VerticalBoxSlot, properties {padding=<ltrb 5 5 5 5>}}
  }
  VerticalBox_2 {
   widgetClassPath=&VerticalBox
   parent=Border_13
   widget=VerticalBox_2
   slot {class=BorderSlot, properties {padding=<ltrb 4 4 4 4>}}
  }
  HorizontalBox_0 {
   widgetClassPath=&HorizontalBox
   parent=VerticalBox_2
   widget=HorizontalBox_0
   slot {class=VerticalBoxSlot, properties {padding=<ltrb 0 5 0 0>}}
  }
  Border_4 {
   widgetClassPath=&Border
   parent=HorizontalBox_0
   widget=Border_4
   properties {brushColor=<rgba 0.0431 0.0627 0.0863 1>, padding=<ltrb 3 3 3 3>}
   slot {class=HorizontalBoxSlot, properties {verticalAlignment=VAlign_Center, padding=<ltrb 10 0 0 0>}}
  }
  HorizontalBox_5 {
   widgetClassPath=&HorizontalBox
   parent=Border_4
   widget=HorizontalBox_5
   slot {class=BorderSlot, properties {padding=<ltrb 3 3 3 3>}}
  }
  txtDrawOnPlay_2 {
   widgetClassPath=&CommonTextBlock
   parent=HorizontalBox_5
   widget=txtDrawOnPlay_2
   properties {
    text=Scale Points:
    font {
     bForceMonospaced=false
     bMaterialIsStencil=false
     fontMaterial=null
     fontObject=&NotoSans
     letterSpacing=0
     monospacedWidth=1
     size=12
     skewAmount=0
     typefaceFontName=NotoSans-Regular
     outlineSettings {
      bApplyOutlineToDropShadows=false
      bMiteredCorners=false
      bSeparateFillAlpha=false
      outlineMaterial=null
      outlineSize=0
      outlineColor=<rgba 0 0 0 1>
     }
    }
    shadowOffset=<0 0>
   }
   slot {class=HorizontalBoxSlot, properties {verticalAlignment=VAlign_Center}}
  }
  spinScale {
   widgetClassPath=&SpinBox
   bIsVariable=true
   parent=HorizontalBox_5
   widget=spinScale
   properties {
    bOverride_MaxSliderValue=true
    bOverride_MaxValue=true
    bOverride_MinSliderValue=true
    bOverride_MinValue=true
    delta=0.1
    maxFractionalDigits=1
    maxSliderValue=10
    maxValue=10
    minSliderValue=0.5
    minValue=0.5
    toolTipText=Playback Speed
    value=1
    foregroundColor {colorUseRule=UseColor_Specified, specifiedColor=<rgba 1 1 1 1>}
    widgetStyle {
     activeBackgroundBrush {
      bIsDynamicallyLoaded=false
      drawAs=RoundedBox
      imageType=NoImage
      mirroring=NoMirror
      resourceName=None
      resourceObject=null
      tiling=NoTile
      imageSize=<0 0>
      margin=<ltrb 0 0 0 0>
      outlineSettings {
       bUseBrushTransparency=false
       roundingType=FixedRadius
       width=1
       color {colorUseRule=UseColor_ColorTable, specifiedColor=<rgba 0 0 0 0>}
       cornerRadii=<4 4 4 4>
      }
      tintColor {colorUseRule=UseColor_Specified, specifiedColor=<rgba 0.0048 0.0048 0.0048 1>}
      uVRegion {bIsValid=false, max=<0 0>, min=<0 0>}
     }
     activeFillBrush {
      bIsDynamicallyLoaded=false
      drawAs=RoundedBox
      imageType=NoImage
      mirroring=NoMirror
      resourceName=None
      resourceObject=null
      tiling=NoTile
      imageSize=<0 0>
      margin=<ltrb 0 0 0 0>
      outlineSettings {
       bUseBrushTransparency=false
       roundingType=FixedRadius
       width=1
       color {colorUseRule=UseColor_Specified, specifiedColor=<rgba 0 0 0 0>}
       cornerRadii=<4 4 4 4>
      }
      tintColor {colorUseRule=UseColor_Specified, specifiedColor=<rgba 0.0953 0.0953 0.0953 1>}
      uVRegion {bIsValid=false, max=<0 0>, min=<0 0>}
     }
     arrowsImage {
      bIsDynamicallyLoaded=false
      drawAs=NoDrawType
      imageType=NoImage
      mirroring=NoMirror
      resourceName=None
      resourceObject=null
      tiling=NoTile
      imageSize=<0 0>
      margin=<ltrb 0 0 0 0>
      outlineSettings {
       bUseBrushTransparency=false
       roundingType=HalfHeightRadius
       width=0
       color {colorUseRule=UseColor_Specified, specifiedColor=<rgba 0 0 0 0>}
       cornerRadii=<0 0 0 0>
      }
      tintColor {colorUseRule=UseColor_Specified, specifiedColor=<rgba 1 1 1 1>}
      uVRegion {bIsValid=false, max=<0 0>, min=<0 0>}
     }
     backgroundBrush {
      bIsDynamicallyLoaded=false
      drawAs=RoundedBox
      imageType=NoImage
      mirroring=NoMirror
      resourceName=None
      resourceObject=null
      tiling=NoTile
      imageSize=<0 0>
      margin=<ltrb 0 0 0 0>
      outlineSettings {
       bUseBrushTransparency=false
       roundingType=FixedRadius
       width=1
       color {colorUseRule=UseColor_ColorTable, specifiedColor=<rgba 0 0 0 0>}
       cornerRadii=<4 4 4 4>
      }
      tintColor {colorUseRule=UseColor_Specified, specifiedColor=<rgba 0.0048 0.0048 0.0048 1>}
      uVRegion {bIsValid=false, max=<0 0>, min=<0 0>}
     }
     foregroundColor {colorUseRule=UseColor_Specified, specifiedColor=<rgba 1 1 1 1>}
     hoveredBackgroundBrush {
      bIsDynamicallyLoaded=false
      drawAs=RoundedBox
      imageType=NoImage
      mirroring=NoMirror
      resourceName=None
      resourceObject=null
      tiling=NoTile
      imageSize=<0 0>
      margin=<ltrb 0 0 0 0>
      outlineSettings {
       bUseBrushTransparency=false
       roundingType=FixedRadius
       width=1
       color {colorUseRule=UseColor_ColorTable, specifiedColor=<rgba 0 0 0 0>}
       cornerRadii=<4 4 4 4>
      }
      tintColor {colorUseRule=UseColor_Specified, specifiedColor=<rgba 0.0048 0.0048 0.0048 1>}
      uVRegion {bIsValid=false, max=<0 0>, min=<0 0>}
     }
     hoveredFillBrush {
      bIsDynamicallyLoaded=false
      drawAs=RoundedBox
      imageType=NoImage
      mirroring=NoMirror
      resourceName=None
      resourceObject=null
      tiling=NoTile
      imageSize=<0 0>
      margin=<ltrb 0 0 0 0>
      outlineSettings {
       bUseBrushTransparency=false
       roundingType=FixedRadius
       width=1
       color {colorUseRule=UseColor_Specified, specifiedColor=<rgba 0 0 0 0>}
       cornerRadii=<4 4 4 4>
      }
      tintColor {colorUseRule=UseColor_Specified, specifiedColor=<rgba 0.0953 0.0953 0.0953 1>}
      uVRegion {bIsValid=false, max=<0 0>, min=<0 0>}
     }
     inactiveFillBrush {
      bIsDynamicallyLoaded=false
      drawAs=RoundedBox
      imageType=NoImage
      mirroring=NoMirror
      resourceName=None
      resourceObject=null
      tiling=NoTile
      imageSize=<0 0>
      margin=<ltrb 0 0 0 0>
      outlineSettings {
       bUseBrushTransparency=false
       roundingType=FixedRadius
       width=1
       color {colorUseRule=UseColor_Specified, specifiedColor=<rgba 0 0 0 0>}
       cornerRadii=<4 4 4 4>
      }
      tintColor {colorUseRule=UseColor_Specified, specifiedColor=<rgba 0.0395 0.0395 0.0395 1>}
      uVRegion {bIsValid=false, max=<0 0>, min=<0 0>}
     }
     insetPadding=<ltrb 1 1 1 1>
     textPadding=<ltrb 10 3.5 10 4>
    }
   }
   slot {class=HorizontalBoxSlot, properties {padding=<ltrb 5 0 0 0>}}
  }
  Border_3 {
   widgetClassPath=&Border
   parent=HorizontalBox_0
   widget=Border_3
   properties {brushColor=<rgba 0.0452 0.0627 0.0885 1>, padding=<ltrb 3 3 3 3>}
   slot {class=HorizontalBoxSlot, properties {verticalAlignment=VAlign_Center, padding=<ltrb 10 0 0 0>}}
  }
  HorizontalBox_4 {
   widgetClassPath=&HorizontalBox
   parent=Border_3
   widget=HorizontalBox_4
   slot {class=BorderSlot, properties {padding=<ltrb 3 3 3 3>}}
  }
  txtDrawOnPlay_1 {
   widgetClassPath=&CommonTextBlock
   parent=HorizontalBox_4
   widget=txtDrawOnPlay_1
   properties {
    text=Draw Every Other:
    font {
     bForceMonospaced=false
     bMaterialIsStencil=false
     fontMaterial=null
     fontObject=&NotoSans
     letterSpacing=0
     monospacedWidth=1
     size=12
     skewAmount=0
     typefaceFontName=NotoSans-Regular
     outlineSettings {
      bApplyOutlineToDropShadows=false
      bMiteredCorners=false
      bSeparateFillAlpha=false
      outlineMaterial=null
      outlineSize=0
      outlineColor=<rgba 0 0 0 1>
     }
    }
    shadowOffset=<0 0>
   }
   slot {class=HorizontalBoxSlot, properties {verticalAlignment=VAlign_Center}}
  }
  spinEveryOther {
   widgetClassPath=&SpinBox
   bIsVariable=true
   parent=HorizontalBox_4
   widget=spinEveryOther
   properties {
    bAlwaysUsesDeltaSnap=true
    bOverride_MaxSliderValue=true
    bOverride_MaxValue=true
    bOverride_MinSliderValue=true
    bOverride_MinValue=true
    delta=1
    maxFractionalDigits=0
    maxSliderValue=50
    maxValue=50
    minSliderValue=1
    minValue=1
    toolTipText="Draw every other ## of points.  If too many points are drawing you can reduce them by raising this.  1 == All Points"
    value=1
    foregroundColor {colorUseRule=UseColor_Specified, specifiedColor=<rgba 1 1 1 1>}
    widgetStyle {
     activeBackgroundBrush {
      bIsDynamicallyLoaded=false
      drawAs=RoundedBox
      imageType=NoImage
      mirroring=NoMirror
      resourceName=None
      resourceObject=null
      tiling=NoTile
      imageSize=<0 0>
      margin=<ltrb 0 0 0 0>
      outlineSettings {
       bUseBrushTransparency=false
       roundingType=FixedRadius
       width=1
       color {colorUseRule=UseColor_ColorTable, specifiedColor=<rgba 0 0 0 0>}
       cornerRadii=<4 4 4 4>
      }
      tintColor {colorUseRule=UseColor_Specified, specifiedColor=<rgba 0.0048 0.0048 0.0048 1>}
      uVRegion {bIsValid=false, max=<0 0>, min=<0 0>}
     }
     activeFillBrush {
      bIsDynamicallyLoaded=false
      drawAs=RoundedBox
      imageType=NoImage
      mirroring=NoMirror
      resourceName=None
      resourceObject=null
      tiling=NoTile
      imageSize=<0 0>
      margin=<ltrb 0 0 0 0>
      outlineSettings {
       bUseBrushTransparency=false
       roundingType=FixedRadius
       width=1
       color {colorUseRule=UseColor_Specified, specifiedColor=<rgba 0 0 0 0>}
       cornerRadii=<4 4 4 4>
      }
      tintColor {colorUseRule=UseColor_Specified, specifiedColor=<rgba 0.0953 0.0953 0.0953 1>}
      uVRegion {bIsValid=false, max=<0 0>, min=<0 0>}
     }
     arrowsImage {
      bIsDynamicallyLoaded=false
      drawAs=NoDrawType
      imageType=NoImage
      mirroring=NoMirror
      resourceName=None
      resourceObject=null
      tiling=NoTile
      imageSize=<0 0>
      margin=<ltrb 0 0 0 0>
      outlineSettings {
       bUseBrushTransparency=false
       roundingType=HalfHeightRadius
       width=0
       color {colorUseRule=UseColor_Specified, specifiedColor=<rgba 0 0 0 0>}
       cornerRadii=<0 0 0 0>
      }
      tintColor {colorUseRule=UseColor_Specified, specifiedColor=<rgba 1 1 1 1>}
      uVRegion {bIsValid=false, max=<0 0>, min=<0 0>}
     }
     backgroundBrush {
      bIsDynamicallyLoaded=false
      drawAs=RoundedBox
      imageType=NoImage
      mirroring=NoMirror
      resourceName=None
      resourceObject=null
      tiling=NoTile
      imageSize=<0 0>
      margin=<ltrb 0 0 0 0>
      outlineSettings {
       bUseBrushTransparency=false
       roundingType=FixedRadius
       width=1
       color {colorUseRule=UseColor_ColorTable, specifiedColor=<rgba 0 0 0 0>}
       cornerRadii=<4 4 4 4>
      }
      tintColor {colorUseRule=UseColor_Specified, specifiedColor=<rgba 0.0048 0.0048 0.0048 1>}
      uVRegion {bIsValid=false, max=<0 0>, min=<0 0>}
     }
     foregroundColor {colorUseRule=UseColor_Specified, specifiedColor=<rgba 1 1 1 1>}
     hoveredBackgroundBrush {
      bIsDynamicallyLoaded=false
      drawAs=RoundedBox
      imageType=NoImage
      mirroring=NoMirror
      resourceName=None
      resourceObject=null
      tiling=NoTile
      imageSize=<0 0>
      margin=<ltrb 0 0 0 0>
      outlineSettings {
       bUseBrushTransparency=false
       roundingType=FixedRadius
       width=1
       color {colorUseRule=UseColor_ColorTable, specifiedColor=<rgba 0 0 0 0>}
       cornerRadii=<4 4 4 4>
      }
      tintColor {colorUseRule=UseColor_Specified, specifiedColor=<rgba 0.0048 0.0048 0.0048 1>}
      uVRegion {bIsValid=false, max=<0 0>, min=<0 0>}
     }
     hoveredFillBrush {
      bIsDynamicallyLoaded=false
      drawAs=RoundedBox
      imageType=NoImage
      mirroring=NoMirror
      resourceName=None
      resourceObject=null
      tiling=NoTile
      imageSize=<0 0>
      margin=<ltrb 0 0 0 0>
      outlineSettings {
       bUseBrushTransparency=false
       roundingType=FixedRadius
       width=1
       color {colorUseRule=UseColor_Specified, specifiedColor=<rgba 0 0 0 0>}
       cornerRadii=<4 4 4 4>
      }
      tintColor {colorUseRule=UseColor_Specified, specifiedColor=<rgba 0.0953 0.0953 0.0953 1>}
      uVRegion {bIsValid=false, max=<0 0>, min=<0 0>}
     }
     inactiveFillBrush {
      bIsDynamicallyLoaded=false
      drawAs=RoundedBox
      imageType=NoImage
      mirroring=NoMirror
      resourceName=None
      resourceObject=null
      tiling=NoTile
      imageSize=<0 0>
      margin=<ltrb 0 0 0 0>
      outlineSettings {
       bUseBrushTransparency=false
       roundingType=FixedRadius
       width=1
       color {colorUseRule=UseColor_Specified, specifiedColor=<rgba 0 0 0 0>}
       cornerRadii=<4 4 4 4>
      }
      tintColor {colorUseRule=UseColor_Specified, specifiedColor=<rgba 0.0395 0.0395 0.0395 1>}
      uVRegion {bIsValid=false, max=<0 0>, min=<0 0>}
     }
     insetPadding=<ltrb 1 1 1 1>
     textPadding=<ltrb 10 3.5 10 4>
    }
   }
   slot {class=HorizontalBoxSlot, properties {padding=<ltrb 5 0 0 0>}}
  }
  Border_0 {
   widgetClassPath=&Border
   parent=HorizontalBox_0
   widget=Border_0
   properties {brushColor=<rgba 0.0452 0.0627 0.0885 1>, padding=<ltrb 3 3 3 3>}
   slot {class=HorizontalBoxSlot, properties {padding=<ltrb 10 0 0 0>}}
  }
  HorizontalBox_3 {
   widgetClassPath=&HorizontalBox
   parent=Border_0
   widget=HorizontalBox_3
   slot {class=BorderSlot, properties {padding=<ltrb 3 3 3 3>}}
  }
  txtDrawOnPlay {
   widgetClassPath=&CommonTextBlock
   parent=HorizontalBox_3
   widget=txtDrawOnPlay
   properties {
    text=Value Threshold:
    font {
     bForceMonospaced=false
     bMaterialIsStencil=false
     fontMaterial=null
     fontObject=&NotoSans
     letterSpacing=0
     monospacedWidth=1
     size=12
     skewAmount=0
     typefaceFontName=NotoSans-Regular
     outlineSettings {
      bApplyOutlineToDropShadows=false
      bMiteredCorners=false
      bSeparateFillAlpha=false
      outlineMaterial=null
      outlineSize=0
      outlineColor=<rgba 0 0 0 1>
     }
    }
    shadowOffset=<0 0>
   }
   slot {class=HorizontalBoxSlot, properties {verticalAlignment=VAlign_Center}}
  }
  spinValueThreshold {
   widgetClassPath=&SpinBox
   bIsVariable=true
   parent=HorizontalBox_3
   widget=spinValueThreshold
   properties {
    bOverride_MaxSliderValue=true
    bOverride_MaxValue=true
    bOverride_MinSliderValue=true
    bOverride_MinValue=true
    delta=0.1
    maxSliderValue=9999
    maxValue=9999
    toolTipText="Only values greater than this will show.  0 means disabled."
    foregroundColor {colorUseRule=UseColor_Specified, specifiedColor=<rgba 1 1 1 1>}
    widgetStyle {
     activeBackgroundBrush {
      bIsDynamicallyLoaded=false
      drawAs=RoundedBox
      imageType=NoImage
      mirroring=NoMirror
      resourceName=None
      resourceObject=null
      tiling=NoTile
      imageSize=<0 0>
      margin=<ltrb 0 0 0 0>
      outlineSettings {
       bUseBrushTransparency=false
       roundingType=FixedRadius
       width=1
       color {colorUseRule=UseColor_ColorTable, specifiedColor=<rgba 0 0 0 0>}
       cornerRadii=<4 4 4 4>
      }
      tintColor {colorUseRule=UseColor_Specified, specifiedColor=<rgba 0.0048 0.0048 0.0048 1>}
      uVRegion {bIsValid=false, max=<0 0>, min=<0 0>}
     }
     activeFillBrush {
      bIsDynamicallyLoaded=false
      drawAs=RoundedBox
      imageType=NoImage
      mirroring=NoMirror
      resourceName=None
      resourceObject=null
      tiling=NoTile
      imageSize=<0 0>
      margin=<ltrb 0 0 0 0>
      outlineSettings {
       bUseBrushTransparency=false
       roundingType=FixedRadius
       width=1
       color {colorUseRule=UseColor_Specified, specifiedColor=<rgba 0 0 0 0>}
       cornerRadii=<4 4 4 4>
      }
      tintColor {colorUseRule=UseColor_Specified, specifiedColor=<rgba 0.0953 0.0953 0.0953 1>}
      uVRegion {bIsValid=false, max=<0 0>, min=<0 0>}
     }
     arrowsImage {
      bIsDynamicallyLoaded=false
      drawAs=NoDrawType
      imageType=NoImage
      mirroring=NoMirror
      resourceName=None
      resourceObject=null
      tiling=NoTile
      imageSize=<0 0>
      margin=<ltrb 0 0 0 0>
      outlineSettings {
       bUseBrushTransparency=false
       roundingType=HalfHeightRadius
       width=0
       color {colorUseRule=UseColor_Specified, specifiedColor=<rgba 0 0 0 0>}
       cornerRadii=<0 0 0 0>
      }
      tintColor {colorUseRule=UseColor_Specified, specifiedColor=<rgba 1 1 1 1>}
      uVRegion {bIsValid=false, max=<0 0>, min=<0 0>}
     }
     backgroundBrush {
      bIsDynamicallyLoaded=false
      drawAs=RoundedBox
      imageType=NoImage
      mirroring=NoMirror
      resourceName=None
      resourceObject=null
      tiling=NoTile
      imageSize=<0 0>
      margin=<ltrb 0 0 0 0>
      outlineSettings {
       bUseBrushTransparency=false
       roundingType=FixedRadius
       width=1
       color {colorUseRule=UseColor_ColorTable, specifiedColor=<rgba 0 0 0 0>}
       cornerRadii=<4 4 4 4>
      }
      tintColor {colorUseRule=UseColor_Specified, specifiedColor=<rgba 0.0048 0.0048 0.0048 1>}
      uVRegion {bIsValid=false, max=<0 0>, min=<0 0>}
     }
     foregroundColor {colorUseRule=UseColor_Specified, specifiedColor=<rgba 1 1 1 1>}
     hoveredBackgroundBrush {
      bIsDynamicallyLoaded=false
      drawAs=RoundedBox
      imageType=NoImage
      mirroring=NoMirror
      resourceName=None
      resourceObject=null
      tiling=NoTile
      imageSize=<0 0>
      margin=<ltrb 0 0 0 0>
      outlineSettings {
       bUseBrushTransparency=false
       roundingType=FixedRadius
       width=1
       color {colorUseRule=UseColor_ColorTable, specifiedColor=<rgba 0 0 0 0>}
       cornerRadii=<4 4 4 4>
      }
      tintColor {colorUseRule=UseColor_Specified, specifiedColor=<rgba 0.0048 0.0048 0.0048 1>}
      uVRegion {bIsValid=false, max=<0 0>, min=<0 0>}
     }
     hoveredFillBrush {
      bIsDynamicallyLoaded=false
      drawAs=RoundedBox
      imageType=NoImage
      mirroring=NoMirror
      resourceName=None
      resourceObject=null
      tiling=NoTile
      imageSize=<0 0>
      margin=<ltrb 0 0 0 0>
      outlineSettings {
       bUseBrushTransparency=false
       roundingType=FixedRadius
       width=1
       color {colorUseRule=UseColor_Specified, specifiedColor=<rgba 0 0 0 0>}
       cornerRadii=<4 4 4 4>
      }
      tintColor {colorUseRule=UseColor_Specified, specifiedColor=<rgba 0.0953 0.0953 0.0953 1>}
      uVRegion {bIsValid=false, max=<0 0>, min=<0 0>}
     }
     inactiveFillBrush {
      bIsDynamicallyLoaded=false
      drawAs=RoundedBox
      imageType=NoImage
      mirroring=NoMirror
      resourceName=None
      resourceObject=null
      tiling=NoTile
      imageSize=<0 0>
      margin=<ltrb 0 0 0 0>
      outlineSettings {
       bUseBrushTransparency=false
       roundingType=FixedRadius
       width=1
       color {colorUseRule=UseColor_Specified, specifiedColor=<rgba 0 0 0 0>}
       cornerRadii=<4 4 4 4>
      }
      tintColor {colorUseRule=UseColor_Specified, specifiedColor=<rgba 0.0395 0.0395 0.0395 1>}
      uVRegion {bIsValid=false, max=<0 0>, min=<0 0>}
     }
     insetPadding=<ltrb 1 1 1 1>
     textPadding=<ltrb 10 3.5 10 4>
    }
   }
   slot {class=HorizontalBoxSlot, properties {padding=<ltrb 5 0 0 0>}}
  }
  btnClearThreshold {
   widgetClassPath=&TA_Generic01_Button
   parent=HorizontalBox_3
   widget=btnClearThreshold
   properties {label=CLEAR, buttonColor_Default=<rgba 0.6771 0.113 0.0564 1>}
   slot {class=HorizontalBoxSlot, properties {verticalAlignment=VAlign_Center, padding=<ltrb 5 0 0 0>}}
  }
  cbThresholdGreater {
   widgetClassPath=&CheckBox
   bIsVariable=true
   parent=HorizontalBox_3
   widget=cbThresholdGreater
   properties {
    checkedState=Checked
    toolTipText="If True, the Threshold show only points above the given value.  Otherwise under."
    widgetStyle {
     checkBoxType=CheckBox
     backgroundHoveredImage {
      bIsDynamicallyLoaded=false
      drawAs=RoundedBox
      imageType=NoImage
      mirroring=NoMirror
      resourceName=None
      resourceObject=null
      tiling=NoTile
      imageSize=<18 18>
      margin=<ltrb 0 0 0 0>
      outlineSettings {
       bUseBrushTransparency=false
       roundingType=FixedRadius
       width=1
       color {colorUseRule=UseColor_ColorTable, specifiedColor=<rgba 0 0 0 0>}
       cornerRadii=<3 3 3 3>
      }
      tintColor {colorUseRule=UseColor_ColorTable, specifiedColor=<rgba 0 0 0 0>}
      uVRegion {bIsValid=false, max=<0 0>, min=<0 0>}
     }
     backgroundImage {
      bIsDynamicallyLoaded=false
      drawAs=RoundedBox
      imageType=NoImage
      mirroring=NoMirror
      resourceName=None
      resourceObject=null
      tiling=NoTile
      imageSize=<18 18>
      margin=<ltrb 0 0 0 0>
      outlineSettings {
       bUseBrushTransparency=false
       roundingType=FixedRadius
       width=1
       color {colorUseRule=UseColor_ColorTable, specifiedColor=<rgba 0 0 0 0>}
       cornerRadii=<3 3 3 3>
      }
      tintColor {colorUseRule=UseColor_ColorTable, specifiedColor=<rgba 0 0 0 0>}
      uVRegion {bIsValid=false, max=<0 0>, min=<0 0>}
     }
     backgroundPressedImage {
      bIsDynamicallyLoaded=false
      drawAs=RoundedBox
      imageType=NoImage
      mirroring=NoMirror
      resourceName=None
      resourceObject=null
      tiling=NoTile
      imageSize=<18 18>
      margin=<ltrb 0 0 0 0>
      outlineSettings {
       bUseBrushTransparency=false
       roundingType=FixedRadius
       width=1
       color {colorUseRule=UseColor_ColorTable, specifiedColor=<rgba 0 0 0 0>}
       cornerRadii=<3 3 3 3>
      }
      tintColor {colorUseRule=UseColor_ColorTable, specifiedColor=<rgba 0 0 0 0>}
      uVRegion {bIsValid=false, max=<0 0>, min=<0 0>}
     }
     borderBackgroundColor {colorUseRule=UseColor_Specified, specifiedColor=<rgba 1 1 1 1>}
     checkedForeground {colorUseRule=UseColor_ColorTable, specifiedColor=<rgba 0 0 0 0>}
     checkedHoveredForeground {colorUseRule=UseColor_ColorTable, specifiedColor=<rgba 0 0 0 0>}
     checkedHoveredImage {
      bIsDynamicallyLoaded=false
      drawAs=Image
      imageType=Vector
      mirroring=NoMirror
      resourceName=../../../Engine/Content/Slate/Starship/CoreWidgets/CheckBox/check.svg
      resourceObject=null
      tiling=NoTile
      imageSize=<16 16>
      margin=<ltrb 0 0 0 0>
      outlineSettings {
       bUseBrushTransparency=false
       roundingType=HalfHeightRadius
       width=0
       color {colorUseRule=UseColor_Specified, specifiedColor=<rgba 0 0 0 0>}
       cornerRadii=<0 0 0 0>
      }
      tintColor {colorUseRule=UseColor_Specified, specifiedColor=<rgba 0.159 0.6445 1 1>}
      uVRegion {bIsValid=false, max=<0 0>, min=<0 0>}
     }
     checkedImage {
      bIsDynamicallyLoaded=false
      drawAs=Image
      imageType=Vector
      mirroring=NoMirror
      resourceName=../../../Engine/Content/Slate/Starship/CoreWidgets/CheckBox/check.svg
      resourceObject=null
      tiling=NoTile
      imageSize=<16 16>
      margin=<ltrb 0 0 0 0>
      outlineSettings {
       bUseBrushTransparency=false
       roundingType=HalfHeightRadius
       width=0
       color {colorUseRule=UseColor_Specified, specifiedColor=<rgba 0 0 0 0>}
       cornerRadii=<0 0 0 0>
      }
      tintColor {colorUseRule=UseColor_Specified, specifiedColor=<rgba 0.0194 0.4969 1 1>}
      uVRegion {bIsValid=false, max=<0 0>, min=<0 0>}
     }
     checkedPressedForeground {colorUseRule=UseColor_ColorTable, specifiedColor=<rgba 0 0 0 0>}
     checkedPressedImage {
      bIsDynamicallyLoaded=false
      drawAs=Image
      imageType=Vector
      mirroring=NoMirror
      resourceName=../../../Engine/Content/Slate/Starship/CoreWidgets/CheckBox/check.svg
      resourceObject=null
      tiling=NoTile
      imageSize=<16 16>
      margin=<ltrb 0 0 0 0>
      outlineSettings {
       bUseBrushTransparency=false
       roundingType=HalfHeightRadius
       width=0
       color {colorUseRule=UseColor_Specified, specifiedColor=<rgba 0 0 0 0>}
       cornerRadii=<0 0 0 0>
      }
      tintColor {colorUseRule=UseColor_Specified, specifiedColor=<rgba 0.0194 0.4969 1 1>}
      uVRegion {bIsValid=false, max=<0 0>, min=<0 0>}
     }
     checkedSlateSound {resourceObject=null}
     foregroundColor {colorUseRule=UseColor_Specified, specifiedColor=<rgba 0.5271 0.5271 0.5271 1>}
     hoveredForeground {colorUseRule=UseColor_ColorTable, specifiedColor=<rgba 0 0 0 0>}
     hoveredSlateSound {resourceObject=null}
     padding=<ltrb 2 0 0 0>
     pressedForeground {colorUseRule=UseColor_ColorTable, specifiedColor=<rgba 0 0 0 0>}
     uncheckedHoveredImage {
      bIsDynamicallyLoaded=false
      drawAs=NoDrawType
      imageType=NoImage
      mirroring=NoMirror
      resourceName=None
      resourceObject=null
      tiling=NoTile
      imageSize=<0 0>
      margin=<ltrb 0 0 0 0>
      outlineSettings {
       bUseBrushTransparency=false
       roundingType=HalfHeightRadius
       width=0
       color {colorUseRule=UseColor_Specified, specifiedColor=<rgba 0 0 0 0>}
       cornerRadii=<0 0 0 0>
      }
      tintColor {colorUseRule=UseColor_Specified, specifiedColor=<rgba 1 1 1 1>}
      uVRegion {bIsValid=false, max=<0 0>, min=<0 0>}
     }
     uncheckedImage {
      bIsDynamicallyLoaded=false
      drawAs=NoDrawType
      imageType=NoImage
      mirroring=NoMirror
      resourceName=None
      resourceObject=null
      tiling=NoTile
      imageSize=<0 0>
      margin=<ltrb 0 0 0 0>
      outlineSettings {
       bUseBrushTransparency=false
       roundingType=HalfHeightRadius
       width=0
       color {colorUseRule=UseColor_Specified, specifiedColor=<rgba 0 0 0 0>}
       cornerRadii=<0 0 0 0>
      }
      tintColor {colorUseRule=UseColor_Specified, specifiedColor=<rgba 1 1 1 1>}
      uVRegion {bIsValid=false, max=<0 0>, min=<0 0>}
     }
     uncheckedPressedImage {
      bIsDynamicallyLoaded=false
      drawAs=NoDrawType
      imageType=NoImage
      mirroring=NoMirror
      resourceName=None
      resourceObject=null
      tiling=NoTile
      imageSize=<0 0>
      margin=<ltrb 0 0 0 0>
      outlineSettings {
       bUseBrushTransparency=false
       roundingType=HalfHeightRadius
       width=0
       color {colorUseRule=UseColor_Specified, specifiedColor=<rgba 0 0 0 0>}
       cornerRadii=<0 0 0 0>
      }
      tintColor {colorUseRule=UseColor_Specified, specifiedColor=<rgba 1 1 1 1>}
      uVRegion {bIsValid=false, max=<0 0>, min=<0 0>}
     }
     uncheckedSlateSound {resourceObject=null}
     undeterminedForeground {colorUseRule=UseColor_ColorTable, specifiedColor=<rgba 0 0 0 0>}
     undeterminedHoveredImage {
      bIsDynamicallyLoaded=false
      drawAs=Image
      imageType=Vector
      mirroring=NoMirror
      resourceName=../../../Engine/Content/Slate/Starship/CoreWidgets/CheckBox/indeterminate.svg
      resourceObject=null
      tiling=NoTile
      imageSize=<16 16>
      margin=<ltrb 0 0 0 0>
      outlineSettings {
       bUseBrushTransparency=false
       roundingType=HalfHeightRadius
       width=0
       color {colorUseRule=UseColor_Specified, specifiedColor=<rgba 0 0 0 0>}
       cornerRadii=<0 0 0 0>
      }
      tintColor {colorUseRule=UseColor_Specified, specifiedColor=<rgba 0.159 0.6445 1 1>}
      uVRegion {bIsValid=false, max=<0 0>, min=<0 0>}
     }
     undeterminedImage {
      bIsDynamicallyLoaded=false
      drawAs=Image
      imageType=Vector
      mirroring=NoMirror
      resourceName=../../../Engine/Content/Slate/Starship/CoreWidgets/CheckBox/indeterminate.svg
      resourceObject=null
      tiling=NoTile
      imageSize=<16 16>
      margin=<ltrb 0 0 0 0>
      outlineSettings {
       bUseBrushTransparency=false
       roundingType=HalfHeightRadius
       width=0
       color {colorUseRule=UseColor_Specified, specifiedColor=<rgba 0 0 0 0>}
       cornerRadii=<0 0 0 0>
      }
      tintColor {colorUseRule=UseColor_Specified, specifiedColor=<rgba 0.0194 0.4969 1 1>}
      uVRegion {bIsValid=false, max=<0 0>, min=<0 0>}
     }
     undeterminedPressedImage {
      bIsDynamicallyLoaded=false
      drawAs=Image
      imageType=Vector
      mirroring=NoMirror
      resourceName=../../../Engine/Content/Slate/Starship/CoreWidgets/CheckBox/indeterminate.svg
      resourceObject=null
      tiling=NoTile
      imageSize=<16 16>
      margin=<ltrb 0 0 0 0>
      outlineSettings {
       bUseBrushTransparency=false
       roundingType=HalfHeightRadius
       width=0
       color {colorUseRule=UseColor_Specified, specifiedColor=<rgba 0 0 0 0>}
       cornerRadii=<0 0 0 0>
      }
      tintColor {colorUseRule=UseColor_Specified, specifiedColor=<rgba 0.0194 0.4969 1 1>}
      uVRegion {bIsValid=false, max=<0 0>, min=<0 0>}
     }
    }
   }
   slot {
    class=HorizontalBoxSlot
    properties {horizontalAlignment=HAlign_Left, verticalAlignment=VAlign_Center, padding=<ltrb 5 0 0 0>, size {sizeRule=Fill, value=1}}
   }
  }
  txtRepeat_1 {
   widgetClassPath=&CommonTextBlock
   parent=cbThresholdGreater
   widget=txtRepeat_1
   properties {
    text=Greater Than
    font {
     bForceMonospaced=false
     bMaterialIsStencil=false
     fontMaterial=null
     fontObject=&NotoSans
     letterSpacing=0
     monospacedWidth=1
     size=12
     skewAmount=0
     typefaceFontName=NotoSans-Regular
     outlineSettings {
      bApplyOutlineToDropShadows=false
      bMiteredCorners=false
      bSeparateFillAlpha=false
      outlineMaterial=null
      outlineSize=0
      outlineColor=<rgba 0 0 0 1>
     }
    }
    shadowOffset=<0 0>
   }
   slot {class=PanelSlot}
  }
  Image_3 {
   widgetClassPath=&Image
   bIsVariable=true
   parent=VerticalBox_2
   widget=Image_3
   properties {
    brush {
     bIsDynamicallyLoaded=false
     drawAs=Image
     imageType=NoImage
     mirroring=NoMirror
     resourceName=None
     resourceObject=null
     tiling=NoTile
     imageSize=<32 2>
     margin=<ltrb 0 0 0 0>
     outlineSettings {
      bUseBrushTransparency=false
      roundingType=HalfHeightRadius
      width=0
      color {colorUseRule=UseColor_Specified, specifiedColor=<rgba 0 0 0 0>}
      cornerRadii=<0 0 0 0>
     }
     tintColor {colorUseRule=UseColor_Specified, specifiedColor=<rgba 1 1 1 1>}
     uVRegion {bIsValid=false, max=<0 0>, min=<0 0>}
    }
    colorAndOpacity=<rgba 1 1 1 0.25>
   }
   slot {class=VerticalBoxSlot, properties {padding=<ltrb 0 5 0 5>}}
  }
  HorizontalBox_431 {widgetClassPath=&HorizontalBox, parent=VerticalBox_2, widget=HorizontalBox_431, slot {class=VerticalBoxSlot}}
  Border_6 {
   widgetClassPath=&Border
   parent=HorizontalBox_431
   widget=Border_6
   properties {brushColor=<rgba 0.0452 0.0627 0.0885 1>, padding=<ltrb 3 3 3 3>}
   slot {class=HorizontalBoxSlot, properties {padding=<ltrb 10 0 0 0>}}
  }
  HorizontalBox_6 {
   widgetClassPath=&HorizontalBox
   parent=Border_6
   widget=HorizontalBox_6
   slot {class=BorderSlot, properties {padding=<ltrb 3 3 3 3>}}
  }
  txtRepeat_3 {
   widgetClassPath=&CommonTextBlock
   parent=HorizontalBox_6
   widget=txtRepeat_3
   properties {
    text=Show Frustum:
    font {
     bForceMonospaced=false
     bMaterialIsStencil=false
     fontMaterial=null
     fontObject=&NotoSans
     letterSpacing=0
     monospacedWidth=1
     size=12
     skewAmount=0
     typefaceFontName=NotoSans-Regular
     outlineSettings {
      bApplyOutlineToDropShadows=false
      bMiteredCorners=false
      bSeparateFillAlpha=false
      outlineMaterial=null
      outlineSize=0
      outlineColor=<rgba 0 0 0 1>
     }
    }
    shadowOffset=<0 0>
   }
   slot {class=HorizontalBoxSlot, properties {verticalAlignment=VAlign_Center}}
  }
  cbFrustumRender {
   widgetClassPath=&CheckBox
   bIsVariable=true
   parent=HorizontalBox_6
   widget=cbFrustumRender
   properties {
    toolTipText="If True, show the frustum of the current data point in world"
    widgetStyle {
     checkBoxType=CheckBox
     backgroundHoveredImage {
      bIsDynamicallyLoaded=false
      drawAs=RoundedBox
      imageType=NoImage
      mirroring=NoMirror
      resourceName=None
      resourceObject=null
      tiling=NoTile
      imageSize=<18 18>
      margin=<ltrb 0 0 0 0>
      outlineSettings {
       bUseBrushTransparency=false
       roundingType=FixedRadius
       width=1
       color {colorUseRule=UseColor_ColorTable, specifiedColor=<rgba 0 0 0 0>}
       cornerRadii=<3 3 3 3>
      }
      tintColor {colorUseRule=UseColor_ColorTable, specifiedColor=<rgba 0 0 0 0>}
      uVRegion {bIsValid=false, max=<0 0>, min=<0 0>}
     }
     backgroundImage {
      bIsDynamicallyLoaded=false
      drawAs=RoundedBox
      imageType=NoImage
      mirroring=NoMirror
      resourceName=None
      resourceObject=null
      tiling=NoTile
      imageSize=<18 18>
      margin=<ltrb 0 0 0 0>
      outlineSettings {
       bUseBrushTransparency=false
       roundingType=FixedRadius
       width=1
       color {colorUseRule=UseColor_ColorTable, specifiedColor=<rgba 0 0 0 0>}
       cornerRadii=<3 3 3 3>
      }
      tintColor {colorUseRule=UseColor_ColorTable, specifiedColor=<rgba 0 0 0 0>}
      uVRegion {bIsValid=false, max=<0 0>, min=<0 0>}
     }
     backgroundPressedImage {
      bIsDynamicallyLoaded=false
      drawAs=RoundedBox
      imageType=NoImage
      mirroring=NoMirror
      resourceName=None
      resourceObject=null
      tiling=NoTile
      imageSize=<18 18>
      margin=<ltrb 0 0 0 0>
      outlineSettings {
       bUseBrushTransparency=false
       roundingType=FixedRadius
       width=1
       color {colorUseRule=UseColor_ColorTable, specifiedColor=<rgba 0 0 0 0>}
       cornerRadii=<3 3 3 3>
      }
      tintColor {colorUseRule=UseColor_ColorTable, specifiedColor=<rgba 0 0 0 0>}
      uVRegion {bIsValid=false, max=<0 0>, min=<0 0>}
     }
     borderBackgroundColor {colorUseRule=UseColor_Specified, specifiedColor=<rgba 1 1 1 1>}
     checkedForeground {colorUseRule=UseColor_ColorTable, specifiedColor=<rgba 0 0 0 0>}
     checkedHoveredForeground {colorUseRule=UseColor_ColorTable, specifiedColor=<rgba 0 0 0 0>}
     checkedHoveredImage {
      bIsDynamicallyLoaded=false
      drawAs=Image
      imageType=Vector
      mirroring=NoMirror
      resourceName=../../../Engine/Content/Slate/Starship/CoreWidgets/CheckBox/check.svg
      resourceObject=null
      tiling=NoTile
      imageSize=<16 16>
      margin=<ltrb 0 0 0 0>
      outlineSettings {
       bUseBrushTransparency=false
       roundingType=HalfHeightRadius
       width=0
       color {colorUseRule=UseColor_Specified, specifiedColor=<rgba 0 0 0 0>}
       cornerRadii=<0 0 0 0>
      }
      tintColor {colorUseRule=UseColor_Specified, specifiedColor=<rgba 0.159 0.6445 1 1>}
      uVRegion {bIsValid=false, max=<0 0>, min=<0 0>}
     }
     checkedImage {
      bIsDynamicallyLoaded=false
      drawAs=Image
      imageType=Vector
      mirroring=NoMirror
      resourceName=../../../Engine/Content/Slate/Starship/CoreWidgets/CheckBox/check.svg
      resourceObject=null
      tiling=NoTile
      imageSize=<16 16>
      margin=<ltrb 0 0 0 0>
      outlineSettings {
       bUseBrushTransparency=false
       roundingType=HalfHeightRadius
       width=0
       color {colorUseRule=UseColor_Specified, specifiedColor=<rgba 0 0 0 0>}
       cornerRadii=<0 0 0 0>
      }
      tintColor {colorUseRule=UseColor_Specified, specifiedColor=<rgba 0.0194 0.4969 1 1>}
      uVRegion {bIsValid=false, max=<0 0>, min=<0 0>}
     }
     checkedPressedForeground {colorUseRule=UseColor_ColorTable, specifiedColor=<rgba 0 0 0 0>}
     checkedPressedImage {
      bIsDynamicallyLoaded=false
      drawAs=Image
      imageType=Vector
      mirroring=NoMirror
      resourceName=../../../Engine/Content/Slate/Starship/CoreWidgets/CheckBox/check.svg
      resourceObject=null
      tiling=NoTile
      imageSize=<16 16>
      margin=<ltrb 0 0 0 0>
      outlineSettings {
       bUseBrushTransparency=false
       roundingType=HalfHeightRadius
       width=0
       color {colorUseRule=UseColor_Specified, specifiedColor=<rgba 0 0 0 0>}
       cornerRadii=<0 0 0 0>
      }
      tintColor {colorUseRule=UseColor_Specified, specifiedColor=<rgba 0.0194 0.4969 1 1>}
      uVRegion {bIsValid=false, max=<0 0>, min=<0 0>}
     }
     checkedSlateSound {resourceObject=null}
     foregroundColor {colorUseRule=UseColor_Specified, specifiedColor=<rgba 0.5271 0.5271 0.5271 1>}
     hoveredForeground {colorUseRule=UseColor_ColorTable, specifiedColor=<rgba 0 0 0 0>}
     hoveredSlateSound {resourceObject=null}
     padding=<ltrb 2 0 0 0>
     pressedForeground {colorUseRule=UseColor_ColorTable, specifiedColor=<rgba 0 0 0 0>}
     uncheckedHoveredImage {
      bIsDynamicallyLoaded=false
      drawAs=NoDrawType
      imageType=NoImage
      mirroring=NoMirror
      resourceName=None
      resourceObject=null
      tiling=NoTile
      imageSize=<0 0>
      margin=<ltrb 0 0 0 0>
      outlineSettings {
       bUseBrushTransparency=false
       roundingType=HalfHeightRadius
       width=0
       color {colorUseRule=UseColor_Specified, specifiedColor=<rgba 0 0 0 0>}
       cornerRadii=<0 0 0 0>
      }
      tintColor {colorUseRule=UseColor_Specified, specifiedColor=<rgba 1 1 1 1>}
      uVRegion {bIsValid=false, max=<0 0>, min=<0 0>}
     }
     uncheckedImage {
      bIsDynamicallyLoaded=false
      drawAs=NoDrawType
      imageType=NoImage
      mirroring=NoMirror
      resourceName=None
      resourceObject=null
      tiling=NoTile
      imageSize=<0 0>
      margin=<ltrb 0 0 0 0>
      outlineSettings {
       bUseBrushTransparency=false
       roundingType=HalfHeightRadius
       width=0
       color {colorUseRule=UseColor_Specified, specifiedColor=<rgba 0 0 0 0>}
       cornerRadii=<0 0 0 0>
      }
      tintColor {colorUseRule=UseColor_Specified, specifiedColor=<rgba 1 1 1 1>}
      uVRegion {bIsValid=false, max=<0 0>, min=<0 0>}
     }
     uncheckedPressedImage {
      bIsDynamicallyLoaded=false
      drawAs=NoDrawType
      imageType=NoImage
      mirroring=NoMirror
      resourceName=None
      resourceObject=null
      tiling=NoTile
      imageSize=<0 0>
      margin=<ltrb 0 0 0 0>
      outlineSettings {
       bUseBrushTransparency=false
       roundingType=HalfHeightRadius
       width=0
       color {colorUseRule=UseColor_Specified, specifiedColor=<rgba 0 0 0 0>}
       cornerRadii=<0 0 0 0>
      }
      tintColor {colorUseRule=UseColor_Specified, specifiedColor=<rgba 1 1 1 1>}
      uVRegion {bIsValid=false, max=<0 0>, min=<0 0>}
     }
     uncheckedSlateSound {resourceObject=null}
     undeterminedForeground {colorUseRule=UseColor_ColorTable, specifiedColor=<rgba 0 0 0 0>}
     undeterminedHoveredImage {
      bIsDynamicallyLoaded=false
      drawAs=Image
      imageType=Vector
      mirroring=NoMirror
      resourceName=../../../Engine/Content/Slate/Starship/CoreWidgets/CheckBox/indeterminate.svg
      resourceObject=null
      tiling=NoTile
      imageSize=<16 16>
      margin=<ltrb 0 0 0 0>
      outlineSettings {
       bUseBrushTransparency=false
       roundingType=HalfHeightRadius
       width=0
       color {colorUseRule=UseColor_Specified, specifiedColor=<rgba 0 0 0 0>}
       cornerRadii=<0 0 0 0>
      }
      tintColor {colorUseRule=UseColor_Specified, specifiedColor=<rgba 0.159 0.6445 1 1>}
      uVRegion {bIsValid=false, max=<0 0>, min=<0 0>}
     }
     undeterminedImage {
      bIsDynamicallyLoaded=false
      drawAs=Image
      imageType=Vector
      mirroring=NoMirror
      resourceName=../../../Engine/Content/Slate/Starship/CoreWidgets/CheckBox/indeterminate.svg
      resourceObject=null
      tiling=NoTile
      imageSize=<16 16>
      margin=<ltrb 0 0 0 0>
      outlineSettings {
       bUseBrushTransparency=false
       roundingType=HalfHeightRadius
       width=0
       color {colorUseRule=UseColor_Specified, specifiedColor=<rgba 0 0 0 0>}
       cornerRadii=<0 0 0 0>
      }
      tintColor {colorUseRule=UseColor_Specified, specifiedColor=<rgba 0.0194 0.4969 1 1>}
      uVRegion {bIsValid=false, max=<0 0>, min=<0 0>}
     }
     undeterminedPressedImage {
      bIsDynamicallyLoaded=false
      drawAs=Image
      imageType=Vector
      mirroring=NoMirror
      resourceName=../../../Engine/Content/Slate/Starship/CoreWidgets/CheckBox/indeterminate.svg
      resourceObject=null
      tiling=NoTile
      imageSize=<16 16>
      margin=<ltrb 0 0 0 0>
      outlineSettings {
       bUseBrushTransparency=false
       roundingType=HalfHeightRadius
       width=0
       color {colorUseRule=UseColor_Specified, specifiedColor=<rgba 0 0 0 0>}
       cornerRadii=<0 0 0 0>
      }
      tintColor {colorUseRule=UseColor_Specified, specifiedColor=<rgba 0.0194 0.4969 1 1>}
      uVRegion {bIsValid=false, max=<0 0>, min=<0 0>}
     }
    }
   }
   slot {class=HorizontalBoxSlot, properties {horizontalAlignment=HAlign_Left, verticalAlignment=VAlign_Center, padding=<ltrb 5 0 0 0>}}
  }
  Border_12 {
   widgetClassPath=&Border
   parent=HorizontalBox_431
   widget=Border_12
   properties {brushColor=<rgba 0.0452 0.0627 0.0885 1>, padding=<ltrb 3 3 3 3>}
   slot {class=HorizontalBoxSlot, properties {padding=<ltrb 10 0 0 0>}}
  }
  HorizontalBox_10 {
   widgetClassPath=&HorizontalBox
   parent=Border_12
   widget=HorizontalBox_10
   slot {class=BorderSlot, properties {padding=<ltrb 3 3 3 3>}}
  }
  txtRepeat_4 {
   widgetClassPath=&CommonTextBlock
   parent=HorizontalBox_10
   widget=txtRepeat_4
   properties {
    text=Only For Current:
    font {
     bForceMonospaced=false
     bMaterialIsStencil=false
     fontMaterial=null
     fontObject=&NotoSans
     letterSpacing=0
     monospacedWidth=1
     size=12
     skewAmount=0
     typefaceFontName=NotoSans-Regular
     outlineSettings {
      bApplyOutlineToDropShadows=false
      bMiteredCorners=false
      bSeparateFillAlpha=false
      outlineMaterial=null
      outlineSize=0
      outlineColor=<rgba 0 0 0 1>
     }
    }
    shadowOffset=<0 0>
   }
   slot {class=HorizontalBoxSlot, properties {verticalAlignment=VAlign_Center}}
  }
  cbFrustumOnlyCurrent {
   widgetClassPath=&CheckBox
   bIsVariable=true
   parent=HorizontalBox_10
   widget=cbFrustumOnlyCurrent
   properties {
    checkedState=Checked
    toolTipText="If True, show the frustum only for the current data point.  If False, draw all frustums for visible data.  Use the threshold value to show less points and the Frustum Opacity to fade them in/out as they stack"
    widgetStyle {
     checkBoxType=CheckBox
     backgroundHoveredImage {
      bIsDynamicallyLoaded=false
      drawAs=RoundedBox
      imageType=NoImage
      mirroring=NoMirror
      resourceName=None
      resourceObject=null
      tiling=NoTile
      imageSize=<18 18>
      margin=<ltrb 0 0 0 0>
      outlineSettings {
       bUseBrushTransparency=false
       roundingType=FixedRadius
       width=1
       color {colorUseRule=UseColor_ColorTable, specifiedColor=<rgba 0 0 0 0>}
       cornerRadii=<3 3 3 3>
      }
      tintColor {colorUseRule=UseColor_ColorTable, specifiedColor=<rgba 0 0 0 0>}
      uVRegion {bIsValid=false, max=<0 0>, min=<0 0>}
     }
     backgroundImage {
      bIsDynamicallyLoaded=false
      drawAs=RoundedBox
      imageType=NoImage
      mirroring=NoMirror
      resourceName=None
      resourceObject=null
      tiling=NoTile
      imageSize=<18 18>
      margin=<ltrb 0 0 0 0>
      outlineSettings {
       bUseBrushTransparency=false
       roundingType=FixedRadius
       width=1
       color {colorUseRule=UseColor_ColorTable, specifiedColor=<rgba 0 0 0 0>}
       cornerRadii=<3 3 3 3>
      }
      tintColor {colorUseRule=UseColor_ColorTable, specifiedColor=<rgba 0 0 0 0>}
      uVRegion {bIsValid=false, max=<0 0>, min=<0 0>}
     }
     backgroundPressedImage {
      bIsDynamicallyLoaded=false
      drawAs=RoundedBox
      imageType=NoImage
      mirroring=NoMirror
      resourceName=None
      resourceObject=null
      tiling=NoTile
      imageSize=<18 18>
      margin=<ltrb 0 0 0 0>
      outlineSettings {
       bUseBrushTransparency=false
       roundingType=FixedRadius
       width=1
       color {colorUseRule=UseColor_ColorTable, specifiedColor=<rgba 0 0 0 0>}
       cornerRadii=<3 3 3 3>
      }
      tintColor {colorUseRule=UseColor_ColorTable, specifiedColor=<rgba 0 0 0 0>}
      uVRegion {bIsValid=false, max=<0 0>, min=<0 0>}
     }
     borderBackgroundColor {colorUseRule=UseColor_Specified, specifiedColor=<rgba 1 1 1 1>}
     checkedForeground {colorUseRule=UseColor_ColorTable, specifiedColor=<rgba 0 0 0 0>}
     checkedHoveredForeground {colorUseRule=UseColor_ColorTable, specifiedColor=<rgba 0 0 0 0>}
     checkedHoveredImage {
      bIsDynamicallyLoaded=false
      drawAs=Image
      imageType=Vector
      mirroring=NoMirror
      resourceName=../../../Engine/Content/Slate/Starship/CoreWidgets/CheckBox/check.svg
      resourceObject=null
      tiling=NoTile
      imageSize=<16 16>
      margin=<ltrb 0 0 0 0>
      outlineSettings {
       bUseBrushTransparency=false
       roundingType=HalfHeightRadius
       width=0
       color {colorUseRule=UseColor_Specified, specifiedColor=<rgba 0 0 0 0>}
       cornerRadii=<0 0 0 0>
      }
      tintColor {colorUseRule=UseColor_Specified, specifiedColor=<rgba 0.159 0.6445 1 1>}
      uVRegion {bIsValid=false, max=<0 0>, min=<0 0>}
     }
     checkedImage {
      bIsDynamicallyLoaded=false
      drawAs=Image
      imageType=Vector
      mirroring=NoMirror
      resourceName=../../../Engine/Content/Slate/Starship/CoreWidgets/CheckBox/check.svg
      resourceObject=null
      tiling=NoTile
      imageSize=<16 16>
      margin=<ltrb 0 0 0 0>
      outlineSettings {
       bUseBrushTransparency=false
       roundingType=HalfHeightRadius
       width=0
       color {colorUseRule=UseColor_Specified, specifiedColor=<rgba 0 0 0 0>}
       cornerRadii=<0 0 0 0>
      }
      tintColor {colorUseRule=UseColor_Specified, specifiedColor=<rgba 0.0194 0.4969 1 1>}
      uVRegion {bIsValid=false, max=<0 0>, min=<0 0>}
     }
     checkedPressedForeground {colorUseRule=UseColor_ColorTable, specifiedColor=<rgba 0 0 0 0>}
     checkedPressedImage {
      bIsDynamicallyLoaded=false
      drawAs=Image
      imageType=Vector
      mirroring=NoMirror
      resourceName=../../../Engine/Content/Slate/Starship/CoreWidgets/CheckBox/check.svg
      resourceObject=null
      tiling=NoTile
      imageSize=<16 16>
      margin=<ltrb 0 0 0 0>
      outlineSettings {
       bUseBrushTransparency=false
       roundingType=HalfHeightRadius
       width=0
       color {colorUseRule=UseColor_Specified, specifiedColor=<rgba 0 0 0 0>}
       cornerRadii=<0 0 0 0>
      }
      tintColor {colorUseRule=UseColor_Specified, specifiedColor=<rgba 0.0194 0.4969 1 1>}
      uVRegion {bIsValid=false, max=<0 0>, min=<0 0>}
     }
     checkedSlateSound {resourceObject=null}
     foregroundColor {colorUseRule=UseColor_Specified, specifiedColor=<rgba 0.5271 0.5271 0.5271 1>}
     hoveredForeground {colorUseRule=UseColor_ColorTable, specifiedColor=<rgba 0 0 0 0>}
     hoveredSlateSound {resourceObject=null}
     padding=<ltrb 2 0 0 0>
     pressedForeground {colorUseRule=UseColor_ColorTable, specifiedColor=<rgba 0 0 0 0>}
     uncheckedHoveredImage {
      bIsDynamicallyLoaded=false
      drawAs=NoDrawType
      imageType=NoImage
      mirroring=NoMirror
      resourceName=None
      resourceObject=null
      tiling=NoTile
      imageSize=<0 0>
      margin=<ltrb 0 0 0 0>
      outlineSettings {
       bUseBrushTransparency=false
       roundingType=HalfHeightRadius
       width=0
       color {colorUseRule=UseColor_Specified, specifiedColor=<rgba 0 0 0 0>}
       cornerRadii=<0 0 0 0>
      }
      tintColor {colorUseRule=UseColor_Specified, specifiedColor=<rgba 1 1 1 1>}
      uVRegion {bIsValid=false, max=<0 0>, min=<0 0>}
     }
     uncheckedImage {
      bIsDynamicallyLoaded=false
      drawAs=NoDrawType
      imageType=NoImage
      mirroring=NoMirror
      resourceName=None
      resourceObject=null
      tiling=NoTile
      imageSize=<0 0>
      margin=<ltrb 0 0 0 0>
      outlineSettings {
       bUseBrushTransparency=false
       roundingType=HalfHeightRadius
       width=0
       color {colorUseRule=UseColor_Specified, specifiedColor=<rgba 0 0 0 0>}
       cornerRadii=<0 0 0 0>
      }
      tintColor {colorUseRule=UseColor_Specified, specifiedColor=<rgba 1 1 1 1>}
      uVRegion {bIsValid=false, max=<0 0>, min=<0 0>}
     }
     uncheckedPressedImage {
      bIsDynamicallyLoaded=false
      drawAs=NoDrawType
      imageType=NoImage
      mirroring=NoMirror
      resourceName=None
      resourceObject=null
      tiling=NoTile
      imageSize=<0 0>
      margin=<ltrb 0 0 0 0>
      outlineSettings {
       bUseBrushTransparency=false
       roundingType=HalfHeightRadius
       width=0
       color {colorUseRule=UseColor_Specified, specifiedColor=<rgba 0 0 0 0>}
       cornerRadii=<0 0 0 0>
      }
      tintColor {colorUseRule=UseColor_Specified, specifiedColor=<rgba 1 1 1 1>}
      uVRegion {bIsValid=false, max=<0 0>, min=<0 0>}
     }
     uncheckedSlateSound {resourceObject=null}
     undeterminedForeground {colorUseRule=UseColor_ColorTable, specifiedColor=<rgba 0 0 0 0>}
     undeterminedHoveredImage {
      bIsDynamicallyLoaded=false
      drawAs=Image
      imageType=Vector
      mirroring=NoMirror
      resourceName=../../../Engine/Content/Slate/Starship/CoreWidgets/CheckBox/indeterminate.svg
      resourceObject=null
      tiling=NoTile
      imageSize=<16 16>
      margin=<ltrb 0 0 0 0>
      outlineSettings {
       bUseBrushTransparency=false
       roundingType=HalfHeightRadius
       width=0
       color {colorUseRule=UseColor_Specified, specifiedColor=<rgba 0 0 0 0>}
       cornerRadii=<0 0 0 0>
      }
      tintColor {colorUseRule=UseColor_Specified, specifiedColor=<rgba 0.159 0.6445 1 1>}
      uVRegion {bIsValid=false, max=<0 0>, min=<0 0>}
     }
     undeterminedImage {
      bIsDynamicallyLoaded=false
      drawAs=Image
      imageType=Vector
      mirroring=NoMirror
      resourceName=../../../Engine/Content/Slate/Starship/CoreWidgets/CheckBox/indeterminate.svg
      resourceObject=null
      tiling=NoTile
      imageSize=<16 16>
      margin=<ltrb 0 0 0 0>
      outlineSettings {
       bUseBrushTransparency=false
       roundingType=HalfHeightRadius
       width=0
       color {colorUseRule=UseColor_Specified, specifiedColor=<rgba 0 0 0 0>}
       cornerRadii=<0 0 0 0>
      }
      tintColor {colorUseRule=UseColor_Specified, specifiedColor=<rgba 0.0194 0.4969 1 1>}
      uVRegion {bIsValid=false, max=<0 0>, min=<0 0>}
     }
     undeterminedPressedImage {
      bIsDynamicallyLoaded=false
      drawAs=Image
      imageType=Vector
      mirroring=NoMirror
      resourceName=../../../Engine/Content/Slate/Starship/CoreWidgets/CheckBox/indeterminate.svg
      resourceObject=null
      tiling=NoTile
      imageSize=<16 16>
      margin=<ltrb 0 0 0 0>
      outlineSettings {
       bUseBrushTransparency=false
       roundingType=HalfHeightRadius
       width=0
       color {colorUseRule=UseColor_Specified, specifiedColor=<rgba 0 0 0 0>}
       cornerRadii=<0 0 0 0>
      }
      tintColor {colorUseRule=UseColor_Specified, specifiedColor=<rgba 0.0194 0.4969 1 1>}
      uVRegion {bIsValid=false, max=<0 0>, min=<0 0>}
     }
    }
   }
   slot {class=HorizontalBoxSlot, properties {horizontalAlignment=HAlign_Left, verticalAlignment=VAlign_Center, padding=<ltrb 5 0 0 0>}}
  }
  Border_9 {
   widgetClassPath=&Border
   parent=HorizontalBox_431
   widget=Border_9
   properties {brushColor=<rgba 0.0452 0.0627 0.0885 1>, padding=<ltrb 3 3 3 3>}
   slot {class=HorizontalBoxSlot, properties {verticalAlignment=VAlign_Center, padding=<ltrb 10 0 0 0>}}
  }
  HorizontalBox_8 {
   widgetClassPath=&HorizontalBox
   parent=Border_9
   widget=HorizontalBox_8
   slot {class=BorderSlot, properties {padding=<ltrb 3 3 3 3>}}
  }
  textField_1 {
   widgetClassPath=&CommonTextBlock
   parent=HorizontalBox_8
   widget=textField_1
   properties {
    text=Frustum Scale:
    font {
     bForceMonospaced=false
     bMaterialIsStencil=false
     fontMaterial=null
     fontObject=&NotoSans
     letterSpacing=0
     monospacedWidth=1
     size=12
     skewAmount=0
     typefaceFontName=NotoSans-Regular
     outlineSettings {
      bApplyOutlineToDropShadows=false
      bMiteredCorners=false
      bSeparateFillAlpha=false
      outlineMaterial=null
      outlineSize=0
      outlineColor=<rgba 0 0 0 1>
     }
    }
    shadowOffset=<0 0>
   }
   slot {class=HorizontalBoxSlot, properties {verticalAlignment=VAlign_Center}}
  }
  spinFrustumScale {
   widgetClassPath=&SpinBox
   bIsVariable=true
   parent=HorizontalBox_8
   widget=spinFrustumScale
   properties {
    bAlwaysUsesDeltaSnap=true
    bOverride_MaxSliderValue=true
    bOverride_MaxValue=true
    bOverride_MinSliderValue=true
    bOverride_MinValue=true
    delta=0.01
    maxFractionalDigits=2
    maxSliderValue=5
    maxValue=5
    minFractionalDigits=2
    minSliderValue=0.1
    minValue=0.1
    toolTipText="Scales the Frustum's width/height by this value.  1 = Full width/height for 60 FOV Game."
    value=0.5
    foregroundColor {colorUseRule=UseColor_Specified, specifiedColor=<rgba 1 1 1 1>}
    widgetStyle {
     activeBackgroundBrush {
      bIsDynamicallyLoaded=false
      drawAs=RoundedBox
      imageType=NoImage
      mirroring=NoMirror
      resourceName=None
      resourceObject=null
      tiling=NoTile
      imageSize=<0 0>
      margin=<ltrb 0 0 0 0>
      outlineSettings {
       bUseBrushTransparency=false
       roundingType=FixedRadius
       width=1
       color {colorUseRule=UseColor_ColorTable, specifiedColor=<rgba 0 0 0 0>}
       cornerRadii=<4 4 4 4>
      }
      tintColor {colorUseRule=UseColor_Specified, specifiedColor=<rgba 0.0048 0.0048 0.0048 1>}
      uVRegion {bIsValid=false, max=<0 0>, min=<0 0>}
     }
     activeFillBrush {
      bIsDynamicallyLoaded=false
      drawAs=RoundedBox
      imageType=NoImage
      mirroring=NoMirror
      resourceName=None
      resourceObject=null
      tiling=NoTile
      imageSize=<0 0>
      margin=<ltrb 0 0 0 0>
      outlineSettings {
       bUseBrushTransparency=false
       roundingType=FixedRadius
       width=1
       color {colorUseRule=UseColor_Specified, specifiedColor=<rgba 0 0 0 0>}
       cornerRadii=<4 4 4 4>
      }
      tintColor {colorUseRule=UseColor_Specified, specifiedColor=<rgba 0.0953 0.0953 0.0953 1>}
      uVRegion {bIsValid=false, max=<0 0>, min=<0 0>}
     }
     arrowsImage {
      bIsDynamicallyLoaded=false
      drawAs=NoDrawType
      imageType=NoImage
      mirroring=NoMirror
      resourceName=None
      resourceObject=null
      tiling=NoTile
      imageSize=<0 0>
      margin=<ltrb 0 0 0 0>
      outlineSettings {
       bUseBrushTransparency=false
       roundingType=HalfHeightRadius
       width=0
       color {colorUseRule=UseColor_Specified, specifiedColor=<rgba 0 0 0 0>}
       cornerRadii=<0 0 0 0>
      }
      tintColor {colorUseRule=UseColor_Specified, specifiedColor=<rgba 1 1 1 1>}
      uVRegion {bIsValid=false, max=<0 0>, min=<0 0>}
     }
     backgroundBrush {
      bIsDynamicallyLoaded=false
      drawAs=RoundedBox
      imageType=NoImage
      mirroring=NoMirror
      resourceName=None
      resourceObject=null
      tiling=NoTile
      imageSize=<0 0>
      margin=<ltrb 0 0 0 0>
      outlineSettings {
       bUseBrushTransparency=false
       roundingType=FixedRadius
       width=1
       color {colorUseRule=UseColor_ColorTable, specifiedColor=<rgba 0 0 0 0>}
       cornerRadii=<4 4 4 4>
      }
      tintColor {colorUseRule=UseColor_Specified, specifiedColor=<rgba 0.0048 0.0048 0.0048 1>}
      uVRegion {bIsValid=false, max=<0 0>, min=<0 0>}
     }
     foregroundColor {colorUseRule=UseColor_Specified, specifiedColor=<rgba 1 1 1 1>}
     hoveredBackgroundBrush {
      bIsDynamicallyLoaded=false
      drawAs=RoundedBox
      imageType=NoImage
      mirroring=NoMirror
      resourceName=None
      resourceObject=null
      tiling=NoTile
      imageSize=<0 0>
      margin=<ltrb 0 0 0 0>
      outlineSettings {
       bUseBrushTransparency=false
       roundingType=FixedRadius
       width=1
       color {colorUseRule=UseColor_ColorTable, specifiedColor=<rgba 0 0 0 0>}
       cornerRadii=<4 4 4 4>
      }
      tintColor {colorUseRule=UseColor_Specified, specifiedColor=<rgba 0.0048 0.0048 0.0048 1>}
      uVRegion {bIsValid=false, max=<0 0>, min=<0 0>}
     }
     hoveredFillBrush {
      bIsDynamicallyLoaded=false
      drawAs=RoundedBox
      imageType=NoImage
      mirroring=NoMirror
      resourceName=None
      resourceObject=null
      tiling=NoTile
      imageSize=<0 0>
      margin=<ltrb 0 0 0 0>
      outlineSettings {
       bUseBrushTransparency=false
       roundingType=FixedRadius
       width=1
       color {colorUseRule=UseColor_Specified, specifiedColor=<rgba 0 0 0 0>}
       cornerRadii=<4 4 4 4>
      }
      tintColor {colorUseRule=UseColor_Specified, specifiedColor=<rgba 0.0953 0.0953 0.0953 1>}
      uVRegion {bIsValid=false, max=<0 0>, min=<0 0>}
     }
     inactiveFillBrush {
      bIsDynamicallyLoaded=false
      drawAs=RoundedBox
      imageType=NoImage
      mirroring=NoMirror
      resourceName=None
      resourceObject=null
      tiling=NoTile
      imageSize=<0 0>
      margin=<ltrb 0 0 0 0>
      outlineSettings {
       bUseBrushTransparency=false
       roundingType=FixedRadius
       width=1
       color {colorUseRule=UseColor_Specified, specifiedColor=<rgba 0 0 0 0>}
       cornerRadii=<4 4 4 4>
      }
      tintColor {colorUseRule=UseColor_Specified, specifiedColor=<rgba 0.0395 0.0395 0.0395 1>}
      uVRegion {bIsValid=false, max=<0 0>, min=<0 0>}
     }
     insetPadding=<ltrb 1 1 1 1>
     textPadding=<ltrb 10 3.5 10 4>
    }
   }
   slot {class=HorizontalBoxSlot, properties {padding=<ltrb 5 0 0 0>}}
  }
  Border_10 {
   widgetClassPath=&Border
   parent=HorizontalBox_431
   widget=Border_10
   properties {brushColor=<rgba 0.0452 0.0627 0.0885 1>, padding=<ltrb 3 3 3 3>}
   slot {class=HorizontalBoxSlot, properties {verticalAlignment=VAlign_Center, padding=<ltrb 10 0 0 0>}}
  }
  HorizontalBox_9 {
   widgetClassPath=&HorizontalBox
   parent=Border_10
   widget=HorizontalBox_9
   slot {class=BorderSlot, properties {padding=<ltrb 3 3 3 3>}}
  }
  textField_2 {
   widgetClassPath=&CommonTextBlock
   parent=HorizontalBox_9
   widget=textField_2
   properties {
    text=Frustum Opacity:
    font {
     bForceMonospaced=false
     bMaterialIsStencil=false
     fontMaterial=null
     fontObject=&NotoSans
     letterSpacing=0
     monospacedWidth=1
     size=12
     skewAmount=0
     typefaceFontName=NotoSans-Regular
     outlineSettings {
      bApplyOutlineToDropShadows=false
      bMiteredCorners=false
      bSeparateFillAlpha=false
      outlineMaterial=null
      outlineSize=0
      outlineColor=<rgba 0 0 0 1>
     }
    }
    shadowOffset=<0 0>
   }
   slot {class=HorizontalBoxSlot, properties {verticalAlignment=VAlign_Center}}
  }
  spinFrustumOpacity {
   widgetClassPath=&SpinBox
   bIsVariable=true
   parent=HorizontalBox_9
   widget=spinFrustumOpacity
   properties {
    bAlwaysUsesDeltaSnap=true
    bOverride_MaxSliderValue=true
    bOverride_MaxValue=true
    bOverride_MinSliderValue=true
    bOverride_MinValue=true
    delta=0.001
    maxFractionalDigits=3
    maxSliderValue=1
    maxValue=1
    minFractionalDigits=3
    toolTipText=Opacity of the Frustum render.
    value=0.1
    foregroundColor {colorUseRule=UseColor_Specified, specifiedColor=<rgba 1 1 1 1>}
    widgetStyle {
     activeBackgroundBrush {
      bIsDynamicallyLoaded=false
      drawAs=RoundedBox
      imageType=NoImage
      mirroring=NoMirror
      resourceName=None
      resourceObject=null
      tiling=NoTile
      imageSize=<0 0>
      margin=<ltrb 0 0 0 0>
      outlineSettings {
       bUseBrushTransparency=false
       roundingType=FixedRadius
       width=1
       color {colorUseRule=UseColor_ColorTable, specifiedColor=<rgba 0 0 0 0>}
       cornerRadii=<4 4 4 4>
      }
      tintColor {colorUseRule=UseColor_Specified, specifiedColor=<rgba 0.0048 0.0048 0.0048 1>}
      uVRegion {bIsValid=false, max=<0 0>, min=<0 0>}
     }
     activeFillBrush {
      bIsDynamicallyLoaded=false
      drawAs=RoundedBox
      imageType=NoImage
      mirroring=NoMirror
      resourceName=None
      resourceObject=null
      tiling=NoTile
      imageSize=<0 0>
      margin=<ltrb 0 0 0 0>
      outlineSettings {
       bUseBrushTransparency=false
       roundingType=FixedRadius
       width=1
       color {colorUseRule=UseColor_Specified, specifiedColor=<rgba 0 0 0 0>}
       cornerRadii=<4 4 4 4>
      }
      tintColor {colorUseRule=UseColor_Specified, specifiedColor=<rgba 0.0953 0.0953 0.0953 1>}
      uVRegion {bIsValid=false, max=<0 0>, min=<0 0>}
     }
     arrowsImage {
      bIsDynamicallyLoaded=false
      drawAs=NoDrawType
      imageType=NoImage
      mirroring=NoMirror
      resourceName=None
      resourceObject=null
      tiling=NoTile
      imageSize=<0 0>
      margin=<ltrb 0 0 0 0>
      outlineSettings {
       bUseBrushTransparency=false
       roundingType=HalfHeightRadius
       width=0
       color {colorUseRule=UseColor_Specified, specifiedColor=<rgba 0 0 0 0>}
       cornerRadii=<0 0 0 0>
      }
      tintColor {colorUseRule=UseColor_Specified, specifiedColor=<rgba 1 1 1 1>}
      uVRegion {bIsValid=false, max=<0 0>, min=<0 0>}
     }
     backgroundBrush {
      bIsDynamicallyLoaded=false
      drawAs=RoundedBox
      imageType=NoImage
      mirroring=NoMirror
      resourceName=None
      resourceObject=null
      tiling=NoTile
      imageSize=<0 0>
      margin=<ltrb 0 0 0 0>
      outlineSettings {
       bUseBrushTransparency=false
       roundingType=FixedRadius
       width=1
       color {colorUseRule=UseColor_ColorTable, specifiedColor=<rgba 0 0 0 0>}
       cornerRadii=<4 4 4 4>
      }
      tintColor {colorUseRule=UseColor_Specified, specifiedColor=<rgba 0.0048 0.0048 0.0048 1>}
      uVRegion {bIsValid=false, max=<0 0>, min=<0 0>}
     }
     foregroundColor {colorUseRule=UseColor_Specified, specifiedColor=<rgba 1 1 1 1>}
     hoveredBackgroundBrush {
      bIsDynamicallyLoaded=false
      drawAs=RoundedBox
      imageType=NoImage
      mirroring=NoMirror
      resourceName=None
      resourceObject=null
      tiling=NoTile
      imageSize=<0 0>
      margin=<ltrb 0 0 0 0>
      outlineSettings {
       bUseBrushTransparency=false
       roundingType=FixedRadius
       width=1
       color {colorUseRule=UseColor_ColorTable, specifiedColor=<rgba 0 0 0 0>}
       cornerRadii=<4 4 4 4>
      }
      tintColor {colorUseRule=UseColor_Specified, specifiedColor=<rgba 0.0048 0.0048 0.0048 1>}
      uVRegion {bIsValid=false, max=<0 0>, min=<0 0>}
     }
     hoveredFillBrush {
      bIsDynamicallyLoaded=false
      drawAs=RoundedBox
      imageType=NoImage
      mirroring=NoMirror
      resourceName=None
      resourceObject=null
      tiling=NoTile
      imageSize=<0 0>
      margin=<ltrb 0 0 0 0>
      outlineSettings {
       bUseBrushTransparency=false
       roundingType=FixedRadius
       width=1
       color {colorUseRule=UseColor_Specified, specifiedColor=<rgba 0 0 0 0>}
       cornerRadii=<4 4 4 4>
      }
      tintColor {colorUseRule=UseColor_Specified, specifiedColor=<rgba 0.0953 0.0953 0.0953 1>}
      uVRegion {bIsValid=false, max=<0 0>, min=<0 0>}
     }
     inactiveFillBrush {
      bIsDynamicallyLoaded=false
      drawAs=RoundedBox
      imageType=NoImage
      mirroring=NoMirror
      resourceName=None
      resourceObject=null
      tiling=NoTile
      imageSize=<0 0>
      margin=<ltrb 0 0 0 0>
      outlineSettings {
       bUseBrushTransparency=false
       roundingType=FixedRadius
       width=1
       color {colorUseRule=UseColor_Specified, specifiedColor=<rgba 0 0 0 0>}
       cornerRadii=<4 4 4 4>
      }
      tintColor {colorUseRule=UseColor_Specified, specifiedColor=<rgba 0.0395 0.0395 0.0395 1>}
      uVRegion {bIsValid=false, max=<0 0>, min=<0 0>}
     }
     insetPadding=<ltrb 1 1 1 1>
     textPadding=<ltrb 10 3.5 10 4>
    }
   }
   slot {class=HorizontalBoxSlot, properties {padding=<ltrb 5 0 0 0>}}
  }
  Image_2 {
   widgetClassPath=&Image
   bIsVariable=true
   parent=VerticalBox_0
   widget=Image_2
   properties {
    brush {
     bIsDynamicallyLoaded=false
     drawAs=Image
     imageType=NoImage
     mirroring=NoMirror
     resourceName=None
     resourceObject=null
     tiling=NoTile
     imageSize=<32 2>
     margin=<ltrb 0 0 0 0>
     outlineSettings {
      bUseBrushTransparency=false
      roundingType=HalfHeightRadius
      width=0
      color {colorUseRule=UseColor_Specified, specifiedColor=<rgba 0 0 0 0>}
      cornerRadii=<0 0 0 0>
     }
     tintColor {colorUseRule=UseColor_Specified, specifiedColor=<rgba 1 1 1 1>}
     uVRegion {bIsValid=false, max=<0 0>, min=<0 0>}
    }
    colorAndOpacity=<rgba 1 1 1 0.781>
   }
   slot {class=VerticalBoxSlot, properties {padding=<ltrb 0 0 0 5>}}
  }
  HorizontalBox_11 {
   widgetClassPath=&HorizontalBox
   parent=VerticalBox_0
   widget=HorizontalBox_11
   slot {class=VerticalBoxSlot, properties {padding=<ltrb 10 0 10 0>}}
  }
  text {
   widgetClassPath=&CommonTextBlock
   parent=HorizontalBox_11
   widget=text
   properties {
    justification=Right
    minDesiredWidth=140
    text="CSV Location: "
    visibility=Visible
    font {
     bForceMonospaced=false
     bMaterialIsStencil=false
     fontMaterial=null
     fontObject=&NotoSans
     letterSpacing=0
     monospacedWidth=1
     size=12
     skewAmount=0
     typefaceFontName=NotoSans-Regular
     outlineSettings {
      bApplyOutlineToDropShadows=false
      bMiteredCorners=false
      bSeparateFillAlpha=false
      outlineMaterial=null
      outlineSize=0
      outlineColor=<rgba 0 0 0 1>
     }
    }
    shadowOffset=<0 0>
   }
   slot {class=HorizontalBoxSlot, properties {verticalAlignment=VAlign_Center}}
  }
  bugit_CSV {
   widgetClassPath=/Game/NeverCook/EditorUtilities/PerfTools/PerfCsvPlaybackTool/SubWidgets/PCPT_BugitLocationDisplay_EWidget
   bIsVariable=true
   parent=HorizontalBox_11
   widget=bugit_CSV
   slot {class=HorizontalBoxSlot, properties {size {sizeRule=Fill, value=1}}}
  }
  HorizontalBox {
   widgetClassPath=&HorizontalBox
   parent=VerticalBox_0
   widget=HorizontalBox
   slot {class=VerticalBoxSlot, properties {padding=<ltrb 10 5 10 0>}}
  }
  text_1 {
   widgetClassPath=&CommonTextBlock
   parent=HorizontalBox
   widget=text_1
   properties {
    justification=Right
    minDesiredWidth=140
    text="Camera Location: "
    visibility=Visible
    font {
     bForceMonospaced=false
     bMaterialIsStencil=false
     fontMaterial=null
     fontObject=&NotoSans
     letterSpacing=0
     monospacedWidth=1
     size=12
     skewAmount=0
     typefaceFontName=NotoSans-Regular
     outlineSettings {
      bApplyOutlineToDropShadows=false
      bMiteredCorners=false
      bSeparateFillAlpha=false
      outlineMaterial=null
      outlineSize=0
      outlineColor=<rgba 0 0 0 1>
     }
    }
    shadowOffset=<0 0>
   }
   slot {class=HorizontalBoxSlot, properties {verticalAlignment=VAlign_Center}}
  }
  bugIt_Camera {
   widgetClassPath=&PCPT_BugitLocationDisplay_EWidget
   bIsVariable=true
   parent=HorizontalBox
   widget=bugIt_Camera
   slot {class=HorizontalBoxSlot, properties {size {sizeRule=Fill, value=1}}}
  }
  PCPT_ShowStatsList_Widget {
   widgetClassPath=&/PCPT_ShowStatsList_Widget
   bIsVariable=true
   parent=Overlay_0
   widget=PCPT_ShowStatsList_Widget
   properties {visibility=Collapsed}
   slot {class=OverlaySlot, properties {horizontalAlignment=HAlign_Fill, verticalAlignment=VAlign_Fill}}
  }
 ]
}
